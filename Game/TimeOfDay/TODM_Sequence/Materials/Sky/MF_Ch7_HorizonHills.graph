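# Material function that tints a sky atmosphere color with a horizon hills texture and also outputs the hills mask.
# UV1 is unwrapped into eight segments around the horizon, each offset to its own vertical strip of the texture.
# Height comes from camera vector Z, or UV1 on mobile via ShadingPathSwitch, mirrored by Abs.
# Boosted texture alpha, clipped by a height cutoff and scaled down, blends the multiplied color over the input.
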MaterialFunction /Game/TimeOfDay/TODM_Sequence/Materials/Sky/MF_Ch7_HorizonHills
graph {nodes [
  n1 = FunctionInput() {inputName=Sky Atmosphere}
  n2 = Multiply(A=Add(A=TextureCoordinate() {coordinateIndex=1}, B=Constant2Vector() {g=-0.5, r=0.38}), B=Constant2Vector() {g=2, r=8})
  n3 = ComponentMask(None=@n2) {r=true}
  n4 = Multiply(A=Floor(None=@n3)) {constB=0.125}
  n5 = CameraVectorWS()
  n6 = ShadingPathSwitch(Default=Add(A=ComponentMask(None=@n5) {b=true}) {constB=-0.002}, Mobile=ComponentMask(None=@n2) {g=true})
  n7 = Multiply(A=Abs(None=@n6)) {constB=-2}
  n8 = TextureSample(UVs=AppendVector(A=@n3, B=Add(A=@n4, B=@n7))) {texture=/Game/Packages/Fortress_Sky/Textures/T_Ch7_HorizonHills}
  n9 = Subtract(A=Multiply(A=@n8.A) {constB=4})
  n10 = Ceil(None=Add(A=@n6) {constB=0.062})
  n11 = Multiply(A=Multiply(A=Saturate(None=@n9), B=Saturate(None=@n10))) {constB=0.35}
  FunctionOutput(None=LinearInterpolate(A=@n1, B=Multiply(A=@n1, B=@n8.RGB), Alpha=@n11) {constAlpha=1})
  FunctionOutput(None=@n11) {outputName=Mask_HorizonHills, sortPriority=10}
 ]}
