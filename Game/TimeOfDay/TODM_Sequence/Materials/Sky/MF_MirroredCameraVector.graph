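# Material function for the time-of-day sky that outputs the world-space camera vector scaled by a Multiplier input, defaulting to 1.
# A ReflectionCapturePassSwitch passes that vector through normally, but in reflection captures keeps XY and rebuilds Z as its absolute value times negated Multiplier.
# At the default multiplier this forces Z negative, mirroring the vector across the horizon so both hemispheres match in captures.
# The Offset input is declared but unconnected.
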MaterialFunction /Game/TimeOfDay/TODM_Sequence/Materials/Sky/MF_MirroredCameraVector
graph {nodes [
  FunctionInput() {bUsePreviewValueAsDefault=true, inputName=Offset, inputType=FunctionInput_Scalar, sortPriority=1}
  n1 = FunctionInput(Preview=Constant() {r=1}) {bUsePreviewValueAsDefault=true, inputName=Multiplier, inputType=FunctionInput_Scalar, previewValue=<1 1 1 1>}
  n2 = Multiply(A=CameraVectorWS(), B=@n1) {constB=-1}
  n3 = ComponentMask(None=@n2) {g=true, r=true}
  n4 = Multiply(A=Abs(None=ComponentMask(None=@n2) {b=true}), B=Multiply(A=@n1) {constB=-1}) {constB=-1}
  FunctionOutput(None=ReflectionCapturePassSwitch!(Default=@n2, Reflection=AppendVector(A=@n3, B=@n4)))
 ]}
