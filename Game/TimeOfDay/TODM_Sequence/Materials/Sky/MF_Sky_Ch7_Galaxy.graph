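# Material function for the chapter 7 time-of-day sky that adds a galaxy over the Sky Atmosphere input and outputs tiled star UVs.
# A rotated, scaled galaxy texture is alpha-masked, biased slightly negative, then multiplied by sequencer-animated opacity and the horizon-hills mask.
# Galaxy UVs and masks feed noise-varied star layers through named reroutes.
# Reflection captures receive only doubled atmosphere; spherical-projection and Milky Way branches remain unconnected.
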MaterialFunction /Game/TimeOfDay/TODM_Sequence/Materials/Sky/MF_Sky_Ch7_Galaxy
graph {nodes [
  n1 = Add(A=NamedRerouteUsage()) {constB=-0.5}
  n2 = AppendVector(
   A=ScalarParameter() {defaultValue=0.45, desc="2\r\n0.6", group=Galaxy, parameterName=Galaxy_Scale_X}
   B=ScalarParameter() {defaultValue=1.2, desc="0.6\r\n2", group=Galaxy, parameterName=Galaxy_Scale_Y}
  )
  n3 = TextureSample(UVs=Add(A=Divide(A=@n1, B=@n2)) {constB=0.5}) {samplerSource=SSM_Clamp_WorldGroupSettings, texture=/Game/TimeOfDay/TODM_Sequence/Materials/Sky/Textures/T_Galaxy_D}
  n4 = Multiply(A=Subtract(A=@n3.RGB, B=Constant() {r=0.05}) {constB=0.5, desc=Add some negative color values to teh blend}, B=@n3.A)
  n5 = ScalarParameter() {defaultValue=1, desc=Animated in the TOD sequencer, group=Galaxy, parameterName=Galaxy_Opacity}
  n6 = NamedRerouteUsage()
  n7 = FunctionInput() {bUsePreviewValueAsDefault=true, inputName=Mask_HorizonHills, inputType=FunctionInput_Scalar, previewValue=<1 1 1 1>}
  n8 = Multiply(A=Multiply(A=Multiply(A=@n4, B=@n5), B=Add(A=@n6)), B=@n7)
  n9 = FunctionInput() {inputName=Sky Atmosphere, sortPriority=-1}
  FunctionOutput(None=ReflectionCapturePassSwitch!(Default=Add(A=@n8, B=@n9), Reflection=Multiply(A=@n9) {constB=2})) {sortPriority=-1}
  n10 = ScalarParameter() {defaultValue=0.372, desc=0.6, group=Galaxy, parameterName="Galaxy_Rotation_OnZ(0-1)"}
  n11 = NamedRerouteUsage()
  n12 = Subtract(A=Rotator(Coordinate=TextureCoordinate(), Time=Multiply(A=@n10, B=@n11)), B=Constant2Vector() {r=-0.2})
  FunctionOutput(None=Multiply(
    A=NamedRerouteDeclaration(None=@n12) {name=UV_Galaxy, nodeColor=<rgba 1 0.5156 0 1>}
    B=ScalarParameter() {defaultValue=5, group=Galaxy, parameterName=Star_UV_GalaxyTile}
   )) {description=The 2D UVs of the virtual plane coordinates., outputName=Star UVs}
  n13 = MaterialFunctionCall!(Multiplier, Offset) {materialFunction=&/MF_MirroredCameraVector}
  n14 = ScalarParameter() {defaultValue=0.9, desc=0.740407, parameterName="Galaxy  Spherical Distribution"}
  n15 = Multiply(A=Add(A=ComponentMask(None=@n13.Result) {b=true}, B=Multiply(A=@n14) {constB=-1}) {constB=-0.2}, B=OneMinus(None=@n14)) {constB=0.8}
  n16 = ComponentMask(None=Multiply(A=Divide(B=@n15), B=@n13.Result)) {g=true, r=true}
  n17 = ScalarParameter() {defaultValue=0.414, parameterName=Galaxy Scale}
  n18 = TextureSample(UVs=Rotator(Coordinate=Multiply(A=@n16, B=@n17), Time=ScalarParameter() {defaultValue=0.3, parameterName=Milky Way Rotation})) {samplerSource=SSM_Clamp_WorldGroupSettings, texture=&T_Galaxy_D}
  Multiply(A=@n18.RGB, B=@n18.RGB)
  Multiply(A=@n18.A)
  n19 = ScalarParameter() {defaultValue=0.894, parameterName=Stars3Min}
  n20 = ScalarParameter() {defaultValue=1.3169, parameterName=Stars3Max}
  n21 = ScalarParameter() {defaultValue=2.0747, desc=2.074689, parameterName=MilkyWayScaleX}
  n22 = ScalarParameter() {defaultValue=1.2, desc=1.2, parameterName=MilkyWayScaleY}
  n23 = Add(
   A=Divide(B=AppendVector(A=@n21, B=@n22))
   B=AppendVector(
    A=ScalarParameter() {defaultValue=-0.0309, parameterName=MilkyWay_Offset_X}
    B=ScalarParameter() {defaultValue=-0.042, parameterName=MilkyWayOffset_Y}
   )
  )
  n24 = TextureSample(UVs=Multiply(A=@n23, B=ScalarParameter() {defaultValue=21.762, parameterName=ManyStarsTiling2})) {samplerType=SAMPLERTYPE_LinearColor, texture=/Game/TimeOfDay/TODM_Sequence/Materials/Sky/Textures/GalaxyTexture_M}
  n25 = CameraVectorWS()
  n26 = OneMinus(None=Abs(None=ComponentMask(None=@n25) {b=true}))
  n27 = ScalarParameter() {defaultValue=5.1798, parameterName=ColorOverrideExponent}
  n28 = Multiply(
   A=Multiply(B=LinearInterpolate(A=@n19, B=@n20, Alpha=@n24.R))
   B=LinearInterpolate(
    A=VectorParameter() {parameterName=Tint_Zeniith, defaultValue=<rgba 0.0908 0.0886 0.1031 1>}.RGB
    B=VectorParameter() {parameterName=Tint_Horizon, defaultValue=<rgba 2 1.4079 0.8117 1>}.RGB
    Alpha=Power(Base=@n26, Exp=@n27)
   )
  )
  n29 = ScalarParameter() {defaultValue=7.7823, parameterName=BigStarsTiling}
  n30 = LinearInterpolate(
   A=ScalarParameter() {defaultValue=1, parameterName=Stars2Min}
   B=ScalarParameter() {defaultValue=9, parameterName=Stars2Max}
   Alpha=TextureSample(UVs=Multiply(A=@n23, B=@n29)) {samplerType=SAMPLERTYPE_LinearColor, texture=&GalaxyTexture_M}.G
  )
  Multiply(B=Multiply(A=Multiply(A=@n28, B=@n30), B=VectorParameter() {parameterName=MilkyWayColorAndBrightness, defaultValue=<rgba 0.0698 0.0698 0.0698 1>}.RGB))
  n31 = Constant() {desc=0.6, r=0.6}
  n32 = NamedRerouteUsage()
  n33 = Subtract(
   A=Rotator(Coordinate=TextureCoordinate(), Time=Multiply(A=@n31, B=@n32))
   B=AppendVector(A=Constant() {desc=-0.18, r=-0.18}, B=Constant() {desc=-0.15, r=-0.15})
  ) {constB=0.5}
  n34 = Multiply(A=Constant() {desc=-0.25, r=-0.25}, B=NamedRerouteUsage())
  n35 = Constant() {desc=0}
  n36 = Constant()
  NamedRerouteDeclaration(None=Subtract(A=Rotator(Coordinate=@n33, Time=@n34), B=AppendVector(A=@n35, B=@n36))) {name=UV_Galaxy_1, nodeColor=<rgba 1 0.5156 0 1>}
  NamedRerouteDeclaration(None=MaterialFunctionCall(Multiplier=Constant() {r=8}) {materialFunction=&Pi}.Result) {name=360Rotation, nodeColor=<rgba 0 1 0.8125 1>}
  n37 = OneMinus(None=NamedRerouteUsage())
  n38 = ScalarParameter() {defaultValue=5, group=Galaxy, parameterName=Stars_GalaxyMask_Pow}
  n39 = ScalarParameter() {defaultValue=-2, group=Galaxy, parameterName=Galaxy_Stars1Min}
  n40 = ScalarParameter() {defaultValue=8, group=Galaxy, parameterName=Galaxy_Stars1Max}
  n41 = ScalarParameter() {defaultValue=20, group=Galaxy, parameterName=Galaxy_StarsTile}
  n42 = Multiply(A=NamedRerouteUsage(), B=@n41)
  n43 = TextureSample(UVs=Divide(A=@n42, B=ScalarParameter() {defaultValue=0.8, group=Galaxy, parameterName=Galaxy_Starts1_Scale})) {samplerType=SAMPLERTYPE_LinearColor, texture=&GalaxyTexture_M}
  n44 = TextureSample(UVs=Multiply(A=NamedRerouteUsage(), B=Multiply(A=@n41) {constB=0.4673})) {samplerType=SAMPLERTYPE_Grayscale, texture=/Game/TimeOfDay/TODM_Sequence/Materials/Sky/Textures/LowResBlurredNoise_Radish_LF_Root}
  n45 = Subtract(A=LinearInterpolate(A=@n43.R, B=@n43.R, Alpha=@n44.R), B=@n44.R)
  n46 = ScalarParameter() {defaultValue=-3, group=Galaxy, parameterName=Galaxy_Stars2Min}
  n47 = ScalarParameter() {defaultValue=20, group=Galaxy, parameterName=Galaxy_Stars2Max}
  n48 = TextureSample(UVs=Divide(A=@n42, B=ScalarParameter() {defaultValue=2.6, group=Galaxy, parameterName=Galaxy_StartsScale_2})) {samplerType=SAMPLERTYPE_LinearColor, texture=&GalaxyTexture_M}
  n49 = Power(Base=@n48.B, Exp=Constant() {r=4})
  n50 = ScalarParameter() {defaultValue=1, group=Galaxy, parameterName=Galaxy_Stars2_Glow}
  n51 = Max(
   A=Multiply(A=Power(Base=@n37, Exp=@n38), B=LinearInterpolate(A=@n39, B=@n40, Alpha=@n45))
   B=Max(A=LinearInterpolate(A=@n46, B=@n47, Alpha=@n49), B=Multiply(A=@n48.B, B=@n50))
  )
  NamedRerouteDeclaration(None=Multiply(A=Max(A=@n51) {constB=0}, B=NamedRerouteUsage())) {name=Galaxy_Stars, nodeColor=<rgba 0 1 0.8359 1>}
  n52 = NamedRerouteDeclaration(None=@n3.A) {name=M_Galaxy, nodeColor=<rgba 0.3828 0 1 1>}
  NamedRerouteDeclaration(None=ComponentMask(None=Multiply(A=@n3.RGB, B=@n52)) {r=true}) {name=Galaxy_R, nodeColor=<rgba 1 0 0.7734 1>}
  n53 = Reroute()
  ReflectionCapturePassSwitch(Default=@n53, Reflection=Multiply(A=@n53, B=ScalarParameter() {defaultValue=2, parameterName=reflctionCapture intensity}))
  Reroute()
  n54 = NamedRerouteUsage()
  n55 = Constant()
  SetMaterialAttributes!(MaterialAttributes, EmissiveColor=Step(Y=AppendVector(A=@n54, B=@n55)) {constX=0}) {attributeSetTypes [#1]}
 ]}
