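# Sky material function that looks up sky color from the SkyAtmosLUT texture by view direction and time of day.
# Camera vector becomes azimuth and elevation UVs, with elevation mirrored across the horizon and clamped near the edges.
# Time of day from FortniteMaterialParameters, plus a half-frame offset from the Frames parameter, supplies the third coordinate.
# A second sample at the target apparent time of day is blended in by the LUT Lerp parameter.
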MaterialFunction /Game/TimeOfDay/TODM_Sequence/Materials/Sky/MF_Sky_SkyAtmosLUT
graph {nodes [
  n1 = CameraVectorWS()
  n2 = ComponentMask(None=@n1) {g=true, r=true}
  n3 = Divide(
   A=Arctangent2Fast(Y=ComponentMask(None=@n2) {g=true}, X=ComponentMask(None=@n2) {r=true})
   B=Multiply(A=MaterialFunctionCall() {materialFunction=&Pi}.Result) {constB=2}
  )
  n4 = ComponentMask(None=@n1) {b=true}
  n5 = MaterialFunctionCall() {materialFunction=&Pi}
  n6 = Abs(None=Multiply(A=ArcsineFast!(None=@n4), B=Divide(B=@n5.Result) {constA=2}))
  n7 = AppendVector(A=Frac(None=Add(A=@n3) {constB=0.25}), B=Clamp(None=OneMinus(None=@n6)) {maxDefault=0.995, minDefault=0.005})
  n8 = CollectionParameter() {collection=&FortniteMaterialParameters, parameterName=TimeOfDay_0-1}
  n9 = Multiply(A=Divide(B=ScalarParameter() {defaultValue=24, parameterName=Frames}) {constB=24}) {constB=0.5}
  n10 = TextureObjectParameter() {parameterName=SkyLUT, samplerType=SAMPLERTYPE_LinearColor, texture=/Game/TimeOfDay/TODM_Sequence/Materials/Sky/Textures/T_SkyAtmosLUT}
  n11 = TextureSample(UVs=AppendVector(A=@n7, B=Add(A=@n8, B=@n9)), Tex=@n10) {
   samplerSource=SSM_Wrap_WorldGroupSettings
   samplerType=SAMPLERTYPE_LinearColor
   texture=&T_SkyAtmosLUT
  }
  n12 = CollectionParameter() {
   collection=&FortniteMaterialParameters
   parameterName=TargetApparentTimeOfDay_0-1
  }
  n13 = AppendVector(A=@n7, B=Add(A=@n9, B=@n12) {constB=0})
  FunctionOutput(None=LinearInterpolate(
    A=@n11.RGB
    B=TextureSample(UVs=@n13, Tex=@n10) {samplerType=SAMPLERTYPE_LinearColor, texture=&T_SkyAtmosLUT}.RGB
    Alpha=ScalarParameter() {parameterName=LUT Lerp}
   )) {sortPriority=-1}
  ScalarParameter() {defaultValue=0.25}
 ]}
