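# Material function for UEFN skybox clouds, outputting color, alpha and a high-altitude cloud layer.
# Panning, noise-warped UVs on a curved virtual plane sample coverage and normals, blending near and far scales on higher quality levels.
# Normals and sun direction drive lighting between time-of-day or custom colors, blended with sky atmosphere by aerial perspective.
# High-altitude clouds combine four panning texture samples under a radial fade, using simpler UVs on Low and Medium.
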MaterialFunction /Game/TimeOfDay/TODM_Sequence/Materials/Sky/Clouds/MF_UEFN_SkyboxClouds
graph {nodes [
  n1 = Constant() {r=0.005}
  n2 = MaterialFunctionCall(
   Vector=VectorParameter() {group=SkyboxClouds, parameterName=CloudDirection, defaultValue=<rgba 1 0 0 1>}.RGB
   Default=Constant3Vector() {constant=<rgba 1 0 0 0>}
  ) {materialFunction=&SafeNormalize}
  n3 = Panner(Speed=ComponentMask(None=Multiply(A=@n1, B=@n2.Result)) {g=true, r=true})
  n4 = ScalarParameter() {defaultValue=10, group=SkyboxClouds, parameterName=CloudFar_MicroNoise_UVTile_Far}
  n5 = Multiply(
   A=TextureSample(UVs=Multiply(A=@n3, B=@n4) {constB=0.5}) {texture=&/T_Clouds_02_M}.R
   B=ScalarParameter() {defaultValue=0.005, group=SkyboxClouds, parameterName=CloudFarUVWarp}
  )
  AppendVector(A=Constant(), B=@n5)
  Constant2Vector() {g=500, r=500}
  DoubleVectorParameter() {group=SkyboxClouds, parameterName=Cloud_MicroNoise_UVDirection, defaultValue=<0.005 0 0 1>}
  n6 = SkyAtmosphereViewLuminance!(WorldDirection)
  n7 = NamedRerouteUsage()
  n8 = SkyAtmosphereAerialPerspective!(World Position=NamedRerouteUsage())
  n9 = LinearInterpolate(
   B=MaterialFunctionCall(Float4=FunctionOutput(None=@n8) {outputName=AerialPerspective}) {materialFunction=&BreakOutFloat4Components}.A
   Alpha=Multiply(A=ScalarParameter() {defaultValue=1, group=SkyboxClouds, parameterName=AerialPerspectvieInfluence}, B=NamedRerouteUsage())
  ) {constA=1}
  FunctionOutput(None=NamedRerouteDeclaration(None=LinearInterpolate(A=@n6, B=@n7, Alpha=@n9)) {name=CloudColorOUT, nodeColor=<rgba 1 0 0.4687 1>}) {outputName=CloudColor}
  FunctionOutput(None=NamedRerouteUsage()) {outputName=CloudAlpha, sortPriority=1}
  n10 = SphereMask(
   A=TextureCoordinate()
   B=Constant2Vector() {g=0.5, r=0.5}
   Radius=ScalarParameter() {defaultValue=0.456, group=SkyboxClouds, parameterName=HA_Clouds_Fade_Radius}
   Hardness=ScalarParameter() {defaultValue=0.7359, group=SkyboxClouds, parameterName=HA_Clouds_Fade_Hardness}
  ) {attenuationRadius=0.5, hardnessPercent=0}
  n11 = ScalarParameter() {defaultValue=3000000, group=SkyboxClouds, parameterName=HA_Cloud_VirtualplaneScale}
  n12 = MaterialFunctionCall(
   Plane Normal Axis=Constant3Vector() {constant=<rgba 0 0 -1 0>}
   Plane X Axis=Constant3Vector() {constant=<rgba 1 0 0 0>}
   Plane Center=VectorParameter() {group=SkyboxClouds, parameterName=HA_Cloud_Height, defaultValue=<rgba 0 0 550000 1>}.RGB
   UV Scale=AppendVector(A=@n11, B=@n11)
  ) {materialFunction=&VirtualPlaneCoordinates}
  n13 = Rotator(
   Coordinate=Multiply(A=TextureCoordinate(), B=ScalarParameter() {defaultValue=1, group=SkyboxClouds, parameterName=HighAlititudeClouds_Tiling_Master})
   Time=ScalarParameter() {defaultValue=11.5849, group=SkyboxClouds, parameterName=HighAlititudeClouds_Rotation}
  )
  n14 = Panner(
   Coordinate=QualitySwitch(Default=@n12.UVs, Low=@n13, High=@n12.UVs, Medium=@n13, Epic=@n12.UVs)
   Speed=AppendVector(
    A=ScalarParameter() {defaultValue=-0.002, group=SkyboxClouds, parameterName=HA_Cloud_Speed_X}
    B=ScalarParameter() {defaultValue=-0.001, group=SkyboxClouds, parameterName=HA_Cloud_Speed_Y}
   )
  )
  n15 = TextureSample(UVs=Multiply(A=@n14, B=ScalarParameter() {defaultValue=2.1989, group=SkyboxClouds, parameterName=HighAlititudeClouds_Tiling})) {samplerSource=SSM_Wrap_WorldGroupSettings, texture=/Game/TimeOfDay/TODM_Sequence/Materials/Sky/Textures/HighAltitudeClouds_M}
  n16 = Panner(
   Coordinate=@n14
   Speed=AppendVector(
    A=ScalarParameter() {defaultValue=0.0015, group=SkyboxClouds, parameterName=HA_Cloud_Speed2_X}
    B=ScalarParameter() {defaultValue=0.0008, group=SkyboxClouds, parameterName=HA_Cloud_Speed2_Y}
   )
  )
  n17 = TextureSample(UVs=Multiply(A=@n16, B=ScalarParameter() {defaultValue=1.4899, group=SkyboxClouds, parameterName=HighAlititudeClouds_Tiling_3})) {samplerSource=SSM_Wrap_WorldGroupSettings, texture=&HighAltitudeClouds_M}
  n18 = ScalarParameter() {defaultValue=2, group=SkyboxClouds, parameterName=HighAlititudeClouds_Tiling_2}
  n19 = ScalarParameter() {defaultValue=1, group=SkyboxClouds, parameterName=HighAlititudeClouds_Tiling_4}
  n20 = Multiply(
   A=TextureSample(UVs=Multiply(A=@n16, B=@n18)) {samplerSource=SSM_Wrap_WorldGroupSettings, texture=&HighAltitudeClouds_M}.G
   B=TextureSample(UVs=Multiply(A=@n14, B=@n19)) {samplerSource=SSM_Wrap_WorldGroupSettings, texture=&HighAltitudeClouds_M}.B
  )
  n21 = Multiply(A=Add(A=Multiply(A=@n15.R, B=@n17.B), B=Multiply(A=@n20) {constB=0.75}), B=NamedRerouteUsage())
  FunctionOutput(None=NamedRerouteDeclaration(None=Multiply(A=@n10, B=@n21)) {name=HighAltitudeClouds, nodeColor=<rgba 0.4297 0 1 1>}) {outputName=HighAltitudeClouds}
  n22 = SkyAtmosphereLightIlluminance!(World Position=NamedRerouteUsage())
  n23 = Multiply(
   A=ComponentMask(None=FunctionOutput(None=@n22) {outputName=LightIlluminance}) {b=true, g=true, r=true}
   B=ScalarParameter() {defaultValue=0.2, group=SkyboxClouds, parameterName=LightIlluminanceStrength}
  ) {constB=0.2}
  LinearInterpolate(B=@n23, Alpha=ScalarParameter() {defaultValue=1, group=SkyboxClouds, parameterName=LightIlluminanceInfluence}) {constA=1}
  n24 = ScalarParameter() {defaultValue=1, group=SkyboxClouds, parameterName=CloudSpeed}
  Multiply(A=@n24) {constB=0.005}
  n25 = VectorParameter() {group=SkyboxClouds, parameterName=CloudShadowColor_Morning, defaultValue=<rgba 0.0667 0.1095 0.2865 1>}
  n26 = VectorParameter() {group=SkyboxClouds, parameterName=CloudShadowColor_Day, defaultValue=<rgba 0.151 0.2806 0.5 1>}
  n27 = VectorParameter() {group=SkyboxClouds, parameterName=CloudShadowColor_Evening, defaultValue=<rgba 0.1198 0.0318 0.0329 1>}
  n28 = VectorParameter() {group=SkyboxClouds, parameterName=CloudShadowColor_Night, defaultValue=<rgba 0.0279 0.0444 0.1094 1>}
  NamedRerouteDeclaration(None=LinearInterpolate(
    A=MaterialFunctionCall(Color_Morning=@n25.RGB, Color_Day=@n26.RGB, Color_Evening=@n27.RGB, Color_Night=@n28.RGB) {materialFunction=&Color_Time_of_Day}.Result
    B=VectorParameter() {group=SkyboxClouds, parameterName=CloudShadowColor_Custom, defaultValue=<rgba 0.151 0.2806 0.5 1>}.RGB
    Alpha=NamedRerouteUsage()
   )) {name=CloudShadowColor, nodeColor=<rgba 1 0 0.3516 1>}
  n29 = ScalarParameter() {defaultValue=0.5, group=SkyboxClouds, parameterName=Cloud Interface Exp, sortPriority=102}
  n30 = CameraVectorWS()
  n31 = Saturate(None=Multiply(A=ComponentMask(None=@n30) {b=true}) {constB=-1})
  NamedRerouteDeclaration(None=MaterialFunctionCall(
    In=Power(Base=@n31, Exp=@n29) {constExponent=3}
    Contrast=ScalarParameter() {defaultValue=2, group=SkyboxClouds, parameterName=Cloud Interface Contrast, sortPriority=103}
   ) {materialFunction=&CheapContrast}.Result) {name=CloudNearFarMask, nodeColor=<rgba 0.3359 0 1 1>}
  NamedRerouteDeclaration(None=MaterialFunctionCall(
    Value_Morning=ScalarParameter() {defaultValue=0.5, group=SkyboxClouds, parameterName=Morning_Brightness}
    Value_Day=ScalarParameter() {defaultValue=0.75, group=SkyboxClouds, parameterName=Day_Brightness}
    Value_Evening=ScalarParameter() {defaultValue=0.5, group=SkyboxClouds, parameterName=EveningBrightness}
    Value_Night=ScalarParameter() {defaultValue=0.05, group=SkyboxClouds, parameterName=NightBrightness}
   ) {materialFunction=&Scalar_Time_of_Day}.Result) {name=TOD_Brightness, nodeColor=<rgba 0 0.7656 1 1>}
  n32 = MaterialFunctionCall(
   Plane Normal Axis=Constant3Vector() {constant=<rgba 0 0 -1 0>}
   Plane X Axis=Constant3Vector() {constant=<rgba 1 0 0 0>}
   Plane Center=AppendVector(A=Constant2Vector(), B=NamedRerouteUsage())
  ) {materialFunction=&VirtualPlaneCoordinates}
  NamedRerouteDeclaration(None=FunctionOutput(None=NamedRerouteDeclaration(None=@n32.World Position) {name=VirtualPlaneWSP, nodeColor=<rgba 1 0.9375 0 1>}) {outputName=WorldPosition}) {name=WorldSpaceUV, nodeColor=<rgba 0 0.2734 1 1>}
  n33 = Multiply(A=Time(), B=Multiply(A=@n24, B=Constant() {r=10000}))
  n34 = ComponentMask(None=Add(A=NamedRerouteUsage(), B=Multiply(A=@n33, B=@n2.Result))) {g=true, r=true}
  n35 = Multiply(A=Constant() {r=10000000}, B=ScalarParameter() {defaultValue=1.5, group=SkyboxClouds, parameterName=CloudSize})
  NamedRerouteDeclaration(None=Divide(A=@n34, B=@n35)) {name=CloudUVPan, nodeColor=<rgba 0 1 0.7891 1>}
  n36 = NamedRerouteUsage()
  n37 = ScalarParameter() {defaultValue=20000000, group=SkyboxClouds, parameterName=CloudMaskDistance}
  NamedRerouteDeclaration(None=SphereMask(
    A=Constant2Vector()
    B=ComponentMask(None=@n36) {g=true, r=true}
    Radius=NamedRerouteDeclaration(None=@n37) {name=CloudMaskDistance, nodeColor=<rgba 0.3828 1 0 1>}
    Hardness=ScalarParameter() {defaultValue=0.4, group=SkyboxClouds, parameterName=CloudMaskHardness}
   ) {hardnessPercent=0.5}) {name=CloudDistanceMask, nodeColor=<rgba 0.4531 0 1 1>}
  n38 = Multiply(A=SkyAtmosphereLightDirection(), B=ScalarParameter() {defaultValue=1000000, group=SkyboxClouds, parameterName=SunDistance}) {constB=1000000}
  n39 = AppendVector(A=ComponentMask(None=NamedRerouteUsage()) {g=true, r=true}, B=Constant())
  NamedRerouteDeclaration(None=Normalize(VectorInput=Subtract(A=@n38, B=@n39))) {name=SunToWSP, nodeColor=<rgba 0 0.1094 1 1>}
  NamedRerouteDeclaration(None=ScalarParameter() {parameterName=UseCustomCloudColor}) {name=UseCustomCloudColor, nodeColor=<rgba 1 0 0.6328 1>}
  n40 = NamedRerouteUsage()
  n41 = Multiply(
   A=NamedRerouteUsage()
   B=Multiply(A=ScalarParameter() {defaultValue=1, group=SkyboxClouds, parameterName=CloudLightStrength}, B=NamedRerouteUsage())
  ) {constB=2}
  n42 = DotProduct(A=NamedRerouteUsage(), B=NamedRerouteUsage())
  n43 = Saturate(None=Power(Base=ConstantBiasScale(None=@n42), Exp=ScalarParameter() {defaultValue=3, group=SkyboxClouds, parameterName=CloudNormalExponent}))
  n44 = SkyAtmosphereLightDirection()
  n45 = Power(Base=DotProduct(A=Multiply(A=@n44) {constA=-1, constB=-1}, B=CameraVectorWS()))
  n46 = ScalarParameter() {group=SkyboxClouds, parameterName=CloudSunBrightness}
  NamedRerouteDeclaration(None=Add(A=Multiply(A=@n45, B=@n46), B=LinearInterpolate(A=@n40, B=@n41, Alpha=@n43))) {name=CloudLighting, nodeColor=<rgba 1 0 0.7969 1>}
  NamedRerouteDeclaration(None=@n23) {name=LightIlluminance, nodeColor=<rgba 0.6406 1 0 1>}
  n47 = Length(None=ComponentMask(None=WorldPosition().XYZ) {g=true, r=true}) {is=class MaterialExpressionLength}
  n48 = NamedRerouteUsage()
  n49 = OneMinus(None=Saturate(None=Divide(A=@n47, B=@n48)))
  n50 = ScalarParameter() {defaultValue=0.65, group=SkyboxClouds, parameterName=CloudHeightCurveExponent}
  NamedRerouteDeclaration(None=LinearInterpolate(
    A=ScalarParameter() {defaultValue=50000, group=SkyboxClouds, parameterName=CloudHeightMin}
    B=ScalarParameter() {defaultValue=1000000, group=SkyboxClouds, parameterName=CloudHeightMax}
    Alpha=Power(Base=@n49, Exp=@n50)
   )) {name=Height, nodeColor=<rgba 0.0078 1 0 1>}
  n51 = ScalarParameter() {defaultValue=2.5, group=SkyboxClouds, parameterName=CloudSizeDistantMultiplier}
  NamedRerouteDeclaration(None=Divide(A=@n34, B=Multiply(A=@n35, B=@n51))) {name=CloudUVPanFar, nodeColor=<rgba 0.7109 1 0 1>}
  n52 = NamedRerouteUsage()
  n53 = Add(A=@n52, B=ScalarParameter() {defaultValue=0.4, group=SkyboxClouds, parameterName=CloudFarSoftness})
  n54 = TextureSampleParameter2D(UVs=NamedRerouteUsage()) {
   group=SkyboxClouds
   parameterName=CloudMask
   samplerSource=SSM_Wrap_WorldGroupSettings
   samplerType=SAMPLERTYPE_LinearGrayscale
   sortPriority=0
   texture=&/T_Asteria_Skybox_Clouds
  }
  n55 = ScalarParameter() {defaultValue=0.3, group=SkyboxClouds, parameterName=CloudCoverage}
  n56 = NamedRerouteDeclaration(None=OneMinus(None=Saturate(None=@n55))) {name=CloudCoverage, nodeColor=<rgba 0.5469 1 0 1>}
  n57 = ScalarParameter() {defaultValue=0.2, group=SkyboxClouds, parameterName=CloudSoftness}
  n58 = TextureSampleParameter2D(UVs=NamedRerouteUsage()) {
   group=SkyboxClouds
   parameterName=CloudMask
   samplerSource=SSM_Wrap_WorldGroupSettings
   samplerType=SAMPLERTYPE_LinearGrayscale
   sortPriority=0
   texture=&T_Asteria_Skybox_Clouds
  }
  n59 = SmoothStep(Min=@n56, Max=Saturate(None=Add(A=@n56, B=@n57)), Value=@n58.G) {constMax=0.3, constMin=0.2, constValue=0.6}
  n60 = LinearInterpolate(
   A=SmoothStep(Min=@n52, Max=Saturate(None=@n53), Value=@n54.G) {constMax=0.3, constMin=0.2, constValue=0.6}
   B=@n59
   Alpha=NamedRerouteUsage()
  )
  n61 = QualitySwitch(Default=@n60, Low=@n59, High=@n60, Medium=@n59, Epic=@n60)
  n62 = NamedRerouteUsage()
  NamedRerouteDeclaration(None=Saturate(None=Multiply(A=@n61, B=@n62))) {name=CloudAlpha, nodeColor=<rgba 0.5703 1 0 1>}
  n63 = VectorParameter() {group=SkyboxClouds, parameterName=CloudColor_Morning, defaultValue=<rgba 0.2292 0.0892 0.0146 1>}
  n64 = VectorParameter() {group=SkyboxClouds, parameterName=CloudColor_Day, defaultValue=<rgba 0.2552 0.2552 0.2552 1>}
  n65 = VectorParameter() {group=SkyboxClouds, parameterName=CloudColor_Evening, defaultValue=<rgba 0.1823 0.0364 0.009 1>}
  n66 = VectorParameter() {group=SkyboxClouds, parameterName=CloudColor_Night, defaultValue=<rgba 0.0319 0.0508 0.125 1>}
  NamedRerouteDeclaration(None=LinearInterpolate(
    A=MaterialFunctionCall(Color_Morning=@n63.RGB, Color_Day=@n64.RGB, Color_Evening=@n65.RGB, Color_Night=@n66.RGB) {materialFunction=&Color_Time_of_Day}.Result
    B=VectorParameter() {group=SkyboxClouds, parameterName=CloudColor_Custom, defaultValue=<rgba 0.2552 0.2552 0.2552 1>}.RGB
    Alpha=NamedRerouteUsage()
   )) {name=CloudLightingColor, nodeColor=<rgba 1 0 0.8203 1>}
  NamedRerouteDeclaration(None=Add(A=NamedRerouteUsage(), B=ConstantBiasScale(None=@n5) {bias=-0.5, scale=2})) {name=CloudFarUVs, nodeColor=<rgba 0 1 0.7422 1>}
  NamedRerouteDeclaration(None=@n32.World Position - Centered) {name=VirtualPlandWSPCentered, nodeColor=<rgba 1 0 0.7031 1>}
  NamedRerouteDeclaration(None=@n32.UVs) {name=VirtualPlaneUV, nodeColor=<rgba 1 0.7266 0 1>}
  n67 = ScalarParameter() {defaultValue=4, group=SkyboxClouds, parameterName=Cloud_MicroNoise_UVTile}
  n68 = ConstantBiasScale(None=TextureSample(UVs=Multiply(A=@n3, B=@n67) {constB=0.5}) {texture=&T_Clouds_02_M}.R) {bias=-0.5, scale=2}
  n69 = ScalarParameter() {defaultValue=0.02, group=SkyboxClouds, parameterName=CloudUVWarp}
  NamedRerouteDeclaration(None=Add(A=NamedRerouteUsage(), B=Multiply(A=@n68, B=@n69))) {name=CloudUVs, nodeColor=<rgba 1 0 0.0469 1>}
  n70 = TextureSampleParameter2D(UVs=NamedRerouteUsage()) {
   group=SkyboxClouds
   parameterName=CloudNormal
   samplerType=SAMPLERTYPE_Normal
   sortPriority=1
   texture=&/T_Asteria_Skybox_Clouds_N
  }
  n71 = TextureSampleParameter2D(UVs=NamedRerouteUsage()) {
   group=SkyboxClouds
   parameterName=CloudNormal
   samplerType=SAMPLERTYPE_Normal
   sortPriority=1
   texture=&T_Asteria_Skybox_Clouds_N
  }
  n72 = LinearInterpolate(A=@n70.RGB, B=@n71.RGB, Alpha=NamedRerouteUsage())
  NamedRerouteDeclaration(None=Normalize(VectorInput=QualitySwitch(Default=@n72, Low=@n71.RGB, High=@n72, Medium=@n71.RGB, Epic=@n72))) {name=CloudNormal, nodeColor=<rgba 1 0.7266 0 1>}
  NamedRerouteUsage()
  TextureCoordinate()
 ]}
