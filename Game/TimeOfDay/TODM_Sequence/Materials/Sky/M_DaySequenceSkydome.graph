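# Unlit sky material for the Day Sequence skydome, layering day and night skies, sun, moon and clouds.
# Sky Atmosphere luminance, a Milky Way night sky and a three-color gradient blend through scalar parameters, while sun/moon UVs derive from atmosphere light direction.
# Stars use masked tiled textures and panning-noise modulation, simplified at Low and Medium quality.
# A skybox cloud function composites on top, enabled by parameter collection or local scalar.
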Material /Game/TimeOfDay/TODM_Sequence/Materials/Sky/M_DaySequenceSkydome
graph {
 properties {bIsSky=true, pixelDepthOffsetMode=PDOM_Legacy, shadingModel=MSM_Unlit, usedShadingModels=Unlit}
 nodes [
  n1 = ScalarParameter() {defaultValue=-0.5, parameterName=GradientBias}
  n2 = Multiply(
   A=OneMinus(None=@n1)
   B=MaterialFunctionCall() {materialFunction=&BoundingBoxBased_0-1_UVW}.B
  )
  NamedRerouteDeclaration(None=Saturate(None=Add(A=@n2, B=@n1))) {name=VerticalGradient, nodeColor=<rgba 0 0.1562 1 1>}
  n3 = VectorParameter() {parameterName=ColorLow, defaultValue=<rgba 1 0.0615 0 1>}
  n4 = VectorParameter() {parameterName=ColorMid, defaultValue=<rgba 1 0.6067 0.349 1>}
  n5 = Saturate(None=LinearInterpolate(A=NamedRerouteUsage(), B=NamedRerouteUsage(), Alpha=ScalarParameter() {parameterName=ColorGradientUseLightDirection}))
  n6 = Saturate(None=ScalarParameter() {defaultValue=0.5, parameterName=GradientCenter})
  n7 = Saturate(None=Divide(A=@n5, B=@n6))
  n8 = Divide(A=Subtract(A=@n5, B=@n6), B=@n6)
  NamedRerouteDeclaration(None=LinearInterpolate(
    A=LinearInterpolate(A=@n3.RGB, B=@n4.RGB, Alpha=@n7)
    B=VectorParameter() {parameterName=ColorHigh, defaultValue=<rgba 0.3338 0.7006 1 1>}.RGB
    Alpha=Saturate(None=@n8)
   )) {name=GradientColor, nodeColor=<rgba 0 0.4844 1 1>}
  NamedRerouteDeclaration(None=Multiply(A=NamedRerouteUsage(), B=ScalarParameter() {defaultValue=0.25, parameterName=Milkway Intensity}) {constB=0.25}) {name=Milkyway, nodeColor=<rgba 0 0.4609 1 1>}
  n9 = SkyAtmosphereLightDiskLuminance!(DiskAngularDiameterOverride=ScalarParameter() {defaultValue=2, group=Sun-Moon, parameterName=SunSize})
  n10 = Multiply(
   A=Desaturation(None=@n9, Fraction=Constant() {r=1})
   B=NamedRerouteDeclaration(None=VectorParameter() {parameterName=SunColor, defaultValue=<rgba 1 1 1 1>}.RGB) {name=SunColor, nodeColor=<rgba 0.7578 0 1 1>}
  )
  n11 = ScalarParameter() {parameterName=SunColorBlend}
  n12 = ScalarParameter() {defaultValue=1, parameterName=SunBrightness}
  NamedRerouteDeclaration(None=Multiply(A=LinearInterpolate(A=@n9, B=@n10, Alpha=@n11), B=NamedRerouteDeclaration(None=@n12) {name=SunBrightness, nodeColor=<rgba 0.0781 1 0 1>})) {name=Sun_1, nodeColor=<rgba 1 0.9141 0 1>}
  NamedRerouteDeclaration(None=TextureSample(UVs=NamedRerouteUsage()) {samplerSource=SSM_Wrap_WorldGroupSettings, texture=/Game/TimeOfDay/TODM_Sequence/Materials/Sky/Textures/T_milkyway_StarMask}.G) {name=StarMask, nodeColor=<rgba 1 0.8438 0 1>}
  n13 = Constant()
  n14 = NamedRerouteUsage()
  NamedRerouteDeclaration(None=Add(A=TextureCoordinate(), B=AppendVector(A=@n13, B=@n14))) {name=star pan, nodeColor=<rgba 0.3594 0 1 1>}
  NamedRerouteDeclaration(None=ScalarParameter() {defaultValue=0.016, parameterName=NightSkyPan}) {name=NightSkyPanSpeed, nodeColor=<rgba 0 1 0.4375 1>}
  n15 = WorldPosition()
  n16 = CameraPositionWS()
  NamedRerouteDeclaration(None=Normalize(VectorInput=Subtract(A=@n15.XYZ, B=@n16))) {name=CameraVectorForVertexShader, nodeColor=<rgba 0 1 0.0859 1>}
  n17 = NamedRerouteUsage()
  n18 = ScalarParameter() {defaultValue=0.07, group=Sun-Moon, parameterName=MoonSize}
  n19 = Add(A=Divide(A=NamedRerouteUsage(), B=NamedRerouteDeclaration(None=@n18) {name=MoonSize, nodeColor=<rgba 0.4297 0 1 1>})) {constB=0.5}
  n20 = TextureSampleParameter2D(UVs=NamedRerouteDeclaration(None=VertexInterpolator(VS=@n19).PS) {name=MoonUVs, nodeColor=<rgba 0 1 0.7188 1>}) {
   group=Sun-Moon
   parameterName=MoonTexture
   samplerSource=SSM_Clamp_WorldGroupSettings
   samplerType=SAMPLERTYPE_LinearColor
   texture=/Game/TimeOfDay/Skydome/T_Moon_Artemis
  }
  n21 = VectorParameter() {group=Sun-Moon, parameterName=MoonColor, defaultValue=<rgba 1 1 1 1>}
  n22 = Multiply(
   A=Multiply(A=Power(Base=@n20.R) {constExponent=1.2}, B=NamedRerouteDeclaration(None=@n21.RGB) {name=MoonColor, nodeColor=<rgba 0.125 1 0 1>})
   B=ScalarParameter() {defaultValue=17, group=Sun-Moon, parameterName=MoonIntensity}
  )
  NamedRerouteDeclaration(None=Multiply(A=NamedRerouteUsage(), B=Add(A=@n17, B=@n22))) {name=Moon, nodeColor=<rgba 0 1 0.8125 1>}
  n23 = Multiply(A=@n20.R) {constB=20}
  NamedRerouteDeclaration(None=Multiply(A=Saturate(None=@n23))) {name=MoonMask, nodeColor=<rgba 0 1 0.1094 1>}
  n24 = ScalarParameter() {defaultValue=0.8, parameterName=MoonHaloSize}
  n25 = SphereMask(
   A=NamedRerouteUsage()
   B=Constant() {r=0.5}
   Radius=Add(A=Multiply(A=@n24) {constB=0.5}) {constB=0.5}
   Hardness=ScalarParameter() {parameterName=MoonHaloHardness}
  ) {hardnessPercent=0.1}
  n26 = Multiply(A=@n25, B=@n25)
  n27 = NamedRerouteUsage()
  NamedRerouteDeclaration(None=Multiply(A=Multiply(A=@n26, B=@n27), B=ScalarParameter() {defaultValue=0.0003, parameterName=MoonHaloIntensity})) {name=MoonHalo, nodeColor=<rgba 0 1 0.7656 1>}
  n28 = ScalarParameter() {defaultValue=1000, group=Sun-Moon, parameterName=SunIntensity}
  n29 = NamedRerouteUsage()
  n30 = ScalarParameter() {defaultValue=2, group=Sun-Moon, parameterName=SunSize}
  n31 = Multiply(
   A=VectorParameter() {group=Sun-Moon, parameterName=SunColor, defaultValue=<rgba 1 1 1 1>}.RGB
   B=SphereMask(A=Constant2Vector(), B=Divide(A=@n29, B=@n30), Hardness=ScalarParameter() {defaultValue=0.5, group=Sun-Moon, parameterName=SunHardness}) {attenuationRadius=0.5}
  )
  NamedRerouteDeclaration(None=Multiply(A=Multiply(A=@n28, B=@n31), B=NamedRerouteUsage())) {name=SunSphereMask, nodeColor=<rgba 0.8281 0 1 1>}
  n32 = NamedRerouteUsage()
  n33 = ComponentMask(None=SkyAtmosphereLightDirection()) {b=true, g=true, r=true}
  NamedRerouteDeclaration(None=Step(X=DotProduct(A=@n32, B=@n33))) {name=HemisphereMask, nodeColor=<rgba 1 0 0.3047 1>}
  n34 = Multiply(A=@n33) {constB=-1}
  n35 = CrossProduct(A=@n34, B=Constant3Vector() {constant=<rgba 0 0 -1 0>})
  n36 = DotProduct(A=Normalize(VectorInput=CrossProduct(A=@n35, B=@n34)), B=NamedRerouteUsage())
  n37 = Normalize(VectorInput=@n35)
  n38 = NamedRerouteUsage()
  NamedRerouteDeclaration(None=AppendVector(A=@n36, B=DotProduct(A=@n37, B=@n38))) {name=SunMoonUVs, nodeColor=<rgba 1 0 0.9141 1>}
  NamedRerouteDeclaration(None=Multiply(A=SkyAtmosphereViewLuminance!(WorldDirection), B=ScalarParameter() {defaultValue=1, parameterName=SkyAtmosphereBrightness})) {name=DaySky, nodeColor=<rgba 1 0 0.7031 1>}
  n39 = Desaturation(None=NamedRerouteUsage(), Fraction=ScalarParameter() {parameterName=MilkywayDesaturation})
  n40 = ScalarParameter() {defaultValue=1, parameterName=MilkywayBrightness}
  NamedRerouteDeclaration(None=MaterialFunctionCall!(Blend=VectorParameter() {parameterName=NightSkyBlendColor, defaultValue=<rgba 0.015 0.0123 0.0149 0>}.RGB, Base=Multiply(A=@n39, B=@n40)) {materialFunction=/Engine/Functions/Engine_MaterialFunctions03/Blends/Blend_Screen}.Result) {name=NightSky, nodeColor=<rgba 0.4531 0 1 1>}
  NamedRerouteDeclaration(None=@n36) {name=TestSunVector, nodeColor=<rgba 1 0.2344 0 1>}
  n41 = DotProduct(
   A=SkyAtmosphereLightDirection()
   B=Normalize(VectorInput=MaterialFunctionCall() {materialFunction=&LocalPosition}.Local Position)
  )
  NamedRerouteDeclaration(None=Saturate(None=ConstantBiasScale(None=@n41))) {name=LightDirectionGradient, nodeColor=<rgba 1 0.4688 0 1>}
  n42 = NamedRerouteUsage()
  n43 = NamedRerouteUsage()
  n44 = ScalarParameter() {parameterName=NightSkyActive}
  NamedRerouteDeclaration(None=LinearInterpolate(A=LinearInterpolate(A=@n42, B=@n43, Alpha=@n44), B=NamedRerouteUsage(), Alpha=ScalarParameter() {parameterName=GradientColorStrength})) {name=FinalSky, nodeColor=<rgba 0.8047 0 1 1>}
  n45 = NamedRerouteUsage()
  n46 = ScalarParameter() {defaultValue=0.5, parameterName=StarsDesaturate}
  NamedRerouteDeclaration(None=Multiply(A=Desaturation(None=@n45, Fraction=@n46), B=ScalarParameter() {parameterName=StarsBrightness})) {name=FinalStars, nodeColor=<rgba 1 0.7969 0 1>}
  n47 = MaterialFunctionCall() {materialFunction=/Game/TimeOfDay/TODM_Sequence/Materials/Sky/Clouds/MF_UEFN_SkyboxClouds}
  NamedRerouteDeclaration(None=@n47.CloudColor) {name=CloudColor2, nodeColor=<rgba 1 0.75 0 1>}
  NamedRerouteDeclaration(None=@n47.CloudAlpha) {name=CloudAlpha2, nodeColor=<rgba 0 1 0.7188 1>}
  NamedRerouteDeclaration(None=@n47.HighAltitudeClouds) {name=HAClouds, nodeColor=<rgba 0 1 0.5781 1>}
  n48 = Multiply(A=NamedRerouteUsage(), B=Constant() {r=60})
  n49 = NamedRerouteUsage()
  n50 = Multiply(A=@n49, B=Constant() {r=12}) {constB=5}
  n51 = ScalarParameter() {defaultValue=200, parameterName=smallStarIndensity}
  n52 = Constant() {r=80}
  n53 = Power(
   Base=TextureSample(UVs=Multiply(A=@n49, B=@n52)) {samplerSource=SSM_Wrap_WorldGroupSettings, texture=&T_milkyway_StarMask}.B
   Exp=Constant() {r=2}
  ) {constExponent=4}
  n54 = NamedRerouteUsage()
  n55 = Constant() {r=4.2}
  n56 = TextureSample(UVs=Multiply(A=@n54, B=Constant() {r=30})) {samplerSource=SSM_Wrap_WorldGroupSettings, texture=&T_milkyway_StarMask}
  n57 = Constant() {r=4}
  n58 = Multiply(
   A=TextureSample(UVs=Multiply(A=@n54, B=@n55) {constB=5}) {samplerSource=SSM_Wrap_WorldGroupSettings, texture=/Game/TimeOfDay/TODM_Sequence/Materials/Sky/Textures/T_SkyDome_Stars_Sharp}.RGB
   B=LinearInterpolate(B=ScalarParameter() {defaultValue=2000, parameterName=BigStarIndensity}, Alpha=Power(Base=@n56.B, Exp=@n57) {constExponent=4}) {constB=0}
  )
  n59 = Add(
   A=Multiply(
    A=TextureSampleParameter2D(UVs=@n50) {parameterName=Param, texture=&T_SkyDome_Stars_Sharp}.RGB
    B=LinearInterpolate(B=@n51, Alpha=@n53) {constB=0}
   )
   B=@n58
  ) {constB=0}
  n60 = QualitySwitch(Default=@n59, Low=@n58, High=@n59, Medium=@n58, Epic=@n59)
  n61 = NamedRerouteUsage()
  n62 = Constant() {r=2}
  n63 = Multiply(A=ComponentMask(None=Multiply(A=@n61, B=@n62) {constB=2}) {b=true}, B=Constant() {r=10})
  NamedRerouteDeclaration(None=LinearInterpolate(B=Multiply(A=@n48, B=@n60), Alpha=Saturate(None=@n63))) {name=Stars, nodeColor=<rgba 1 0.2109 0 1>}
  n64 = TextureSample(UVs=NamedRerouteUsage()) {samplerSource=SSM_Wrap_WorldGroupSettings, texture=/Game/TimeOfDay/TODM_Sequence/Materials/Sky/Textures/T_milkyway}
  n65 = Power(Base=@n64.A) {constExponent=4}
  n66 = Multiply(A=NamedRerouteUsage(), B=ScalarParameter() {defaultValue=400, parameterName=starglow brightness})
  n67 = TextureCoordinate()
  n68 = TextureSample(UVs=Multiply(A=Panner(Coordinate=@n67) {speedX=0.02}) {constB=18}) {
   samplerSource=SSM_Wrap_WorldGroupSettings
   samplerType=SAMPLERTYPE_Grayscale
   texture=/Game/TimeOfDay/TODM_Sequence/Materials/Sky/Textures/LowResBlurredNoise_Radish_LF_Root
  }
  n69 = NamedRerouteUsage()
  n70 = Multiply(A=LinearInterpolate(Alpha=@n68.R) {constA=0.05}, B=@n69)
  NamedRerouteDeclaration(None=Add(A=Multiply(A=@n65, B=@n66) {constB=0.5}, B=QualitySwitch(Default=@n70, Low=@n69, High=@n70, Medium=@n69, Epic=@n70))) {name=MilkywayStars, nodeColor=<rgba 0.7812 0 1 1>}
  n71 = Saturate(None=ScalarParameter() {defaultValue=1, parameterName=CloudOpacity})
  n72 = Add(A=NamedRerouteUsage(), B=NamedRerouteUsage())
  n73 = NamedRerouteUsage()
  n74 = NamedRerouteUsage()
  n75 = NamedRerouteDeclaration(None=ScalarParameter() {parameterName=MoonActive}) {name=MoonActive, nodeColor=<rgba 0.7812 1 0 1>}
  n76 = Add(
   A=Multiply(A=NamedRerouteUsage(), B=@n71)
   B=Add(A=NamedRerouteDeclaration(None=@n72) {name=Sky, nodeColor=<rgba 0.1719 0 1 1>}, B=LinearInterpolate(A=@n73, B=@n74, Alpha=@n75))
  )
  n77 = NamedRerouteUsage()
  n78 = Multiply(A=NamedRerouteUsage(), B=@n71)
  n79 = Max(
   A=CollectionParameter() {collection=&FortniteMaterialParameters, parameterName=SkyBoxClouds}
   B=ScalarParameter() {parameterName=SkyboxClouds}
  )
  NamedRerouteDeclaration(None=LinearInterpolate(A=@n76, B=LinearInterpolate(A=@n76, B=@n77, Alpha=@n78), Alpha=Saturate(None=@n79))) {name=CloudsAndSky, nodeColor=<rgba 0.4297 1 0 1>}
  NamedRerouteDeclaration(None=@n64.RGB) {name=GalaxyRGB, nodeColor=<rgba 0.8047 0 1 1>}
  NamedRerouteDeclaration(None=MaterialFunctionCall(UVs=NamedRerouteUsage(), Texture Scale=ScalarParameter() {defaultValue=1, parameterName=MIkyWauSize}) {materialFunction=&ScaleUVsByCenter}.UVs) {name=milkyway_uv, nodeColor=<rgba 0 1 0.8359 1>}
  NamedRerouteUsage()
  NamedRerouteUsage()
  NamedRerouteUsage()
  OneMinus(None=@n6)
  n80 = NamedRerouteUsage()
  n81 = NamedRerouteUsage()
  VertexInterpolator(VS=Add(A=Divide(A=@n80, B=@n81)) {constB=0.5})
 ]
}
