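# Material function for the Asteria time-of-day skybox clouds, outputting cloud color and alpha.
# The main layer projects a panning, noise-distorted mask onto a distance-curved virtual plane, with near/far parameters shaping coverage and lighting.
# A normal map, sun direction and time-of-day colors light it, alongside a separately scrolling high-altitude layer.
# Sky atmosphere aerial perspective blends the clouds into the sky's view luminance.
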MaterialFunction /Game/TimeOfDay/TODM_Sequence/Materials/Sky/Clouds/MF_Asteria_SkyBoxClouds
graph {nodes [
  n1 = SkyAtmosphereDistantLightScatteredLuminance()
  n2 = ScalarParameter() {defaultValue=1, group=SkyboxClouds, parameterName=SkylightBrightness}
  n3 = Multiply(A=NamedRerouteUsage(), B=NamedRerouteUsage())
  Add(A=Multiply(A=Multiply(A=@n1, B=@n2), B=NamedRerouteUsage()), B=@n3)
  n4 = NamedRerouteUsage()
  ComponentMask(None=@n4) {b=true, g=true, r=true}
  Constant() {r=1}
  Constant() {r=1}
  n5 = Constant2Vector()
  n6 = ScalarParameter() {defaultValue=1000000, group=SkyboxClouds, parameterName=HighAltCloudsAltitude}
  n7 = Time()
  n8 = ScalarParameter() {defaultValue=4161.4268, group=SkyboxClouds, parameterName=HighAltCloudsSpeed}
  n9 = Add(
   A=MaterialFunctionCall(Plane Center=AppendVector(A=@n5, B=@n6)) {materialFunction=&VirtualPlaneCoordinates}.World Position
   B=Multiply(A=Multiply(A=@n7, B=@n8), B=Constant3Vector() {constant=<rgba -1 1 0 0>})
  )
  n10 = ScalarParameter() {defaultValue=2350000, group=SkyboxClouds, parameterName=HighAltCloudsSize}
  n11 = TextureSample(
   UVs=ComponentMask(None=Divide(A=@n9, B=@n10)) {g=true, r=true}
   Tex=TextureObjectParameter() {group=SkyboxClouds, parameterName=HighAltCloudsTexture, texture=/Game/TimeOfDay/LightFunction/T_CloudNoise_Artemis}
  ) {samplerSource=SSM_Wrap_WorldGroupSettings, samplerType=SAMPLERTYPE_LinearGrayscale}
  DotProduct(A=@n11.RGB, B=VectorParameter() {group=SkyboxClouds, parameterName=HighAltCloudsChannel, defaultValue=<rgba 1 0 0 1>}.RGB)
  n12 = SkyAtmosphereViewLuminance!(WorldDirection)
  n13 = NamedRerouteUsage()
  n14 = SkyAtmosphereAerialPerspective!(World Position=NamedRerouteUsage())
  n15 = LinearInterpolate(
   B=MaterialFunctionCall(Float4=FunctionOutput(None=@n14) {outputName=AerialPerspective}) {materialFunction=&BreakOutFloat4Components}.A
   Alpha=ScalarParameter() {defaultValue=1, group=SkyboxClouds, parameterName=AerialPerspectvieInfluence}
  ) {constA=1}
  FunctionOutput(None=NamedRerouteDeclaration(None=LinearInterpolate(A=@n12, B=@n13, Alpha=@n15)) {name=CloudColorOUT, nodeColor=<rgba 1 0 0.4687 1>}) {outputName=CloudColor}
  n16 = NamedRerouteUsage()
  n17 = NamedRerouteUsage()
  n18 = NamedRerouteUsage()
  n19 = NamedRerouteDeclaration(None=Multiply(A=NamedRerouteUsage(), B=Multiply(A=@n17, B=@n18))) {name=ModifiedHighAltCloudAlpha, nodeColor=<rgba 0.5 0 1 1>}
  FunctionOutput(None=NamedRerouteDeclaration(None=Max(A=@n16, B=@n19)) {name=CombinedCloudAlpha, nodeColor=<rgba 0.8516 1 0 1>}) {outputName=CloudAlpha, sortPriority=1}
  n20 = SkyAtmosphereLightIlluminance!(World Position=NamedRerouteUsage())
  n21 = Multiply(
   A=ComponentMask(None=FunctionOutput(None=@n20) {outputName=LightIlluminance}) {b=true, g=true, r=true}
   B=ScalarParameter() {defaultValue=0.2, group=SkyboxClouds, parameterName=LightIlluminanceStrength}
  ) {constB=0.2}
  LinearInterpolate(B=@n21, Alpha=ScalarParameter() {defaultValue=1, group=SkyboxClouds, parameterName=LightIlluminanceInfluence}) {constA=1}
  n22 = Saturate(None=Multiply(A=ScalarParameter() {defaultValue=0.375, group=SkyboxClouds, parameterName=HighAltCloudsOpacity}, B=Saturate(None=@n11.R)))
  n23 = VectorParameter() {group=SkyboxClouds, parameterName=HighAltCloudsColor, defaultValue=<rgba 0.9323 0.9061 0.8643 1>}
  Multiply(A=@n22, B=@n23.RGB)
  n24 = NamedRerouteUsage()
  n25 = ComponentMask(None=@n24) {r=true}
  n26 = Add(A=@n25, B=ComponentMask(None=@n24) {g=true})
  n27 = TextureSampleParameter2D(UVs=NamedRerouteUsage()) {
   group=SkyboxClouds
   parameterName=CloudMask
   samplerSource=SSM_Wrap_WorldGroupSettings
   samplerType=SAMPLERTYPE_LinearGrayscale
   sortPriority=0
   texture=&/T_Asteria_Skybox_Clouds
  }
  n28 = Multiply(
   A=SmoothStep(Min=Saturate(None=@n25), Max=Saturate(None=@n26), Value=@n27.G) {constMax=0.3, constMin=0.2, constValue=0.6}
   B=NamedRerouteUsage()
  )
  Multiply(A=NamedRerouteUsage(), B=@n28)
  Multiply(A=ScalarParameter() {defaultValue=1984.7985, parameterName=CloudMovementDebug}) {constB=10000}
  n29 = NamedRerouteUsage()
  n30 = SkyAtmosphereLightDirection()
  n31 = Power(Base=DotProduct(A=Multiply(A=@n30) {constA=-1, constB=-1}, B=CameraVectorWS()))
  Multiply(A=Power(Base=OneMinus(None=@n29), Exp=Constant() {r=3}), B=@n31)
  n32 = CollectionParameter() {
   collection=&FortniteMaterialParameters
   parameterName=FortniteActiveTimeOfDay
  }
  n33 = MaterialFunctionCall(Float4=@n32) {materialFunction=&BreakOutFloat4Components}
  NamedRerouteDeclaration(None=LinearInterpolate(B=ScalarParameter() {defaultValue=1, group=SkyboxClouds, parameterName=CloudNightOpacity}, Alpha=@n33.A) {constA=1}) {name=TODOpacity, nodeColor=<rgba 0.7344 1 0 1>}
  n34 = SphereMask(
   A=Constant2Vector()
   B=ComponentMask(None=NamedRerouteUsage()) {g=true, r=true}
   Radius=ScalarParameter() {defaultValue=21000000, group=SkyboxClouds, parameterName=CloudNearFarMaskDistance}
   Hardness=ScalarParameter() {defaultValue=0.01, group=SkyboxClouds, parameterName=CloudNearFarMaskHardness}
  ) {hardnessPercent=0.5}
  NamedRerouteDeclaration(None=Power(Base=Saturate(None=@n34), Exp=ScalarParameter() {defaultValue=3, group=SkyboxClouds, parameterName=CloudNearFarMaskExponent})) {name=CloudNearFarMask, nodeColor=<rgba 0.3359 0 1 1>}
  n35 = MaterialFunctionCall(Plane Center=AppendVector(A=Constant2Vector(), B=NamedRerouteUsage())) {materialFunction=&VirtualPlaneCoordinates}
  n36 = FunctionOutput(None=NamedRerouteDeclaration(None=@n35.World Position) {name=VirtualPlaneWSP, nodeColor=<rgba 1 0.9375 0 1>}) {outputName=WorldPosition}
  NamedRerouteDeclaration(None=@n36) {name=WorldSpaceUV, nodeColor=<rgba 0 0.2734 1 1>}
  n37 = Time()
  n38 = ScalarParameter() {defaultValue=10000, group=SkyboxClouds, parameterName=CloudSpeed}
  n39 = Add(
   A=NamedRerouteUsage()
   B=Multiply(A=Multiply(A=@n37, B=@n38), B=VectorParameter() {group=SkyboxClouds, parameterName=CloudPanDirection, defaultValue=<rgba 0 0 0 1>}.RGB)
  )
  NamedRerouteDeclaration(None=Divide(A=ComponentMask(None=@n39) {g=true, r=true}, B=ScalarParameter() {defaultValue=10000000, group=SkyboxClouds, parameterName=CloudUVDivisor})) {name=CloudUVPan, nodeColor=<rgba 0 1 0.7891 1>}
  n40 = NamedRerouteUsage()
  n41 = ScalarParameter() {defaultValue=18250000, group=SkyboxClouds, parameterName=CloudMaskDistance}
  NamedRerouteDeclaration(None=SphereMask(
    A=Constant2Vector()
    B=ComponentMask(None=@n40) {g=true, r=true}
    Radius=NamedRerouteDeclaration(None=@n41) {name=CloudMaskDistance, nodeColor=<rgba 0.3828 1 0 1>}
    Hardness=ScalarParameter() {defaultValue=0.4, group=SkyboxClouds, parameterName=CloudMaskHardness}
   ) {hardnessPercent=0.5}) {name=CloudDistanceMask, nodeColor=<rgba 0.4531 0 1 1>}
  n42 = NamedRerouteUsage()
  n43 = ScalarParameter() {parameterName=Min}
  n44 = ScalarParameter() {defaultValue=1, parameterName=Max}
  NamedRerouteDeclaration(None=LinearInterpolate(A=Multiply(A=@n19, B=@n42), B=NamedRerouteUsage(), Alpha=SmoothStep(Min=@n43, Max=@n44, Value=@n16))) {name=CombinedCloudColor, nodeColor=<rgba 0.9688 1 0 1>}
  n45 = Multiply(A=SkyAtmosphereLightDirection(), B=ScalarParameter() {defaultValue=1000000, group=SkyboxClouds, parameterName=SunDistance}) {constB=1000000}
  n46 = AppendVector(A=ComponentMask(None=@n4) {g=true, r=true}, B=Constant())
  NamedRerouteDeclaration(None=Normalize(VectorInput=Subtract(A=@n45, B=@n46))) {name=SunToWSP, nodeColor=<rgba 0 0.1094 1 1>}
  n47 = VectorParameter() {group=SkyboxClouds, parameterName=CloudShadowColor, defaultValue=<rgba 0.1963 0.3142 0.6875 1>}
  n48 = Multiply(A=@n31, B=ScalarParameter() {defaultValue=0.15, group=SkyboxClouds, parameterName=CloudSunBrightness})
  n49 = NamedRerouteUsage()
  n50 = TextureSampleParameter2D(UVs=NamedRerouteUsage()) {
   group=SkyboxClouds
   parameterName=CloudNormal
   samplerType=SAMPLERTYPE_Normal
   sortPriority=1
   texture=&/T_Asteria_Skybox_Clouds_N
  }
  n51 = ConstantBiasScale(None=DotProduct(A=NamedRerouteUsage(), B=Normalize(VectorInput=@n50.RGB)))
  n52 = ScalarParameter() {defaultValue=3, group=SkyboxClouds, parameterName=CloudNormalExponent}
  n53 = Multiply(A=Multiply(A=NamedRerouteUsage(), B=ComponentMask(None=@n49) {b=true}) {constB=2}, B=Saturate(None=Power(Base=@n51, Exp=@n52)))
  NamedRerouteDeclaration(None=Add(A=NamedRerouteDeclaration(None=@n47.RGB) {name=CloudColor, nodeColor=<rgba 0 0.6484 1 1>}, B=Add(A=@n48, B=@n53))) {name=CloudLighting, nodeColor=<rgba 1 0 0.7969 1>}
  NamedRerouteDeclaration(None=@n21) {name=LightIlluminance, nodeColor=<rgba 0.6406 1 0 1>}
  NamedRerouteDeclaration(None=@n3) {name=CloudsWithSkylight, nodeColor=<rgba 0 0.3906 1 1>}
  NamedRerouteDeclaration(None=MaterialFunctionCall(
    Color_Morning=VectorParameter() {group=SkyboxClouds, parameterName=CloudLightColor_Morning, defaultValue=<rgba 1 0.6321 0.4635 1>}.RGB
    Color_Day=VectorParameter() {group=SkyboxClouds, parameterName=CloudLightColor_Day, defaultValue=<rgba 1 0.6321 0.4635 1>}.RGB
    Color_Evening=VectorParameter() {group=SkyboxClouds, parameterName=CloudLightColor_Evening, defaultValue=<rgba 1 0.6321 0.4635 1>}.RGB
    Color_Night=VectorParameter() {group=SkyboxClouds, parameterName=CloudLightColor_Night, defaultValue=<rgba 1 0.6321 0.4635 1>}.RGB
   ) {materialFunction=&Color_Time_of_Day}.Result) {name=CloudTODLightColor, nodeColor=<rgba 0.1719 1 0 1>}
  NamedRerouteDeclaration(None=@n22) {name=HighAltCloudAlpha, nodeColor=<rgba 0 1 0.6484 1>}
  NamedRerouteDeclaration(None=@n23.RGB) {name=HighAltCloudColor, nodeColor=<rgba 0.4297 1 0 1>}
  n54 = Length(None=ComponentMask(None=WorldPosition().XYZ) {g=true, r=true}) {is=class MaterialExpressionLength}
  n55 = NamedRerouteUsage()
  n56 = OneMinus(None=Saturate(None=Divide(A=@n54, B=@n55)))
  n57 = ScalarParameter() {defaultValue=2, group=SkyboxClouds, parameterName=CloudHeightCurveExponent}
  NamedRerouteDeclaration(None=LinearInterpolate(
    A=ScalarParameter() {defaultValue=50000, group=SkyboxClouds, parameterName=CloudHeightMin}
    B=ScalarParameter() {defaultValue=1000000, group=SkyboxClouds, parameterName=CloudHeightMax}
    Alpha=Power(Base=@n56, Exp=@n57)
   )) {name=Height, nodeColor=<rgba 0.0078 1 0 1>}
  NamedRerouteDeclaration(None=Multiply(A=Saturate(None=@n28), B=NamedRerouteUsage())) {name=CloudAlpha, nodeColor=<rgba 0.5703 1 0 1>}
  n58 = ScalarParameter() {defaultValue=0.2, group=SkyboxClouds, parameterName=CloudWidthFar}
  n59 = ScalarParameter() {defaultValue=0.576, group=SkyboxClouds, parameterName=CloudWidthSoftnessFar}
  n60 = ScalarParameter() {defaultValue=3, group=SkyboxClouds, parameterName=CloudLightStrengthFar}
  n61 = ScalarParameter() {defaultValue=0.01, group=SkyboxClouds, parameterName=CloudMicroNoiseFar}
  n62 = ScalarParameter() {defaultValue=0.85, group=SkyboxClouds, parameterName=CloudWidthNear}
  n63 = ScalarParameter() {defaultValue=0.2, group=SkyboxClouds, parameterName=CloudWidthSoftnessNear}
  n64 = ScalarParameter() {defaultValue=0.2, group=SkyboxClouds, parameterName=CloudLightStrengthNear}
  n65 = ScalarParameter() {defaultValue=0.045, group=SkyboxClouds, parameterName=CloudMicroNoiseNear}
  NamedRerouteDeclaration(None=LinearInterpolate(
    A=MaterialFunctionCall(R=@n58, G=@n59, B=@n60, A=@n61) {materialFunction=&AppendMany}.RGBA
    B=MaterialFunctionCall(R=@n62, G=@n63, B=@n64, A=@n65) {materialFunction=&AppendMany}.RGBA
    Alpha=NamedRerouteUsage()
   )) {name=CloudNearFarParams, nodeColor=<rgba 0.9453 0 1 1>}
  NamedRerouteDeclaration(None=@n35.World Position - Centered) {name=VirtualPlandWSPCentered, nodeColor=<rgba 1 0 0.7031 1>}
  NamedRerouteDeclaration(None=@n35.UVs) {name=VirtualPlaneUV, nodeColor=<rgba 1 0.7266 0 1>}
  n66 = DoubleVectorParameter() {group=SkyboxClouds, parameterName=Cloud_MicroNoise_UVDirection, defaultValue=<0.007 0 0 1>}
  n67 = Multiply(
   A=Panner(Speed=ComponentMask(None=@n66.XYZ) {g=true, r=true})
   B=ScalarParameter() {defaultValue=5, group=SkyboxClouds, parameterName=Cloud_MicroNoise_UVTile}
  ) {constB=0.5}
  n68 = TextureSample(UVs=@n67) {
   samplerSource=SSM_Wrap_WorldGroupSettings
   samplerType=SAMPLERTYPE_LinearGrayscale
   texture=&T_Asteria_Skybox_Clouds
  }
  n69 = ComponentMask(None=NamedRerouteUsage()) {a=true}
  NamedRerouteDeclaration(None=Add(A=NamedRerouteUsage(), B=Multiply(A=@n68.R, B=@n69))) {name=CloudUVs, nodeColor=<rgba 1 0 0.0469 1>}
  NamedRerouteDeclaration(None=LinearInterpolate(
    A=VectorParameter() {group=SkyboxClouds, parameterName=CloudShadowColor, defaultValue=<rgba 0.1963 0.3142 0.6875 1>}.RGB
    B=VectorParameter() {group=SkyboxClouds, parameterName=CloudNightShadowColor, defaultValue=<rgba 0.1963 0.3142 0.6875 1>}.RGB
    Alpha=@n33.A
   ) {constA=1}) {name=CloudShadowColor, nodeColor=<rgba 0 1 0.2031 1>}
  NamedRerouteUsage()
  NamedRerouteUsage()
  NamedRerouteUsage()
  NamedRerouteUsage()
  NamedRerouteUsage()
  ScalarParameter() {defaultValue=500000, group=SkyboxClouds, parameterName=CloudHeight}
  ScalarParameter() {defaultValue=1, parameterName=CloudAlphaExponent}
  StaticSwitchParameter(
   True=MaterialFunctionCall!(
    Worldposition=@n36
    SwirlTexture=TextureObject() {samplerType=SAMPLERTYPE_LinearGrayscale, texture=&T_Asteria_Skybox_Clouds}
    RotationAngle=ScalarParameter() {defaultValue=0.822, group=SkyboxClouds, parameterName=CloudTwistAngle}
   ) {materialFunction=/Game/TimeOfDay/Skydome/MF_VirtualPlane_Twist}.Result
   False=@n36
  ) {group=SkyboxClouds, parameterName=UseCloudTwist}
  TextureSample(UVs=@n67) {texture=&/T_Clouds_02_M}
  VectorParameter() {group=SkyboxClouds, parameterName=CloudLightColor, defaultValue=<rgba 1 0.4361 0 1>}
 ]}
