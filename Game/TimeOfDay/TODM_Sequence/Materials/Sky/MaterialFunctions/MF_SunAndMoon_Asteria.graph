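# Sky material function for the Time of Day system that draws either the sun or the moon, blended by MoonActive, and outputs moon masks.
# UVs come from projecting camera-to-pixel direction around the SkyAtmosphere light direction.
# The sun blends atmosphere disk luminance with a fake sphere-mask sun, optionally eclipsed with ring and animated noise corona, clamped in reflection captures.
# The moon combines masked texture, halo and optional curve-atlas blood moon tint.
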MaterialFunction /Game/TimeOfDay/TODM_Sequence/Materials/Sky/MaterialFunctions/MF_SunAndMoon_Asteria
graph {nodes [
  n1 = NamedRerouteUsage()
  n2 = NamedRerouteUsage()
  n3 = ScalarParameter() {group=Sun-Moon, parameterName=MoonActive}
  FunctionOutput(None=NamedRerouteDeclaration(None=LinearInterpolate(A=@n1, B=@n2, Alpha=@n3)) {name=SunMoon, nodeColor=<rgba 0 0.2969 1 1>})
  n4 = NamedRerouteUsage()
  n5 = SphereMask(
   A=VertexInterpolator(VS=Add(A=@n4) {constB=0.5}).PS
   B=Constant() {r=0.5}
   Radius=ScalarParameter() {defaultValue=0.8, group=Sun-Moon, parameterName=Moon Halo Radius}
   Hardness=ScalarParameter() {defaultValue=-0.2, group=Sun-Moon, parameterName=moon halo Hardness}
  ) {hardnessPercent=0.1}
  FunctionOutput(None=@n5) {outputName=MoonHaloMask}
  n6 = NamedRerouteUsage()
  n7 = ScalarParameter() {defaultValue=0.07, group=Sun-Moon, parameterName=MoonSize}
  n8 = VertexInterpolator(VS=Add(A=Divide(A=@n6, B=@n7)) {constB=0.5})
  n9 = TextureSampleParameter2D(UVs=@n8.PS) {
   group=Sun-Moon
   parameterName=MoonTexture
   samplerSource=SSM_Wrap_WorldGroupSettings
   samplerType=SAMPLERTYPE_LinearColor
   texture=/Game/TimeOfDay/Skydome/T_Moon_Artemis
  }
  n10 = Saturate(None=Multiply(A=@n9.R) {constB=20})
  n11 = Constant() {r=0.5}
  n12 = MaterialFunctionCall(Float2=@n8.PS) {materialFunction=&BreakOutFloat2Components}
  n13 = Multiply(A=SphereMask(A=@n11, B=@n12.R) {attenuationRadius=0.5}, B=SphereMask(A=@n12.G, B=@n11) {attenuationRadius=0.5})
  FunctionOutput(None=NamedRerouteDeclaration(None=Multiply(A=@n10, B=@n13)) {name=MoonMask, nodeColor=<rgba 0 1 0.1094 1>}) {outputName=MoonMask, sortPriority=3}
  FunctionOutput(None=@n3) {outputName=MoonActive, sortPriority=4}
  n14 = Constant2Vector()
  n15 = NamedRerouteUsage()
  n16 = ScalarParameter() {defaultValue=0.15, desc=.04, group=Sun-Moon, parameterName=SunSize}
  n17 = Divide(A=@n15, B=@n16)
  n18 = SphereMask(A=@n14, B=@n17, Hardness=ScalarParameter() {defaultValue=1, group=Sun-Moon, parameterName=SunHardness}) {attenuationRadius=0.5}
  n19 = Add(
   A=@n14
   B=AppendVector(
    A=ScalarParameter() {defaultValue=0.2, group=Eclipse, parameterName=EclipseBiasU}
    B=ScalarParameter() {group=Eclipse, parameterName=EclipseBiasV}
   )
  )
  n20 = ScalarParameter() {defaultValue=0.51, parameterName=EclipseRadius}
  n21 = SphereMask(A=@n19, B=@n17, Radius=@n20, Hardness=ScalarParameter() {defaultValue=1, parameterName=EclipseHardness}) {attenuationRadius=0.5}
  n22 = Multiply(A=@n18, B=Multiply(A=@n18, B=OneMinus(None=@n21)))
  n23 = Multiply(
   A=@n18
   B=SphereMask(A=@n21, B=Constant() {r=0.78}, Radius=ScalarParameter() {defaultValue=0.2, parameterName=EclipseRingThickness}) {attenuationRadius=0.28, hardnessPercent=0.9}
  )
  n24 = Add(
   A=Multiply(A=VectorParameter() {group=Sun-Moon, parameterName=SunColor, defaultValue=<rgba 175 79.4452 0 1>}.RGB, B=Saturate(None=@n22))
   B=Multiply(A=Saturate(None=@n23), B=VectorParameter() {group=Eclipse, parameterName=EclipseRingColor, defaultValue=<rgba 5 5 5 1>}.RGB)
  )
  n25 = Saturate(None=Multiply(A=@n18, B=Constant() {r=2}))
  n26 = VectorParameter() {parameterName=EclipseCoronaColor, defaultValue=<rgba 5 1.5112 0.3483 1>}
  n27 = MaterialFunctionCall(Vector or UVs=@n15) {materialFunction=&VectorToRadialValue}
  n28 = Panner(Coordinate=Multiply(A=@n27.Radial Coordinates, B=Constant() {r=1}) {constB=0.5}) {speedX=-0.002, speedY=-0.01}
  n29 = TextureObject() {samplerType=SAMPLERTYPE_LinearColor, texture=/Game/TimeOfDay/Skydome/PackedEclipseNoise_M}
  n30 = Multiply(
   A=@n27.Radial Coordinates
   B=AppendVector(
    A=ScalarParameter() {defaultValue=2, group=Eclipse, parameterName=EclipseCoronaScale}
    B=ScalarParameter() {defaultValue=4, group=Eclipse, parameterName=EclipseCoronaScaleV}
   )
  )
  n31 = TextureSample(UVs=Multiply(A=Panner(Coordinate=@n30) {speedY=-0.033}, B=Constant2Vector() {g=1, r=-1}), Tex=@n29) {samplerType=SAMPLERTYPE_LinearColor, texture=/Game/Accessories/FORT_Backpacks/Backpack_M_MED_AllKnowing/FX/T_Backpack_AllKnowing_Noise_01}
  n32 = Max(
   A=TextureSample(UVs=Panner(Coordinate=@n30) {speedY=0.04}, Tex=@n29) {samplerType=SAMPLERTYPE_LinearColor, texture=/Game/Effects/Fort_Effects/Textures/T_DarkAndLight_Noise}.R
   B=@n31.G
  )
  n33 = SphereMask(A=@n14, B=Divide(A=@n15, B=Multiply(A=@n16) {constB=3})) {attenuationRadius=1, hardnessPercent=0}
  n34 = Multiply(A=@n33, B=@n33)
  n35 = Multiply(
   A=Power(
    Base=TextureSample(UVs=@n28, Tex=@n29) {samplerType=SAMPLERTYPE_LinearColor, texture=/Game/Effects/FogVolume/noise_flamy_Tile}.B
    Exp=Constant() {r=3}
   )
   B=Multiply(A=Multiply(A=@n32, B=@n34), B=ScalarParameter() {defaultValue=2, parameterName=BoostCorona})
  )
  n36 = SphereMask(A=@n19, B=@n17, Radius=Multiply(A=@n20, B=ScalarParameter() {defaultValue=1.6, parameterName=CoronaRad}) {constB=2}) {attenuationRadius=0.5, hardnessPercent=0}
  n37 = Multiply(
   A=Multiply(A=Multiply(A=@n36, B=@n36), B=ScalarParameter() {defaultValue=1.5, parameterName=CoronaBoost})
   B=Multiply(A=OneMinus(None=@n25), B=Multiply(A=@n26.RGB, B=@n35))
  )
  n38 = Multiply(A=StaticSwitchParameter(True=Add(A=@n24, B=@n37), False=@n18) {group=Eclipse, parameterName=UseEclipseForFakeSun}, B=NamedRerouteUsage())
  n39 = MaterialFunctionCall(LDR Input=Saturate(None=@n38), HDR Tint=@n38) {materialFunction=&DeriveHDRfromLDR}
  n40 = ScalarParameter() {defaultValue=10, parameterName=ClampSunReflection}
  NamedRerouteDeclaration(None=ReflectionCapturePassSwitch!(Default=@n39.RemappedRange, Reflection=Clamp(None=@n39.RemappedRange, Max=@n40) {maxDefault=10})) {name=SunSphereMask, nodeColor=<rgba 0.8281 0 1 1>}
  n41 = NamedRerouteUsage()
  n42 = ComponentMask(None=SkyAtmosphereLightDirection()) {b=true, g=true, r=true}
  NamedRerouteDeclaration(None=Step(X=DotProduct(A=@n41, B=@n42))) {name=HemisphereMask, nodeColor=<rgba 1 0 0.3047 1>}
  n43 = Multiply(A=@n42) {constB=-1}
  n44 = CrossProduct(A=@n43, B=Constant3Vector() {constant=<rgba 0 0 -1 0>})
  n45 = Normalize(VectorInput=CrossProduct(A=@n44, B=@n43))
  n46 = NamedRerouteUsage()
  n47 = Normalize(VectorInput=@n44)
  n48 = NamedRerouteUsage()
  NamedRerouteDeclaration(None=AppendVector(A=DotProduct(A=@n45, B=@n46), B=DotProduct(A=@n47, B=@n48))) {name=SunMoonUVs, nodeColor=<rgba 1 0 0.9141 1>}
  n49 = ScalarParameter() {defaultValue=1, group=Sun-Moon, parameterName=SunStrength}
  n50 = SkyAtmosphereLightDiskLuminance!(DiskAngularDiameterOverride)
  n51 = ScalarParameter() {group=Sun-Moon, parameterName=UseFakeSun}
  NamedRerouteDeclaration(None=LinearInterpolate(A=Multiply(A=@n49, B=@n50) {constA=1}, B=NamedRerouteUsage(), Alpha=Saturate(None=@n51))) {name=Sun, nodeColor=<rgba 0 0.9766 1 1>}
  n52 = ScalarParameter() {defaultValue=2, parameterName=Moon Halo Exponent}
  n53 = CurveAtlasRowParameter(CurveTime=SphereMask(A=Constant2Vector(), B=@n8.PS) {attenuationRadius=1, hardnessPercent=0}) {
   atlas=&/CA_Athena_TOD
   curve=&/SkyBloodMoonColorCurve
  }
  n54 = Multiply(A=Multiply(A=@n9.RGB, B=@n13), B=VectorParameter() {group=Sun-Moon, parameterName=MoonColor, defaultValue=<rgba 1 1 1 1>}.RGB)
  n55 = Add(
   A=Multiply(A=Power(Base=@n5, Exp=@n52), B=VectorParameter() {group=Sun-Moon, parameterName=moon halo Color, defaultValue=<rgba 1 1 1 1>}.RGB)
   B=StaticSwitchParameter(True=Multiply(A=@n53, B=@n54), False=@n54) {parameterName=useBloodMoon}
  )
  n56 = ScalarParameter() {defaultValue=17, group=Sun-Moon, parameterName=MoonIntensity}
  NamedRerouteDeclaration(None=Multiply(A=NamedRerouteUsage(), B=Multiply(A=@n55, B=@n56))) {name=Moon, nodeColor=<rgba 0 1 0.8125 1>}
  n57 = WorldPosition()
  n58 = CameraPositionWS()
  NamedRerouteDeclaration(None=Normalize(VectorInput=Subtract(A=@n57.XYZ, B=@n58))) {name=CameraVectorForVertexShader, nodeColor=<rgba 0 1 0.0859 1>}
  Power(Base=@n9.RGB) {constExponent=1.2}
  WorldPosition()
 ]}
