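# Unlit sky dome material for time-of-day lighting, blending Sky Atmosphere luminance with a tinted night sky of Milky Way and masked, noise-twinkled star layers.
# A sun-and-moon function, time-of-day-tinted high-altitude clouds panned on a virtual plane, and a skybox cloud function layer into the result.
# Quality and feature-level switches simplify or drop effects, collection parameters drive cloud and safe-zone blends, and reflection captures get separate desaturation and tint.
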Material /Game/TimeOfDay/TODM_Sequence/Materials/Sky/M_Sky_Basic_v2
graph {
 properties {bCastRayTracedShadows=false, bIsSky=true, pixelDepthOffsetMode=PDOM_Legacy, shadingModel=MSM_Unlit, usedShadingModels=Unlit}
 nodes [
  Add(A=TextureCoordinate(), B=AppendVector(A=Constant(), B=NamedRerouteUsage()))
  Constant2Vector() {g=500, r=500}
  Constant3Vector() {constant=<rgba 50 50 50 0>}
  Constant3Vector()
  n1 = NamedRerouteUsage()
  Desaturation(None=@n1, Fraction=ScalarParameter() {parameterName=desaturation})
  LinearInterpolate(Alpha=ScalarParameter() {defaultValue=1, parameterName=Debug})
  n2 = NamedRerouteUsage()
  n3 = Constant() {r=60}
  n4 = NamedRerouteUsage()
  n5 = Constant() {r=12}
  n6 = TextureSample(UVs=Multiply(A=@n4, B=Constant() {r=80})) {samplerSource=SSM_Wrap_WorldGroupSettings, texture=/Game/TimeOfDay/TODM_Sequence/Materials/Sky/Textures/T_milkyway_StarMask}
  n7 = Constant() {r=2}
  n8 = Multiply(
   A=TextureSampleParameter2D(UVs=Multiply(A=@n4, B=@n5) {constB=5}) {parameterName=Param, texture=/Game/TimeOfDay/TODM_Sequence/Materials/Sky/Textures/T_SkyDome_Stars_Sharp}.RGB
   B=LinearInterpolate(B=ScalarParameter() {defaultValue=200, parameterName=smallStarIndensity}, Alpha=Power(Base=@n6.B, Exp=@n7) {constExponent=4}) {constB=0}
  )
  n9 = NamedRerouteUsage()
  n10 = Constant() {r=4.2}
  n11 = TextureSample(UVs=Multiply(A=@n9, B=Constant() {r=30})) {samplerSource=SSM_Wrap_WorldGroupSettings, texture=&T_milkyway_StarMask}
  n12 = Constant() {r=4}
  n13 = Multiply(
   A=TextureSample(UVs=Multiply(A=@n9, B=@n10) {constB=5}) {samplerSource=SSM_Wrap_WorldGroupSettings, texture=&T_SkyDome_Stars_Sharp}.RGB
   B=LinearInterpolate(B=ScalarParameter() {defaultValue=2000, parameterName=BigStarIndensity}, Alpha=Power(Base=@n11.B, Exp=@n12) {constExponent=4}) {constB=0}
  )
  n14 = ComponentMask(None=Multiply(A=NamedRerouteUsage(), B=Constant() {r=2}) {constB=2}) {b=true}
  n15 = Constant() {r=10}
  n16 = LinearInterpolate(B=Multiply(A=Multiply(A=@n2, B=@n3), B=Add(A=@n8, B=@n13)), Alpha=Saturate(None=Multiply(A=@n14, B=@n15)))
  n17 = TextureSampleParameter2D(UVs=Multiply(A=NamedRerouteUsage(), B=Constant() {r=16}) {constB=5}) {
   parameterName=Param
   samplerSource=SSM_Wrap_WorldGroupSettings
   texture=&T_SkyDome_Stars_Sharp
  }
  n18 = ScalarParameter() {defaultValue=3, parameterName=BGStarIndensity}
  n19 = NamedRerouteUsage()
  n20 = Constant() {r=140}
  n21 = NamedRerouteUsage()
  n22 = Power(
   Base=TextureSample(UVs=Multiply(A=@n19, B=@n20)) {samplerSource=SSM_Wrap_WorldGroupSettings, texture=&T_milkyway_StarMask}.B
   Exp=LinearInterpolate(Alpha=TextureSample(UVs=@n21) {samplerSource=SSM_Wrap_WorldGroupSettings, texture=&T_milkyway_StarMask}.R) {constA=3, constB=2}
  ) {constExponent=4}
  Max(A=@n16, B=Multiply(A=@n17.RGB, B=LinearInterpolate(B=@n18, Alpha=@n22) {constB=0}) {constB=0.5})
  n23 = SkyAtmosphereLightDirection()
  Multiply(A=Constant() {r=-1}, B=@n23)
  n24 = VectorParameter() {parameterName=DarkColor, defaultValue=<rgba 0.0985 0.044 0.17 1>}
  n25 = TextureCoordinate()
  n26 = TextureSample(UVs=Multiply(A=Panner(Coordinate=@n25) {speedX=0.005}) {constB=12}) {
   samplerSource=SSM_Wrap_WorldGroupSettings
   samplerType=SAMPLERTYPE_Grayscale
   texture=/Game/TimeOfDay/TODM_Sequence/Materials/Sky/Textures/LowResBlurredNoise_Radish_LF_Root
  }
  n27 = LinearInterpolate(A=@n24.RGB, B=VectorParameter() {parameterName=LightColor, defaultValue=<rgba 0.1451 0.088 0.22 1>}.RGB, Alpha=@n26.R)
  n28 = QualitySwitch(Default=@n24.RGB, Low=@n24.RGB, High=@n27, Medium=@n24.RGB, Epic=@n27)
  n29 = NamedRerouteUsage()
  n30 = TextureCoordinate() {uTiling=0.5, vTiling=0.5}
  n31 = AppendVector(A=Constant() {r=0.25}, B=Constant() {r=0.25})
  n32 = Constant() {r=0.24}
  n33 = Constant() {r=0.65}
  NamedRerouteDeclaration(None=LinearInterpolate(A=Multiply(A=@n28, B=@n29), B=@n29, Alpha=SphereMask(A=@n30, B=@n31, Radius=@n32, Hardness=@n33))) {name=horizon darkens, nodeColor=<rgba 0.9922 0 1 1>}
  n34 = Normalize(VectorInput=Subtract(A=WorldPosition().XYZ, B=CameraPositionWS()))
  n35 = ConstantBiasScale(None=Clamp(None=DotProduct(A=@n23, B=@n34)))
  n36 = Constant() {r=3}
  n37 = Multiply(
   A=LinearInterpolate(A=Constant() {r=0.7}, Alpha=Power(Base=@n35, Exp=@n36))
   B=Multiply(
    A=VectorParameter() {group=FakeBoost, parameterName=FakeSkyLight, defaultValue=<rgba 0.0144 0.1912 0.3073 1>}.RGB
    B=ScalarParameter() {group=FakeBoost, parameterName=FakeSkyLightStrength}
   )
  )
  n38 = SkyAtmosphereViewLuminance!(WorldDirection)
  n39 = TextureCoordinate() {uTiling=0.5, vTiling=0.5}
  n40 = AppendVector(A=ScalarParameter() {defaultValue=0.25, parameterName=UVX}, B=ScalarParameter() {defaultValue=0.25, parameterName=UVY})
  n41 = ScalarParameter() {defaultValue=4.5, parameterName=Nightsky Horizon Radius}
  n42 = ScalarParameter() {defaultValue=0.04, parameterName=NightskyHorizonHardness}
  n43 = Multiply(
   A=LinearInterpolate(
    A=VectorParameter() {parameterName=nightsky horizon color, defaultValue=<rgba 15 0 0.4283 1>}.RGB
    B=VectorParameter() {parameterName=nightsky color, defaultValue=<rgba 0.405 0.45 0.4458 1>}.RGB
    Alpha=SphereMask(A=@n39, B=@n40, Radius=@n41, Hardness=@n42)
   )
   B=LinearInterpolate(B=SkyAtmosphereViewLuminance(), Alpha=ScalarParameter() {defaultValue=1, parameterName=NightSkyViewLuminanceFactor}) {constA=1}
  )
  n44 = Add(
   A=MaterialFunctionCall!(Blend=NamedRerouteUsage(), Base=@n43) {materialFunction=/Engine/Functions/Engine_MaterialFunctions03/Blends/Blend_Screen}.Result
   B=NamedRerouteUsage()
  )
  n45 = ScalarParameter() {defaultValue=1, parameterName=Nightsky Galaxy Strength}
  NamedRerouteDeclaration(None=LinearInterpolate(
    A=Add(A=@n37, B=@n38)
    B=LinearInterpolate(A=@n43, B=@n44, Alpha=@n45)
    Alpha=ScalarParameter() {defaultValue=1, parameterName=Night Sky blend}
   )) {name="Horizon  Normal", nodeColor=<rgba 1 0 0.8437 1>}
  n46 = Constant()
  n47 = NamedRerouteUsage()
  NamedRerouteDeclaration(None=Add(A=TextureCoordinate(), B=AppendVector(A=@n46, B=@n47))) {name=star pan, nodeColor=<rgba 0.3594 0 1 1>}
  NamedRerouteDeclaration(None=ScalarParameter() {defaultValue=0.016, parameterName=NightSkyPan}) {name=NightSkyPanSpeed, nodeColor=<rgba 0 1 0.4375 1>}
  n48 = MaterialFunctionCall() {materialFunction=/Game/TimeOfDay/TODM_Sequence/Materials/Sky/Clouds/MF_Asteria_SkyBoxClouds}
  NamedRerouteDeclaration(None=@n48.CloudAlpha) {name=cloud_alpha, nodeColor=<rgba 0 0.5312 1 1>}
  n49 = TextureSample(UVs=NamedRerouteUsage()) {samplerSource=SSM_Wrap_WorldGroupSettings, texture=/Game/TimeOfDay/TODM_Sequence/Materials/Sky/Textures/T_milkyway}
  NamedRerouteDeclaration(None=@n49.RGB) {name=GalaxyRGB, nodeColor=<rgba 0.8047 0 1 1>}
  n50 = MaterialFunctionCall() {materialFunction=/Game/TimeOfDay/TODM_Sequence/Materials/Sky/MaterialFunctions/MF_SunAndMoon_Asteria}
  NamedRerouteDeclaration(None=@n50.MoonMask) {name=MoonMask, nodeColor=<rgba 0 0.7422 1 1>}
  NamedRerouteDeclaration(None=@n50.MoonHaloMask) {name=MoonHaloMask, nodeColor=<rgba 0.4297 1 0 1>}
  n51 = Multiply(
   A=Power(Base=@n49.A) {constExponent=4}
   B=Multiply(A=NamedRerouteUsage(), B=ScalarParameter() {defaultValue=400, parameterName=starglow brightness})
  ) {constB=0.5}
  n52 = TextureCoordinate()
  n53 = TextureSample(UVs=Multiply(A=Panner(Coordinate=@n52) {speedX=0.02}) {constB=18}) {
   samplerSource=SSM_Wrap_WorldGroupSettings
   samplerType=SAMPLERTYPE_Grayscale
   texture=&LowResBlurredNoise_Radish_LF_Root
  }
  n54 = Saturate(None=Multiply(A=NamedRerouteUsage(), B=Constant() {r=5}))
  n55 = Multiply(A=LinearInterpolate(Alpha=@n53.R) {constA=0.05}, B=Multiply(A=@n16, B=LinearInterpolate(Alpha=@n54) {constA=1, constB=0.6}))
  n56 = NamedRerouteUsage()
  NamedRerouteDeclaration(None=Multiply(A=Add(A=@n51, B=@n55), B=OneMinus(None=@n56))) {name=MilkywayStars, nodeColor=<rgba 0.7812 0 1 1>}
  NamedRerouteDeclaration(None=Multiply(A=@n1, B=ScalarParameter() {defaultValue=0.15, parameterName=Milkway Intensity}) {constB=0.25}) {name=Milkyway, nodeColor=<rgba 0 0.4609 1 1>}
  NamedRerouteDeclaration(None=TextureSample(UVs=NamedRerouteUsage()) {samplerSource=SSM_Wrap_WorldGroupSettings, texture=&T_milkyway_StarMask}.G) {name=StarMask, nodeColor=<rgba 1 0.8438 0 1>}
  NamedRerouteDeclaration(None=MaterialFunctionCall(UVs=NamedRerouteUsage(), Texture Scale=ScalarParameter() {defaultValue=1, parameterName=MIkyWauSize}) {materialFunction=&ScaleUVsByCenter}.UVs) {name=milkyway_uv, nodeColor=<rgba 0 1 0.8359 1>}
  n57 = SphereMask(
   A=TextureCoordinate()
   B=Constant2Vector() {g=0.5, r=0.5}
   Radius=ScalarParameter() {defaultValue=0.456, parameterName=HA_Clouds_Fade_Radius}
   Hardness=ScalarParameter() {defaultValue=0.7359, parameterName=HA_Clouds_Fade_Hardness}
  ) {attenuationRadius=0.5, hardnessPercent=0}
  n58 = ScalarParameter() {defaultValue=3000000, parameterName=HA_Cloud_VirtualplaneScale}
  n59 = MaterialFunctionCall(
   Plane Normal Axis=Constant3Vector() {constant=<rgba 0 0 -1 0>}
   Plane X Axis=Constant3Vector() {constant=<rgba 1 0 0 0>}
   Plane Center=VectorParameter() {parameterName=HA_Cloud_Height, defaultValue=<rgba 0 0 550000 1>}.RGB
   UV Scale=AppendVector(A=@n58, B=@n58)
  ) {materialFunction=&VirtualPlaneCoordinates}
  n60 = Rotator(
   Coordinate=Multiply(A=TextureCoordinate(), B=ScalarParameter() {defaultValue=1, parameterName=HighAlititudeClouds_Tiling_Master})
   Time=ScalarParameter() {defaultValue=11.5849, parameterName=HighAlititudeClouds_Rotation}
  )
  n61 = Panner(
   Coordinate=QualitySwitch(Default=@n59.UVs, Low=@n60, High=@n59.UVs, Medium=@n60, Epic=@n59.UVs)
   Speed=AppendVector(
    A=ScalarParameter() {defaultValue=-0.002, parameterName=HA_Cloud_Speed_X}
    B=ScalarParameter() {defaultValue=-0.001, parameterName=HA_Cloud_Speed_Y}
   )
  )
  n62 = TextureSample(UVs=Multiply(A=@n61, B=ScalarParameter() {defaultValue=2.1989, parameterName=HighAlititudeClouds_Tiling})) {samplerSource=SSM_Wrap_WorldGroupSettings, texture=/Game/TimeOfDay/TODM_Sequence/Materials/Sky/Textures/HighAltitudeClouds_M}
  n63 = Panner(
   Coordinate=@n61
   Speed=AppendVector(
    A=ScalarParameter() {defaultValue=0.0015, parameterName=HA_Cloud_Speed2_X}
    B=ScalarParameter() {defaultValue=0.0008, parameterName=HA_Cloud_Speed2_Y}
   )
  )
  n64 = TextureSample(UVs=Multiply(A=@n63, B=ScalarParameter() {defaultValue=1.4899, parameterName=HighAlititudeClouds_Tiling_3})) {samplerSource=SSM_Wrap_WorldGroupSettings, texture=&HighAltitudeClouds_M}
  n65 = ScalarParameter() {defaultValue=2, parameterName=HighAlititudeClouds_Tiling_2}
  n66 = ScalarParameter() {defaultValue=1, parameterName=HighAlititudeClouds_Tiling_4}
  n67 = Multiply(
   A=TextureSample(UVs=Multiply(A=@n63, B=@n65)) {samplerSource=SSM_Wrap_WorldGroupSettings, texture=&HighAltitudeClouds_M}.G
   B=TextureSample(UVs=Multiply(A=@n61, B=@n66)) {samplerSource=SSM_Wrap_WorldGroupSettings, texture=&HighAltitudeClouds_M}.B
  )
  n68 = Multiply(
   A=Add(A=Multiply(A=@n62.R, B=@n64.B), B=Multiply(A=@n67) {constB=0.75})
   B=MaterialFunctionCall(
    Color_Morning=VectorParameter() {parameterName=HA_Cloud_Color_Morning, defaultValue=<rgba 0.2292 0.0892 0.0146 1>}.RGB
    Color_Day=VectorParameter() {parameterName=HA_Cloud_Color_Day, defaultValue=<rgba 0.1317 0.1456 0.1927 1>}.RGB
    Color_Evening=VectorParameter() {parameterName=HA_Cloud_Color_Evening, defaultValue=<rgba 0.1823 0.0364 0.009 1>}.RGB
    Color_Night=VectorParameter() {parameterName=HA_Cloud_Color_Night, defaultValue=<rgba 0.0306 0.0487 0.1198 1>}.RGB
   ) {materialFunction=&Color_Time_of_Day}.Result
  )
  n69 = Multiply(A=NamedRerouteUsage()) {constB=1.5}
  n70 = Multiply(A=@n50.Result, B=OneMinus(None=Saturate(None=@n69)))
  n71 = Saturate(None=CollectionParameter() {collection=&FortniteMaterialParameters, parameterName=SkyBoxClouds})
  n72 = NamedRerouteUsage()
  n73 = CollectionParameter() {collection=/Game/Athena/SafeZone/MPC_SafeZoneClouds, parameterName=CloudStartingBias}
  n74 = LinearInterpolate(A=NamedRerouteUsage(), B=NamedRerouteUsage(), Alpha=Saturate(None=Multiply(A=@n73) {constB=2}))
  n75 = Add(
   A=QualitySwitch(Default=@n70, Epic=LinearInterpolate(A=@n50.Result, B=@n70, Alpha=@n71))
   B=QualitySwitch(Default=@n72, Low=@n72, High=@n74, Medium=@n72, Epic=@n74)
  )
  n76 = FeatureLevelSwitch(Default=Add(A=Multiply(A=@n57, B=@n68), B=@n75), ES3_1=@n75)
  n77 = LinearInterpolate(A=@n76, B=@n48.CloudColor, Alpha=@n48.CloudColor)
  n78 = CollectionParameter() {collection=&FortniteMaterialParameters, parameterName=SkyBoxClouds}
  n79 = QualitySwitch(Default=@n77, Epic=LinearInterpolate(A=@n76, B=@n77, Alpha=Saturate(None=@n78)))
  n80 = ScalarParameter() {parameterName=ReflectionCaptureDesaturation}
  ReflectionCapturePassSwitch!(
   Default=@n79
   Reflection=Multiply(A=Desaturation(None=@n79, Fraction=@n80), B=VectorParameter() {parameterName=ReflectionCaptureColorMult, defaultValue=<rgba 1 1 1 0>}.RGB)
  )
  Saturate(None=Multiply(A=@n48.CloudAlpha) {constB=10})
  SkyAtmosphereDistantLightScatteredLuminance()
  TextureCoordinate()
  TextureCoordinate()
  TextureCoordinate()
  n81 = TextureSample(UVs=Multiply(A=NamedRerouteUsage()) {constB=50}) {samplerType=SAMPLERTYPE_Grayscale, texture=&LowResBlurredNoise_Radish_LF_Root}
  n82 = ScalarParameter() {parameterName=adadad}
  TextureSample(UVs=Add(A=NamedRerouteUsage(), B=Multiply(A=@n81.R, B=@n82) {constB=0.001})) {samplerSource=SSM_Wrap_WorldGroupSettings, texture=&T_milkyway}
 ]
}
