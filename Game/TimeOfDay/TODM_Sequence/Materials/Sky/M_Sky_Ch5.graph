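# Unlit Chapter 5 sky material for the time-of-day sky, layering atmosphere, night gradient, stars, moon, clouds and storm horizon.
# Sky atmosphere luminance blends with a curve-atlas night horizon, twinkling Milky Way stars and panned virtual-plane high-altitude clouds.
# Helios functions supply sun, moon and skybox clouds, gated by quality switches and parameter collections.
# Reflection captures get boosted, time-of-day-desaturated color, and a mobile-only branch derives a vertical offset from camera height.
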Material /Game/TimeOfDay/TODM_Sequence/Materials/Sky/M_Sky_Ch5
graph {
 properties {bIsSky=true, pixelDepthOffsetMode=PDOM_Legacy, shadingModel=MSM_Unlit, usedShadingModels=Unlit}
 nodes [
  n1 = Constant() {r=0.7}
  n2 = WorldPosition()
  n3 = CameraPositionWS()
  n4 = DotProduct(A=SkyAtmosphereLightDirection(), B=Normalize(VectorInput=Subtract(A=@n2.XYZ, B=@n3)))
  n5 = Power(Base=ConstantBiasScale(None=Saturate(None=@n4)), Exp=Constant() {r=3})
  n6 = StaticSwitchParameter(
   True=Multiply(
    A=LinearInterpolate(A=@n1, Alpha=@n5)
    B=VectorParameter() {group=Night Sky, parameterName=FakeSkyLight, defaultValue=<rgba 0.0144 0.1912 0.3073 1>}.RGB
   )
   False=Constant()
  ) {parameterName=UseFakeSkylight}
  n7 = SkyAtmosphereViewLuminance!(WorldDirection)
  n8 = TextureCoordinate() {uTiling=0.5, vTiling=0.5}
  n9 = AppendVector(A=Constant() {r=0.25}, B=Constant() {r=0.25})
  n10 = LinearInterpolate(Alpha=ScalarParameter() {defaultValue=0.25, group=Night Sky, parameterName=Nightsky Horizon Height}) {constA=0.25, constB=0.2}
  n11 = Constant()
  n12 = CurveAtlasRowParameter(CurveTime=NamedRerouteDeclaration(None=SphereMask(A=@n8, B=@n9, Radius=@n10, Hardness=@n11)) {name=HorizonLine, nodeColor=<rgba 0.875 1 0 1>}) {
   atlas=/Game/TimeOfDay/TODM_Sequence/Materials/Sky/MaterialFunctions/CA_Athena_TOD
   curve=/Game/TimeOfDay/TODM_Sequence/Materials/Sky/MaterialFunctions/NightSkyHorizonCurve
   defaultValue=1
   parameterName=Sky Horizon Curve
  }
  n13 = LinearInterpolate(
   A=LinearInterpolate(
    A=VectorParameter() {group=Night Sky, parameterName=nightsky horizon color, defaultValue=<rgba 15 0 0.4283 1>}.RGB
    B=VectorParameter() {group=Night Sky, parameterName=nightsky color, defaultValue=<rgba 0.405 0.45 0.4458 1>}.RGB
    Alpha=Saturate(None=@n12.G)
   )
   B=SkyAtmosphereViewLuminance()
   Alpha=ScalarParameter() {defaultValue=1, group=Night Sky, parameterName=NightSkyViewLuminanceFactor}
  ) {constA=1}
  n14 = Add(A=@n13, B=NamedRerouteUsage())
  n15 = ScalarParameter() {defaultValue=1, group=Night Sky, parameterName=Nightsky Galaxy Strength}
  NamedRerouteDeclaration(None=LinearInterpolate(
    A=Add(A=@n6, B=@n7)
    B=LinearInterpolate(A=@n13, B=@n14, Alpha=@n15)
    Alpha=ScalarParameter() {defaultValue=1, group=Night Sky, parameterName=Night Sky blend}
   )) {name=Horizon Normal, nodeColor=<rgba 1 0 0.8437 1>}
  n16 = SkyAtmosphereDistantLightScatteredLuminance()
  n17 = Saturate(None=Multiply(A=NamedRerouteUsage()) {constB=6})
  n18 = ScalarParameter() {defaultValue=19000000, group=HighAltitudeClouds, parameterName=HA_Cloud_VirtualplaneScale}
  n19 = Panner(
   Coordinate=MaterialFunctionCall(
    Plane Normal Axis=Constant3Vector() {constant=<rgba 0 0 -1 0>}
    Plane X Axis=Constant3Vector() {constant=<rgba 1 0 0 0>}
    Plane Center=VectorParameter() {parameterName=HA_Cloud_Height, defaultValue=<rgba 0 0 1200000 1>}.RGB
    UV Scale=AppendVector(A=@n18, B=@n18)
   ) {materialFunction=&VirtualPlaneCoordinates}.UVs
   Speed=AppendVector(
    A=ScalarParameter() {defaultValue=-0.0001, group=HighAltitudeClouds, parameterName=HA_Cloud_Speed_X}
    B=ScalarParameter() {defaultValue=0.0005, group=HighAltitudeClouds, parameterName=HA_Cloud_Speed_Y}
   )
  )
  n20 = TextureSampleParameter2D(UVs=Multiply(A=@n19, B=ScalarParameter() {defaultValue=2.1989, group=HighAltitudeClouds, parameterName=HighAlititudeClouds_Tiling})) {
   group=HighAltitudeClouds
   parameterName=HighAlittudeCloudTexture
   samplerSource=SSM_Wrap_WorldGroupSettings
   samplerType=SAMPLERTYPE_LinearColor
   texture=/Game/TimeOfDay/TODM_Sequence/Materials/Sky/Textures/Clouds_m
  }
  n21 = AppendVector(
   A=ScalarParameter() {defaultValue=0, group=HighAltitudeClouds, parameterName=HA_Cloud_Speed2_X}
   B=ScalarParameter() {defaultValue=0.0001, group=HighAltitudeClouds, parameterName=HA_Cloud_Speed2_Y}
  )
  n22 = TextureSampleParameter2D(UVs=Multiply(
    A=Panner(Coordinate=@n19, Speed=@n21)
    B=ScalarParameter() {defaultValue=3, group=HighAltitudeClouds, parameterName=HighAlititudeClouds_Tiling_3}
   )) {
   group=HighAltitudeClouds
   parameterName=HighAlittudeCloudTexture
   samplerSource=SSM_Wrap_WorldGroupSettings
   samplerType=SAMPLERTYPE_LinearColor
   texture=&Clouds_m
  }
  n23 = Power(Base=Multiply(A=@n20.G, B=@n22.B)) {constExponent=1.8}
  n24 = Saturate(None=Multiply(A=Multiply(A=@n17, B=@n23), B=ScalarParameter() {defaultValue=0.5, group=HighAltitudeClouds, parameterName=High Alt Cloud Opacity}))
  NamedRerouteDeclaration(None=Multiply(A=LinearInterpolate(A=@n16) {constAlpha=0.05, constB=0.5}, B=@n24)) {name=High Alt Cloud Color, nodeColor=<rgba 1 0 0.6562 1>}
  NamedRerouteDeclaration(None=@n24) {name=High Alt Cloud Mask, nodeColor=<rgba 1 0.3047 0 1>}
  n25 = Constant()
  n26 = NamedRerouteDeclaration(None=ScalarParameter() {defaultValue=0.016, group=Night Sky, parameterName=NightSkyRotation}) {name=NightSkyRotation, nodeColor=<rgba 0 1 0.4375 1>}
  NamedRerouteDeclaration(None=Add(A=TextureCoordinate(), B=AppendVector(A=@n25, B=@n26))) {name=star pan, nodeColor=<rgba 0.3594 0 1 1>}
  n27 = MaterialFunctionCall() {materialFunction=/Game/TimeOfDay/TODM_Sequence/Materials/Sky/MaterialFunctions/MF_Helios_SunMoon}
  NamedRerouteDeclaration(None=@n27.MoonMask) {name=MoonMask, nodeColor=<rgba 1 0.8672 0 1>}
  n28 = Constant3Vector() {constant=<rgba 43.7499 39.9786 39.1023 1>}
  n29 = NamedRerouteUsage()
  n30 = Multiply(A=@n29, B=ScalarParameter() {defaultValue=20, parameterName=Start1}) {constB=5}
  n31 = TextureObject() {samplerType=SAMPLERTYPE_LinearColor, texture=/Game/TimeOfDay/TODM_Sequence/Materials/Sky/Textures/SkyNightMap_output}
  n32 = Multiply(A=@n29, B=ScalarParameter() {defaultValue=8, parameterName=StarMask})
  n33 = NamedRerouteUsage()
  n34 = Multiply(A=@n33, B=ScalarParameter() {defaultValue=12, parameterName=Start2}) {constB=5}
  n35 = Multiply(A=@n33, B=ScalarParameter() {defaultValue=15, parameterName=StarMask2})
  n36 = Add(
   A=Multiply(
    A=TextureSample(UVs=@n30, Tex=@n31) {samplerType=SAMPLERTYPE_LinearColor}.G
    B=TextureSample(UVs=@n32, Tex=@n31) {samplerSource=SSM_Wrap_WorldGroupSettings, samplerType=SAMPLERTYPE_LinearColor}.R
   )
   B=Multiply(
    A=TextureSample(UVs=@n34, Tex=@n31) {samplerSource=SSM_Wrap_WorldGroupSettings, samplerType=SAMPLERTYPE_LinearColor}.G
    B=TextureSample(UVs=@n35, Tex=@n31) {samplerSource=SSM_Wrap_WorldGroupSettings, samplerType=SAMPLERTYPE_LinearColor}.B
   )
  )
  n37 = Multiply(A=NamedRerouteUsage(), B=Constant() {r=5})
  n38 = Multiply(
   A=LinearInterpolate(B=Multiply(A=@n28, B=@n36), Alpha=NamedRerouteUsage())
   B=LinearInterpolate(Alpha=Saturate(None=@n37)) {constA=1, constB=0.6}
  )
  n39 = TextureSample(UVs=Panner(Coordinate=TextureCoordinate() {uTiling=18, vTiling=18}) {speedX=0.36}) {
   samplerSource=SSM_Wrap_WorldGroupSettings
   samplerType=SAMPLERTYPE_Grayscale
   texture=/Game/TimeOfDay/TODM_Sequence/Materials/Sky/Textures/LowResBlurredNoise_Radish_LF_Root
  }
  n40 = MaterialFunctionCall!(Default=LinearInterpolate(Alpha=@n39.R) {constA=0.05}, Low/Mobile=Constant() {r=0.8}) {materialFunction=/Game/Effects/Fort_Effects/Materials/_Functions/MF_LowMobile_PathingSwitch}
  n41 = NamedRerouteUsage()
  n42 = NamedRerouteUsage()
  NamedRerouteDeclaration(None=LinearInterpolate(A=Multiply(A=@n38, B=@n40.Result), Alpha=Multiply(A=@n41, B=@n42)) {constB=0}) {name=MilkywayStars, nodeColor=<rgba 0.7812 0 1 1>}
  NamedRerouteDeclaration(None=@n27.MoonHaloMask) {name=MoonHaloMask, nodeColor=<rgba 0.4297 1 0 1>}
  NamedRerouteDeclaration(None=@n27.MoonActive) {name=MoonActive, nodeColor=<rgba 0 0.2734 1 1>}
  n43 = MaterialFunctionCall(UVs=NamedRerouteUsage(), Texture Scale=ScalarParameter() {defaultValue=1, group=Night Sky, parameterName=MIkyWaySize}) {materialFunction=&ScaleUVsByCenter}
  n44 = TextureSample(
   UVs=Multiply(
    A=NamedRerouteDeclaration(None=@n43.UVs) {name=milkyway_uv, nodeColor=<rgba 0 1 0.8359 1>}
    B=ScalarParameter() {defaultValue=5500, group=Night Sky, parameterName=StarmaskFrequency}
   )
   Tex=@n31
  ) {
   samplerSource=SSM_Wrap_WorldGroupSettings
   samplerType=SAMPLERTYPE_LinearColor
   texture=/Game/TimeOfDay/TODM_Sequence/Materials/Sky/Textures/T_milkyway_StarMask
  }
  NamedRerouteDeclaration(None=@n44.B) {name=StarMask, nodeColor=<rgba 1 0.8438 0 1>}
  n45 = MaterialFunctionCall() {materialFunction=/Game/TimeOfDay/TODM_Sequence/Materials/Sky/Clouds/MF_Helios_Skybox_Clouds}
  n46 = NamedRerouteDeclaration(None=@n45.CloudAlpha) {name=cloudAlpha, nodeColor=<rgba 1 0 0.4687 1>}
  n47 = NamedRerouteUsage()
  n48 = Multiply(A=NamedRerouteUsage()) {constB=1.5}
  n49 = Multiply(A=@n27.Result, B=OneMinus(None=Saturate(None=@n48)))
  n50 = Saturate(None=CollectionParameter() {collection=&FortniteMaterialParameters, parameterName=SkyBoxClouds})
  n51 = CollectionParameter() {collection=/Game/Athena/SafeZone/MPC_SafeZoneClouds, parameterName=CloudStartingBias}
  n52 = OneMinus(None=Saturate(None=Multiply(A=@n51) {constB=2}))
  n53 = Constant() {r=0.5}
  n54 = SphereMask(
   A=SphereMask(A=TextureCoordinate(), B=@n53, Radius=@n53, Hardness=Constant())
   B=Constant() {r=0.085}
   Radius=Constant() {r=0.07}
   Hardness=Constant() {r=0.1128}
  )
  n55 = LinearInterpolate(
   A=NamedRerouteUsage()
   B=VectorParameter() {parameterName=StormHorizonGradientColor, defaultValue=<rgba 0.0907 0.0645 0.125 1>}.RGB
   Alpha=NamedRerouteDeclaration(None=Multiply(A=@n52, B=@n54)) {name=StormGradientMask, nodeColor=<rgba 0.4297 1 0 1>}
  ) {constAlpha=1}
  n56 = Add(
   A=QualitySwitch(Default=@n49, Epic=LinearInterpolate(A=@n27.Result, B=@n49, Alpha=@n50))
   B=QualitySwitch(Default=NamedRerouteUsage(), High=@n55, Epic=@n55)
  )
  n57 = Add(A=NamedRerouteUsage(), B=@n56)
  n58 = MaterialFunctionCall!(
   Low _ PC Low and Med=@n56
   "High _ XB1, PS4"=@n57
   "Epic _ PC High, XSX, PS5"=@n57
   --------
   "Mobile Low _ PC Perf, Android"=@n56
   Mobile Med _ Switch=@n56
  ) {materialFunction=/Game/Global/MaterialFunctions/TechArt/MF_Fortnite_QualitySwitch_Vector3}
  n59 = LinearInterpolate(A=@n58.Result, B=@n45.CloudColor, Alpha=Saturate(None=Subtract(A=@n46, B=@n47))) {constAlpha=0, constB=0}
  n60 = CollectionParameter() {collection=&FortniteMaterialParameters, parameterName=SkyBoxClouds}
  n61 = QualitySwitch(Default=@n59, Epic=LinearInterpolate(A=@n58.Result, B=@n59, Alpha=Saturate(None=@n60)) {constAlpha=1})
  n62 = ScalarParameter() {defaultValue=1, parameterName=Reflect_Color_Boost}
  n63 = ScalarParameter() {defaultValue=0.2, parameterName=ReflectionDesaturation_Morning}
  n64 = ScalarParameter() {defaultValue=0.2, parameterName=ReflectionDesaturation_Day}
  n65 = ScalarParameter() {defaultValue=0.2, parameterName=ReflectionDesaturation_Evening}
  n66 = ScalarParameter() {defaultValue=0.45, parameterName=ReflectionDesaturation_Night}
  ReflectionCapturePassSwitch!(
   Default=@n61
   Reflection=Desaturation(
    None=Multiply(A=@n62, B=@n61)
    Fraction=MaterialFunctionCall(Value_Morning=@n63, Value_Day=@n64, Value_Evening=@n65, Value_Night=@n66) {materialFunction=&Scalar_Time_of_Day}.Result
   )
  )
  Saturate(None=@n23)
  n67 = ComponentMask(None=CameraPositionWS()) {b=true}
  ShadingPathSwitch(Default=Constant(), Mobile=AppendVector(A=Constant2Vector(), B=Multiply(A=@n67) {constB=0.9}))
 ]
}
