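# Sky material function that tints an input color with separate horizon and zenith override colors.
# A horizon weight comes from one minus the absolute vertical component of the camera vector, raised to an exponent parameter, so it is mirrored below the horizon.
# The color lerps to the horizon override by that weight, then to the zenith override by its inverse.
# Each override color's alpha scales its blend strength.
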MaterialFunction /Game/TimeOfDay/CloudsAndStorm/Materials/MaterialFunctions/MF_Sky_ColorOverride
graph {nodes [
  n1 = FunctionInput() {inputName=Color}
  n2 = VectorParameter() {parameterName=ColorOverrideHorizon, defaultValue=<rgba 1 1 1 0>}
  n3 = ComponentMask(None=CameraVectorWS()) {b=true}
  n4 = Power(Base=OneMinus(None=Abs(None=@n3)), Exp=ScalarParameter() {defaultValue=4, parameterName=ColorOverrideExponent})
  n5 = Multiply(A=@n2.A, B=@n4)
  n6 = VectorParameter() {parameterName=ColorOverrideZenith, defaultValue=<rgba 1 1 1 0>}
  n7 = OneMinus(None=@n4)
  FunctionOutput(None=LinearInterpolate(A=LinearInterpolate(A=@n1, B=@n2.RGB, Alpha=@n5), B=@n6.RGB, Alpha=Multiply(A=@n6.A, B=@n7)))
 ]}
