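# M_Sky_Ch6 is the unlit sky material for the time-of-day sequence, layering atmosphere, stars, halo and high-altitude clouds.
# A sharpened vertex-colour horizon mask and a sun bloom from atmosphere light direction are shared through named reroutes.
# The layered colour passes through colour override and circular storm darkening functions.
# Reflection capture switches substitute separate colour and world position offset from MF_Sky_Reflection, with zero offset by default.
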Material /Game/TimeOfDay/TODM_Sequence/Materials/Sky/M_Sky_Ch6
graph {
 properties {bAutomaticallySetUsageInEditor=false, bIsSky=true, shadingModel=MSM_Unlit, usedShadingModels=Unlit}
 nodes [
  n1 = VertexColor()
  n2 = Multiply(A=@n1.R, B=@n1.R)
  n3 = Multiply(A=@n2, B=@n2)
  n4 = Multiply(A=@n3, B=@n3)
  NamedRerouteDeclaration(None=OneMinus(None=Multiply(A=@n4) {constB=0.9})) {name=Horizon Mask, nodeColor=<rgba 0 1 0.7422 1>}
  n5 = SkyAtmosphereLightDirection()
  n6 = MaterialFunctionCall!(Multiplier, Offset) {materialFunction=&/MF_MirroredCameraVector}
  n7 = Saturate(None=Multiply(A=DotProduct(A=@n5, B=@n6.Result)) {constB=-1})
  n8 = ScalarParameter() {defaultValue=80, parameterName=SunBloomExp}
  NamedRerouteDeclaration(None=Multiply(A=Power(Base=@n7, Exp=@n8), B=VectorParameter() {parameterName=SunBloomColor, defaultValue=<rgba 0 0 0 1>}.RGB)) {name=Sun Bloom, nodeColor=<rgba 1 0 0.8906 1>}
  n9 = MaterialFunctionCall(Multiplier=Constant() {r=-1}) {materialFunction=&MF_MirroredCameraVector}
  n10 = NamedRerouteUsage()
  n11 = MaterialFunctionCall!(
   Sky Atmosphere=Add(
    A=SkyAtmosphereViewLuminance!(WorldDirection=@n9.Result)
    B=MaterialFunctionCall!(HorizonMask, UV) {materialFunction=/Game/TimeOfDay/TODM_Sequence/Materials/Sky/MaterialFunctions/MF_Sky_Stars}.Result
   )
   Horizon Mask=@n10
  ) {materialFunction=/Game/TimeOfDay/TODM_Sequence/Materials/Sky/MaterialFunctions/MF_Sky_Halo}
  n12 = Add(
   A=NamedRerouteDeclaration(None=MaterialFunctionCall!(Color=@n11.Result, Halo Mask=@n11.Result, Horizon Mask=@n10) {materialFunction=/Game/TimeOfDay/TODM_Sequence/Materials/Sky/MaterialFunctions/MF_Sky_HighAltitudeClouds}.Color) {name=Pre Skylight Captures, nodeColor=<rgba 0 1 0.5781 1>}
   B=NamedRerouteUsage()
  )
  n13 = MaterialFunctionCall!(Color=NamedRerouteUsage()) {materialFunction=/Game/TimeOfDay/TODM_Sequence/Materials/Sky/MaterialFunctions/MF_Sky_Reflection}
  ReflectionCapturePassSwitch!(
   Default=MaterialFunctionCall!(
    Color=MaterialFunctionCall!(Color=@n12) {materialFunction=/Game/TimeOfDay/CloudsAndStorm/Materials/MaterialFunctions/MF_Sky_ColorOverride}.Result
    Storm Color=VectorParameter() {parameterName=Storm Darkening, defaultValue=<rgba 0.1332 0.0989 0.4219 1>}.RGB
    Height
   ) {materialFunction=/Game/TimeOfDay/CloudsAndStorm/Materials/MaterialFunctions/MF_Storm_DarkeningCircle}.Result
   Reflection=@n13.Reflection Color
  )
  ReflectionCapturePassSwitch(Default=Constant(), Reflection=@n13.Reflection WPO)
 ]
}
