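# Unlit time-of-day sky material (Ch6 Jethro) that layers atmosphere, stars, halo, high-altitude clouds, sun bloom and storm darkening.
# Stars are cut out by an analytic ray-sphere test, with the sphere shifting as safe zone scale drops, driven by parameter collection or custom primitive data.
# A horizon mask comes from vertex color red.
# Reflection captures switch to separate color and vertex offset from MF_Sky_Reflection.
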Material /Game/TimeOfDay/TODM_Sequence/Materials/Sky/M_Sky_Ch6_Jethro
graph {
 properties {bAutomaticallySetUsageInEditor=false, bIsSky=true, shadingModel=MSM_Unlit, usedShadingModels=Unlit}
 nodes [
  n1 = VertexColor()
  n2 = Multiply(A=@n1.R, B=@n1.R)
  n3 = Multiply(A=@n2, B=@n2)
  n4 = Multiply(A=@n3, B=@n3)
  NamedRerouteDeclaration(None=OneMinus(None=Multiply(A=@n4) {constB=0.9})) {name=Horizon Mask, nodeColor=<rgba 0 1 0.7422 1>}
  n5 = SkyAtmosphereLightDirection()
  n6 = MaterialFunctionCall!(Multiplier, Offset) {materialFunction=&/MF_MirroredCameraVector}
  n7 = Saturate(None=Multiply(A=DotProduct(A=@n5, B=@n6.Result)) {constB=-1})
  n8 = ScalarParameter() {defaultValue=80, parameterName=SunBloomExp}
  NamedRerouteDeclaration(None=Multiply(A=Power(Base=@n7, Exp=@n8), B=VectorParameter() {parameterName=SunBloomColor, defaultValue=<rgba 0 0 0 1>}.RGB)) {name=Sun Bloom, nodeColor=<rgba 1 0 0.8906 1>}
  n9 = MaterialFunctionCall(Multiplier=Constant() {r=-1}) {materialFunction=&MF_MirroredCameraVector}
  n10 = MaterialFunctionCall!(HorizonMask, UV) {materialFunction=/Game/TimeOfDay/TODM_Sequence/Materials/Sky/MaterialFunctions/MF_Sky_Stars}
  n11 = CollectionParameter() {collection=/Game/Athena/SafeZone/MPC_SafeZoneClouds, parameterName=StarCutoutPosition+Radius}
  n12 = ComponentMask(None=@n11) {b=true, g=true, r=true}
  n13 = VectorParameter() {
   bUseCustomPrimitiveData=true
   desc="Direction in local space. Ie we can travel \"forward\" "
   group=Storm Progression
   parameterName=StormProgressWPO.TravelDirection.CustomPrim
   primitiveDataIndex=14
   channelNames {a=TravelDistance, b=Z, g=Y, r=X}
   defaultValue=<rgba 0 1 0 1>
  }
  n14 = Normalize(VectorInput=@n13)
  n15 = Length(None=@n14) {is=class MaterialExpressionLength}
  n16 = LinearInterpolate(A=Constant3Vector() {constant=<rgba 0 1 0 0>}, B=@n14, Alpha=Saturate(None=Sign(None=@n15)))
  n17 = CollectionParameter() {collection=&MPC_SafeZoneClouds, parameterName=StarCutoutTravelDirection+Distance}
  n18 = Normalize(VectorInput=ComponentMask(None=@n17) {b=true, g=true, r=true})
  n19 = ComponentMask(None=@n17) {a=true}
  n20 = CollectionParameter() {collection=&MPC_SafeZoneClouds, parameterName=SafeZoneScale}
  n21 = LinearInterpolate(
   B=StaticSwitchParameter(True=Multiply(A=@n16, B=@n13.TravelDistance), False=Multiply(A=@n18, B=@n19)) {group=Storm Progression, parameterName=UseCustomPrimForTransform, sortPriority=2}
   Alpha=Saturate(None=InverseLinearInterpolate(Value=@n20) {constA=1, constB=0})
  )
  n22 = Subtract(A=CameraPositionWS(), B=ComponentMask(None=Add(A=@n12, B=@n21)) {b=true, g=true, r=true})
  n23 = DotProduct(A=@n22, B=CameraVectorWS())
  n24 = ComponentMask(None=@n11) {a=true}
  n25 = Subtract(A=Multiply(A=@n23, B=@n23), B=Subtract(A=DotProduct(A=@n22, B=@n22), B=Multiply(A=@n24, B=@n24)))
  n26 = OneMinus(None=Ceil(None=Saturate(None=@n25)))
  n27 = NamedRerouteUsage()
  n28 = MaterialFunctionCall!(Sky Atmosphere=Add(A=SkyAtmosphereViewLuminance!(WorldDirection=@n9.Result), B=Multiply(A=@n10.Result, B=@n26)), Horizon Mask=@n27) {materialFunction=/Game/TimeOfDay/TODM_Sequence/Materials/Sky/MaterialFunctions/MF_Sky_Halo}
  n29 = NamedRerouteDeclaration(None=MaterialFunctionCall!(Color=@n28.Result, Halo Mask=@n28.Result, Horizon Mask=@n27) {materialFunction=/Game/TimeOfDay/TODM_Sequence/Materials/Sky/MaterialFunctions/MF_Sky_HighAltitudeClouds}.Color) {name=Pre Skylight Captures, nodeColor=<rgba 0 1 0.5781 1>}
  n30 = NamedRerouteUsage()
  n31 = MaterialFunctionCall!(Color=NamedRerouteUsage()) {materialFunction=/Game/TimeOfDay/TODM_Sequence/Materials/Sky/MaterialFunctions/MF_Sky_Reflection}
  ReflectionCapturePassSwitch!(
   Default=MaterialFunctionCall!(
    Color=Add(A=@n29, B=@n30)
    Storm Color=VectorParameter() {parameterName=Storm Darkening, defaultValue=<rgba 0.1332 0.0989 0.4219 1>}.RGB
    Height
   ) {materialFunction=/Game/TimeOfDay/CloudsAndStorm/Materials/MaterialFunctions/MF_Storm_DarkeningCircle}.Result
   Reflection=@n31.Reflection Color
  )
  ReflectionCapturePassSwitch(Default=Constant(), Reflection=@n31.Reflection WPO)
  ScalarParameter() {
   defaultValue=300000
   desc=Distacne to we will ahve traveled when storm is at max
   group=Storm Progression
   parameterName=StormProgressWPO.TravelDistance
  }
  VectorParameter() {group=Storm Progression, parameterName=Strong Arm Position + Radius, defaultValue=<rgba -123980.7656 10709891 1500000 5000000>}
  VectorParameter() {
   desc="Direction in local space. Ie we can travel \"forward\" "
   group=Storm Progression
   parameterName=StormProgressWPO.TravelDirection
   defaultValue=<rgba 0 1 0 1>
  }
 ]
}
