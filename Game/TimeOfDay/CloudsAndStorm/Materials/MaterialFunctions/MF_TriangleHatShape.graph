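# Material function in the clouds and storm time-of-day materials that reshapes a scalar R into a triangle hat curve.
# R, previewed with texture U, is remapped to 2R-1, then one minus its absolute value gives a ramp peaking at 0.5.
# Extra outputs give the remapped value's sign and a 0-1 step marking the upper half.
# A PlotFunctionOnGraph node graphs the curve for debugging and feeds no output.
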MaterialFunction /Game/TimeOfDay/CloudsAndStorm/Materials/MaterialFunctions/MF_TriangleHatShape
graph {nodes [
  n1 = ComponentMask(None=TextureCoordinate()) {r=true}
  n2 = Subtract(A=Multiply(A=FunctionInput(Preview=@n1) {bUsePreviewValueAsDefault=true, inputName=R, inputType=FunctionInput_Scalar}) {constB=2})
  FunctionOutput(None=Sign(None=@n2)) {outputName=Sign}
  n3 = OneMinus(None=Abs(None=@n2))
  FunctionOutput(None=@n3) {outputName=TriangleHatShape}
  FunctionOutput(None=Ceil(None=Saturate(None=@n2))) {outputName=0-1}
  MaterialFunctionCall!(X range Y range=Constant4Vector() {constant=<rgba 0 1 0 1>}.RGBA, Gradient Height=Constant(), ShowGrid?, Thickness, Color, "f(x)"=@n3) {materialFunction=/Engine/Functions/Engine_MaterialFunctions01/Debug/PlotFunctionOnGraph}
 ]}
