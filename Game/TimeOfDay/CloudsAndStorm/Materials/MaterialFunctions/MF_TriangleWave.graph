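# MF_TriangleWave is a material function in the TimeOfDay CloudsAndStorm folder that turns a scalar input into a repeating 0 to 1 triangle wave.
# It adds Offset to the input and multiplies by Period, which therefore acts as a frequency.
# Frac, a 0.5 subtraction, Abs and a doubling shape the wave, peaking where the scaled value is a whole number.
# Offset and Period are optional inputs with preview-value defaults.
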MaterialFunction /Game/TimeOfDay/CloudsAndStorm/Materials/MaterialFunctions/MF_TriangleWave
graph {nodes [
  n1 = Add(
   A=FunctionInput() {inputName="Input (x)", inputType=FunctionInput_Scalar}
   B=FunctionInput() {bUsePreviewValueAsDefault=true, inputName=Offset, inputType=FunctionInput_Scalar}
  ) {constB=-0.25, desc=offset}
  n2 = FunctionInput() {bUsePreviewValueAsDefault=true, inputName=Period, inputType=FunctionInput_Scalar, previewValue=<1 0 0 1>}
  n3 = Subtract(A=Frac(None=Multiply(A=@n1, B=@n2) {desc=period})) {constB=0.5}
  FunctionOutput(None=Multiply(A=Abs(None=@n3)) {constB=2}) {outputName="Multiply(,2)"}
 ]}
