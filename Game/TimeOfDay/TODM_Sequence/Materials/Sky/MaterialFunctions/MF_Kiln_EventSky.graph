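# Material function that turns an input sky color into the Kiln event sky, behind the KilnEventSky? static switch.
# It darkens the sky with a mask and adds color built from graffiti, radially panned textures, sine-pulsed rays and shock-wave UV scaling.
# Named reroutes carry these layers, including a center sphere mask sized by the Kiln_sky_Blend collection parameter; Sky_phase4 blends to the full effect.
# Mobile uses low textures and drops several layers.
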MaterialFunction /Game/TimeOfDay/TODM_Sequence/Materials/Sky/MaterialFunctions/MF_Kiln_EventSky
graph {nodes [
  n1 = FunctionInput(Preview=Constant3Vector() {constant=<rgba 0.3379 0.7336 1 1>}) {bUsePreviewValueAsDefault=true, inputName=SkyColor}
  n2 = Desaturation(None=@n1, Fraction=Constant())
  n3 = NamedRerouteUsage()
  n4 = TextureSampleParameter2D(UVs=@n3) {
   group=Kiln
   parameterName=T_SkyBox_Mask_A_low
   samplerSource=SSM_Wrap_WorldGroupSettings
   texture=&T_BlacksRGB
  }
  n5 = LinearInterpolate(Alpha=ShadingPathSwitch(
    Default=TextureSampleParameter2D(UVs=@n3) {group=Kiln, parameterName=T_SkyBox_Mask_A, texture=&T_BlacksRGB}.R
    Mobile=@n4.R
   )) {constA=1, constB=0.3}
  n6 = NamedRerouteUsage()
  n7 = Saturate(None=Add(A=Multiply(A=@n2, B=@n5), B=Saturate(None=@n6)))
  n8 = LinearInterpolate(A=@n1, B=@n7, Alpha=NamedRerouteUsage())
  n9 = ScalarParameter() {group=Kiln, parameterName=Sky_phase4}
  FunctionOutput(None=StaticSwitchParameter(True=LinearInterpolate(A=@n8, B=@n7, Alpha=@n9), False=@n1) {group=Kiln, parameterName=KilnEventSky?})
  n10 = SphereMask(A=TextureCoordinate(), B=AppendVector(A=Constant() {r=0.5}, B=Constant() {r=0.5}), Radius=Constant() {r=0.55}, Hardness=Constant())
  n11 = Multiply(A=@n10, B=Constant() {r=4}) {constB=2}
  n12 = NamedRerouteUsage()
  n13 = TextureSampleParameter2D(UVs=@n12) {
   group=Kiln
   parameterName=T_SkyBox_Mask_A
   samplerSource=SSM_Wrap_WorldGroupSettings
   texture=&T_BlacksRGB
  }
  n14 = TextureSampleParameter2D(UVs=@n12) {
   group=Kiln
   parameterName=T_SkyBox_Mask_A_low
   samplerSource=SSM_Wrap_WorldGroupSettings
   texture=&T_BlacksRGB
  }
  n15 = ShadingPathSwitch(Default=@n13.B, Mobile=@n14.B)
  n16 = ShadingPathSwitch(Default=@n13.G, Mobile=@n14.G)
  n17 = Power(Base=Add(A=@n16) {constB=0.5}, Exp=Constant() {r=20}) {constExponent=15}
  n18 = Saturate(None=Multiply(A=@n16) {constB=1.5})
  n19 = Constant() {r=0.9}
  n20 = LinearInterpolate(A=@n15, B=Saturate(None=Max(A=@n15, B=@n17)), Alpha=Saturate(None=Multiply(A=@n18, B=@n19) {constB=0.95}))
  n21 = Constant() {r=7}
  n22 = Multiply(A=Add(A=Time(), B=Multiply(A=@n20, B=@n21) {constB=13}), B=Constant() {r=0.27})
  n23 = Add(A=Sine(None=Add(A=@n11, B=@n22)))
  n24 = Time()
  n25 = NamedRerouteUsage()
  n26 = TextureSampleParameter2D(UVs=@n25) {
   group=Kiln
   parameterName=T_SkyBox_Mask_A
   samplerSource=SSM_Wrap_WorldGroupSettings
   texture=&T_BlacksRGB
  }
  n27 = TextureSampleParameter2D(UVs=@n25) {
   group=Kiln
   parameterName=T_SkyBox_Mask_A_low
   samplerSource=SSM_Wrap_WorldGroupSettings
   texture=&T_BlacksRGB
  }
  n28 = Multiply(A=ShadingPathSwitch(Default=@n26.B, Mobile=@n27.B), B=Constant() {r=7}) {constB=13}
  n29 = Add(A=@n11, B=Multiply(A=Add(A=@n24, B=@n28), B=Constant() {r=0.27}))
  n30 = Multiply(A=Add(A=Sine(None=@n29))) {constB=0.5}
  n31 = Add(A=SphereMask(A=TextureCoordinate(), B=Constant() {r=0.5}, Radius=Constant() {r=0.1}, Hardness=Constant() {r=0.1}) {attenuationRadius=0.2}) {constB=0.1}
  NamedRerouteDeclaration(None=LinearInterpolate(
    A=Multiply(A=@n23) {constB=0.5}
    B=NamedRerouteDeclaration(None=@n30) {name=Light ray Mask_1, nodeColor=<rgba 0 0.9531 1 1>}
    Alpha=Saturate(None=@n31)
   )) {name=Light ray Mask, nodeColor=<rgba 0 0.9531 1 1>}
  n32 = TextureCoordinate()
  n33 = Constant2Vector() {g=1, r=1}
  n34 = Add(A=Multiply(A=Multiply(A=@n32, B=@n33), B=Constant() {r=1}) {constB=3}, B=AppendVector(A=Constant() {r=-0.5}, B=Constant() {r=-0.5})) {constB=-1.5}
  n35 = AppendVector(A=Constant(), B=Constant() {r=0.06})
  n36 = Panner(
   Coordinate=MaterialFunctionCall(Vector or UVs=@n34) {materialFunction=&VectorToRadialValue}.Vector Converted to Angle
   Speed=@n35
  ) {speedY=-0.2}
  NamedRerouteDeclaration(None=ShadingPathSwitch(
    Default=TextureSampleParameter2D(UVs=@n36) {group=Kiln, parameterName=T_SkyBox_Mask_B, texture=&T_BlacksRGB}.R
    Mobile=TextureSampleParameter2D(UVs=@n36) {group=Kiln, parameterName=T_SkyBox_Mask_B_low, texture=&T_BlacksRGB}.R
   )) {name=Disc trailing ray, nodeColor=<rgba 0 0.2734 1 1>}
  n37 = MaterialFunctionCall(Vector or UVs=@n34) {materialFunction=&VectorToRadialValue}
  NamedRerouteDeclaration(None=TextureSample(UVs=Panner(Coordinate=@n37.Radial Coordinates, Speed=@n35) {speedY=-0.2}) {constCoordinate=1, samplerSource=SSM_Wrap_WorldGroupSettings, texture=/Game/Global/Textures/Default/Noise/T_Noise_Cloud}.RGB) {name=Disc trailing color, nodeColor=<rgba 0.1484 1 0 1>}
  n38 = Subtract(A=TextureCoordinate()) {constB=0.5}
  n39 = Time()
  n40 = Constant() {r=0.002}
  n41 = Constant() {r=-0.003}
  n42 = MaterialFunctionCall(
   UVs=Panner(
    Coordinate=MaterialFunctionCall(Vector or UVs=@n38) {materialFunction=&VectorToRadialValue}.Radial Coordinates
    Time=Add(A=@n39) {constB=0.1}
    Speed=AppendVector(A=@n40, B=@n41)
   )
   "Rotation Angle (0-1)"=Constant() {r=1}
  ) {materialFunction=&CustomRotator}
  n43 = TextureSampleParameter2D(UVs=@n42.Rotated Values) {group=Kiln, parameterName=T_SkyBox_Fall, samplerSource=SSM_Wrap_WorldGroupSettings, texture=&T_BlacksRGB}
  n44 = TextureSampleParameter2D(UVs=@n42.Rotated Values) {
   group=Kiln
   parameterName=T_SkyBox_Fall_low
   samplerSource=SSM_Wrap_WorldGroupSettings
   texture=&T_BlacksRGB
  }
  n45 = TextureCoordinate()
  n46 = Panner(
   Coordinate=MaterialFunctionCall(Vector or UVs=Subtract(A=@n45) {constB=0.5}) {materialFunction=&VectorToRadialValue}.Radial Coordinates
   Speed=AppendVector(A=Constant() {r=0.0006}, B=Constant() {r=-0.01})
  )
  n47 = TextureSampleParameter2D(UVs=@n46) {group=Kiln, parameterName=T_SkyBox_Fall, samplerSource=SSM_Wrap_WorldGroupSettings, texture=&T_BlacksRGB}
  n48 = TextureSampleParameter2D(UVs=@n46) {
   group=Kiln
   parameterName=T_SkyBox_Fall_low
   samplerSource=SSM_Wrap_WorldGroupSettings
   texture=&T_BlacksRGB
  }
  n49 = Max(
   A=Multiply(A=ShadingPathSwitch(Default=@n43.RGB, Mobile=@n44.RGB)) {constB=0.1}
   B=Multiply(A=ShadingPathSwitch(Default=@n47.RGB, Mobile=@n48.RGB), B=Constant() {r=0.8})
  )
  n50 = NamedRerouteUsage()
  n51 = TextureSampleParameter2D(UVs=@n50) {
   group=Kiln
   parameterName=T_SkyBox_Mask_A
   samplerSource=SSM_Wrap_WorldGroupSettings
   texture=&T_BlacksRGB
  }
  n52 = TextureSampleParameter2D(UVs=@n50) {
   group=Kiln
   parameterName=T_SkyBox_Mask_A_low
   samplerSource=SSM_Wrap_WorldGroupSettings
   texture=&T_BlacksRGB
  }
  n53 = Multiply(A=ShadingPathSwitch(Default=@n51.G, Mobile=@n52.G)) {constB=0.9}
  n54 = TextureCoordinate()
  n55 = Constant() {r=0.5}
  n56 = Constant() {r=0.1}
  n57 = Constant() {r=0.1}
  n58 = NamedRerouteUsage()
  n59 = Power(Base=NamedRerouteUsage(), Exp=Constant() {r=3})
  n60 = Power(Base=ShadingPathSwitch(Default=@n51.R, Mobile=@n52.R))
  n61 = Constant() {r=0.02}
  n62 = NamedRerouteUsage()
  n63 = TextureSampleParameter2D(UVs=@n62) {group=Kiln, parameterName=Graffiti_C, samplerSource=SSM_Wrap_WorldGroupSettings, texture=&T_BlacksRGB}
  n64 = TextureSampleParameter2D(UVs=@n62) {
   group=Kiln
   parameterName=Graffiti_C_low
   samplerSource=SSM_Wrap_WorldGroupSettings
   texture=&T_BlacksRGB
  }
  n65 = ShadingPathSwitch(Default=@n63.RGB, Mobile=@n64.RGB)
  n66 = Multiply(
   A=Multiply(
    A=Multiply(A=@n59, B=@n60)
    B=MaterialFunctionCall(Hue Shift Percentage=@n61, Texture=@n65) {materialFunction=&HueShift}.Result
   )
   B=Constant() {r=1}
  ) {constB=5}
  n67 = Multiply(A=@n65)
  n68 = ShadingPathSwitch(Default=Add(A=Add(A=@n58, B=@n66), B=@n67), Mobile=@n67)
  n69 = Add(
   A=LinearInterpolate(A=LinearInterpolate(A=@n49, Alpha=@n53) {constB=0}, Alpha=SphereMask(A=@n54, B=@n55, Radius=@n56, Hardness=@n57) {attenuationRadius=0.2}) {constB=0}
   B=@n68
  )
  n70 = TextureCoordinate()
  n71 = Constant() {r=0.5}
  n72 = Constant() {r=0.1}
  n73 = Constant() {r=0.15}
  NamedRerouteDeclaration(None=LinearInterpolate(
    A=ShadingPathSwitch(Default=@n69, Mobile=@n68)
    B=Constant3Vector() {constant=<rgba 0.7344 0 0.1777 1>}
    Alpha=SphereMask(A=@n70, B=@n71, Radius=@n72, Hardness=@n73) {attenuationRadius=0.2}
   )) {name=Final color, nodeColor=<rgba 1 0.7734 0 1>}
  n74 = NamedRerouteUsage()
  n75 = Multiply(
   A=ShadingPathSwitch(
    Default=TextureSampleParameter2D(UVs=@n74) {group=Kiln, parameterName=T_SkyBox_Mask_B, texture=&T_BlacksRGB}.G
    Mobile=TextureSampleParameter2D(UVs=@n74) {group=Kiln, parameterName=T_SkyBox_Mask_B_low, texture=&T_BlacksRGB}.G
   )
   B=NamedRerouteUsage()
  )
  n76 = Multiply(A=NamedRerouteUsage()) {constB=10}
  NamedRerouteDeclaration(None=ShadingPathSwitch(Default=Multiply(A=@n75, B=@n76) {constB=3}, Mobile=Constant())) {name=Disc trailing, nodeColor=<rgba 0.7578 1 0 1>}
  n77 = TextureCoordinate()
  n78 = Constant() {r=0.5}
  n79 = CollectionParameter() {collection=/Game/Athena/SafeZone/MPC_SafeZoneClouds, parameterName=Kiln_sky_Blend}
  n80 = Saturate(None=Subtract(A=Multiply(A=@n79), B=Constant() {r=0.005}) {constB=0.2})
  n81 = Add(A=@n80, B=Multiply(A=@n80, B=NamedRerouteUsage()))
  NamedRerouteDeclaration(None=Power(Base=SphereMask(A=@n77, B=@n78, Radius=@n81) {hardnessPercent=90}) {constExponent=1}) {name=SkyBlendMask, nodeColor=<rgba 1 0.9141 0 1>}
  n82 = TextureSampleParameter2D() {group=Kiln, parameterName=Graffiti_C, samplerSource=SSM_Wrap_WorldGroupSettings, texture=&T_BlacksRGB}
  n83 = TextureSampleParameter2D() {
   group=Kiln
   parameterName=Graffiti_C_low
   samplerSource=SSM_Wrap_WorldGroupSettings
   texture=&T_BlacksRGB
  }
  n84 = Power(Base=Multiply(A=ShadingPathSwitch(Default=@n82.R, Mobile=@n83.R))) {constExponent=0.5}
  n85 = Max(
   A=Multiply(A=ShadingPathSwitch(Default=@n82.G, Mobile=@n83.G)) {constB=0.2}
   B=Multiply(A=ShadingPathSwitch(Default=@n82.B, Mobile=@n83.B)) {constB=0.5}
  )
  NamedRerouteDeclaration(None=Multiply(A=Max(A=@n84, B=@n85))) {name=Radius_Noise, nodeColor=<rgba 0 1 0.1094 1>}
  n86 = Multiply(A=@n10, B=Constant() {r=2.5}) {constB=2}
  n87 = Multiply(A=Time()) {constB=0.15}
  n88 = Add(A=Sine(None=Add(A=@n86, B=@n87)))
  NamedRerouteDeclaration(None=Saturate(None=Multiply(A=@n88) {constB=0.5})) {name=Shock Wave, nodeColor=<rgba 1 0.2344 0 1>}
  n89 = Constant() {r=1.14}
  n90 = Constant() {r=1.155}
  n91 = NamedRerouteUsage()
  NamedRerouteDeclaration(None=MaterialFunctionCall(UVs=TextureCoordinate(), Texture Scale=LinearInterpolate(A=@n89, B=@n90, Alpha=@n91) {constB=1.3}) {materialFunction=&ScaleUVsByCenter}.UVs) {name=WaveUV, nodeColor=<rgba 0 0.0859 1 1>}
 ]}
