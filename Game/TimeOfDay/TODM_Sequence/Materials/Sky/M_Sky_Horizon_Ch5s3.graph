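# Unlit sky material for Chapter 5 Season 3 time of day, combining atmosphere and night-sky color, sun/moon, stars, storm gradient and clouds.
# Horizon clouds use noise-warped, two-phase flow-map sampling in lat-long UVs, with reveal masking and dawn/dusk tints.
# Panned virtual-plane textures form high-altitude clouds, and a skybox cloud function layers over the sky.
# Quality switches gate cloud and storm features, and reflection captures get a boosted, time-of-day-desaturated result.
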Material /Game/TimeOfDay/TODM_Sequence/Materials/Sky/M_Sky_Horizon_Ch5s3
graph {
 properties {bIsSky=true, pixelDepthOffsetMode=PDOM_Legacy, shadingModel=MSM_Unlit, usedShadingModels=Unlit}
 nodes [
  n1 = NamedRerouteUsage()
  Add(A=MaterialFunctionCall!(R=ComponentMask(None=@n1) {r=true}) {materialFunction=/Game/TimeOfDay/CloudsAndStorm/Materials/MaterialFunctions/MF_TriangleHatShape}.TriangleHatShape) {constB=0}
  n2 = SkyAtmosphereLightIlluminance!(World Position=Multiply(A=WorldPosition().XYZ, B=Constant3Vector() {constant=<rgba 1 1 100 0>}))
  Add(A=@n2)
  n3 = TextureSampleParameter2D!(UVs=Clamp(None=NamedRerouteUsage(), Min=NamedRerouteUsage(), Max=NamedRerouteUsage()), Level=Constant() {r=4}, Apply View MipBias) {
   group=Horizon Clouds
   mipValueMode=TMVM_MipLevel
   parameterName=HorizonTexture
   samplerType=SAMPLERTYPE_LinearColor
   texture=/Game/TimeOfDay/CloudsAndStorm/Textures/T_CloudCombinedDefault
  }
  Add(A=@n3.A) {constB=0}
  Constant()
  Constant()
  Constant() {r=1}
  Desaturation()
  n4 = CollectionParameter() {collection=&FortniteMaterialParameters, parameterName=TimeOfDay_0-1}
  n5 = Constant() {r=2048}
  n6 = Divide(A=Add(A=@n5), B=@n5)
  n7 = TextureSample(UVs=MaterialFunctionCall(X=Multiply(A=@n4, B=@n6) {constB=512}, Y=Constant()) {materialFunction=&MakeFloat2}.Result) {samplerType=SAMPLERTYPE_LinearColor, texture=/Game/TimeOfDay/CloudsAndStorm/TechArt/RT_LUT_Tex2}
  Desaturation(None=@n7.RGB, Fraction=ScalarParameter() {parameterName=LUTSaturation})
  Divide(B=Constant() {r=1.3})
  LinearInterpolate(A=SkyAtmosphereViewLuminance!(WorldDirection))
  n8 = CollectionParameter() {collection=&FortniteMaterialParameters, parameterName=HadesActive}
  n9 = TextureCoordinate() {uTiling=0.5, vTiling=0.5}
  n10 = AppendVector(A=Constant() {r=0.25}, B=Constant() {r=0.25})
  n11 = LinearInterpolate(Alpha=ScalarParameter() {defaultValue=0.25, group=Night Sky, parameterName=Nightsky Horizon Height}) {constA=0.25, constB=0.2}
  n12 = Constant()
  n13 = CurveAtlasRowParameter(CurveTime=NamedRerouteDeclaration(None=SphereMask(A=@n9, B=@n10, Radius=@n11, Hardness=@n12)) {name=HorizonLine, nodeColor=<rgba 0.875 1 0 1>}) {
   atlas=/Game/TimeOfDay/TODM_Sequence/Materials/Sky/MaterialFunctions/CA_Athena_TOD
   curve=/Game/TimeOfDay/TODM_Sequence/Materials/Sky/MaterialFunctions/NightSkyHorizonCurve
   defaultValue=1
   parameterName=Sky Horizon Curve
  }
  n14 = LinearInterpolate(
   A=LinearInterpolate(
    A=VectorParameter() {group=Night Sky, parameterName=nightsky horizon color, defaultValue=<rgba 0 8.3587 15 1>}.RGB
    B=VectorParameter() {group=Night Sky, parameterName=nightsky horizon color _ Hades, defaultValue=<rgba 15 6.878 0 1>}.RGB
    Alpha=@n8
   )
   B=LinearInterpolate(
    A=VectorParameter() {group=Night Sky, parameterName=nightsky color, defaultValue=<rgba 0.405 0.45 0.4458 1>}.RGB
    B=VectorParameter() {group=Night Sky, parameterName=nightsky color _ Hades, defaultValue=<rgba 0.1775 0.0862 0.45 1>}.RGB
    Alpha=@n8
   )
   Alpha=Saturate(None=@n13.G)
  )
  LinearInterpolate(A=@n14, B=NamedRerouteDeclaration() {name=Clouds_2, nodeColor=<rgba 0.4062 0 1 1>}, Alpha=NamedRerouteUsage())
  n15 = NamedRerouteUsage()
  LinearInterpolate(A=VectorParameter() {parameterName=Halo DuskDawn, defaultValue=<rgba 1 1 1 1>}.RGB, Alpha=@n15)
  n16 = NamedRerouteUsage()
  n17 = Multiply(A=@n16) {constB=-1}
  n18 = ScalarParameter() {defaultValue=0.6, parameterName=Facing Darken Factor Dusk}
  n19 = ScalarParameter() {defaultValue=0.6, parameterName=Facing Darken Factor Dawn}
  n20 = NamedRerouteUsage()
  n21 = Saturate(None=MaterialFunctionCall(
    Input=@n17
    Input Low=Constant() {r=-0.5}
    Input High=Constant() {r=1}
    Target Low=LinearInterpolate(A=@n18, B=@n19, Alpha=@n20)
    Target High=Constant() {r=1}
   ) {materialFunction=&RemapValueRange}.Result)
  n22 = OneMinus(None=Abs(None=@n17))
  n23 = Divide(B=Constant() {r=1})
  LinearInterpolate(B=@n21, Alpha=OneMinus(None=Power(Base=@n22, Exp=@n23))) {constA=1, desc="makes it 1 when sun is above and only darkens at 90  degrees"}
  n24 = ComponentMask(None=TextureCoordinate()) {r=true}
  MaterialFunctionCall!(
   X range Y range=Constant4Vector() {constant=<rgba 0 1 0 1>}.RGBA
   Gradient Height=Constant()
   ShowGrid?
   Thickness
   Color
   "f(x)"=MaterialFunctionCall!(Offset=Constant() {r=-0.25}, Period, "Input (x)"=@n24) {materialFunction=/Game/TimeOfDay/CloudsAndStorm/Materials/MaterialFunctions/MF_TriangleWave}."Multiply(,2)"
  ) {materialFunction=/Engine/Functions/Engine_MaterialFunctions01/Debug/PlotFunctionOnGraph}
  n25 = Add(A=@n24) {constB=0.5}
  Multiply(A=Subtract(A=@n24, B=Floor(None=@n25))) {constB=2}
  n26 = MaterialFunctionCall() {materialFunction=/Game/TimeOfDay/TODM_Sequence/Materials/Sky/MaterialFunctions/MF_Helios_SunMoon}
  n27 = Multiply(A=NamedRerouteUsage()) {constB=1.5}
  n28 = Multiply(A=@n26.Result, B=OneMinus(None=Saturate(None=@n27)))
  n29 = Saturate(None=CollectionParameter() {collection=&FortniteMaterialParameters, parameterName=SkyBoxClouds})
  n30 = Multiply(A=CollectionParameter() {collection=/Game/Athena/SafeZone/MPC_SafeZoneClouds, parameterName=CloudStartingBias}) {constB=2}
  n31 = TextureCoordinate()
  n32 = Constant() {r=0.5}
  n33 = Constant()
  n34 = Multiply(
   A=OneMinus(None=Saturate(None=@n30))
   B=SphereMask(A=SphereMask(A=@n31, B=@n32, Radius=@n32, Hardness=@n33), B=Constant() {r=0.085}, Radius=Constant() {r=0.07}, Hardness=Constant() {r=0.1128})
  )
  n35 = Saturate(None=CollectionParameter() {collection=&MPC_SafeZoneClouds, parameterName=EnableVolumetricStorm})
  n36 = LinearInterpolate(
   A=NamedRerouteUsage()
   B=VectorParameter() {parameterName=StormHorizonGradientColor, defaultValue=<rgba 0.0907 0.0645 0.125 1>}.RGB
   Alpha=NamedRerouteDeclaration(None=Multiply(A=@n34, B=@n35)) {name=StormGradientMask, nodeColor=<rgba 0.4297 1 0 1>}
  ) {constAlpha=1}
  n37 = Add(
   A=QualitySwitch(Default=@n28, Epic=LinearInterpolate(A=@n26.Result, B=@n28, Alpha=@n29))
   B=QualitySwitch(Default=NamedRerouteUsage(), High=@n36, Epic=@n36)
  )
  n38 = Add(A=NamedRerouteUsage(), B=@n37)
  n39 = MaterialFunctionCall!(
   Low _ PC Low and Med=@n37
   "High _ XB1, PS4"=@n38
   "Epic _ PC High, XSX, PS5"=@n38
   --------
   "Mobile Low _ PC Perf, Android"=@n37
   Mobile Med _ Switch=@n37
  ) {materialFunction=/Game/Global/MaterialFunctions/TechArt/MF_Fortnite_QualitySwitch_Vector3}
  n40 = VectorParameter() {parameterName=Global Shadow Tint, defaultValue=<rgba 0.3601 0.5465 0.7889 1>}
  n41 = VectorParameter() {parameterName=Global Highlight Tint, defaultValue=<rgba 1.8 1.8 1.8 1>}
  n42 = NamedRerouteUsage()
  n43 = MaterialFunctionCall(
   Input=@n42
   Input High=ScalarParameter() {defaultValue=2.71, group=Horizon Clouds, parameterName=Horizon Highlight Level Reference}
   Target Low=ScalarParameter() {defaultValue=1, group=Horizon Clouds, parameterName=Raise Horizon Shadow Gain}
  ) {materialFunction=&RemapValueRange}
  n44 = Multiply(A=Multiply(A=LinearInterpolate(A=@n40.RGB, B=@n41.RGB, Alpha=@n42), B=@n43.Result), B=NamedRerouteUsage())
  n45 = LinearInterpolate(A=@n39.Result, B=LinearInterpolate(A=@n44, B=NamedRerouteUsage(), Alpha=NamedRerouteUsage()), Alpha=NamedRerouteUsage())
  Multiply(A=@n45, B=Constant3Vector() {constant=<rgba 0.9689 0 1 1>})
  n46 = LinearInterpolate(A=@n14, B=SkyAtmosphereViewLuminance(), Alpha=ScalarParameter() {defaultValue=1, group=Night Sky, parameterName=NightSkyViewLuminanceFactor}) {constA=1}
  Multiply(A=@n46) {constB=0.5}
  n47 = NamedRerouteUsage()
  Multiply(A=@n47)
  Multiply(A=Multiply(A=@n43.Result, B=NamedRerouteUsage()), B=NamedRerouteUsage())
  Multiply(A=@n44, B=NamedRerouteUsage())
  n48 = Subtract(A=WorldPosition().XYZ, B=CameraPositionWS())
  n49 = Saturate(None=DotProduct(A=SkyAtmosphereLightDirection(), B=Normalize(VectorInput=@n48)))
  n50 = LinearInterpolate(A=Constant() {r=0.7}, Alpha=Power(Base=ConstantBiasScale(None=@n49), Exp=Constant() {r=3}))
  n51 = VectorParameter() {group=Night Sky, parameterName=FakeSkyLight, defaultValue=<rgba 0.0144 0.1912 0.3073 1>}
  n52 = Add(A=StaticSwitchParameter(True=Multiply(A=@n50, B=@n51.RGB), False=Constant()) {parameterName=UseFakeSkylight}, B=SkyAtmosphereViewLuminance())
  n53 = LinearInterpolate(
   A=@n46
   B=Add(A=@n46, B=NamedRerouteUsage())
   Alpha=ScalarParameter() {defaultValue=1, group=Night Sky, parameterName=Nightsky Galaxy Strength}
  )
  n54 = ScalarParameter() {defaultValue=1, group=Night Sky, parameterName=Night Sky blend}
  NamedRerouteDeclaration(None=MaterialFunctionCall!(SkyColor=LinearInterpolate(A=@n52, B=@n53, Alpha=@n54)) {materialFunction=/Game/TimeOfDay/TODM_Sequence/Materials/Sky/MaterialFunctions/MF_Kiln_EventSky}.Result) {name=Horizon Normal, nodeColor=<rgba 1 0 0.8437 1>}
  n55 = SkyAtmosphereDistantLightScatteredLuminance()
  n56 = Saturate(None=Multiply(A=NamedRerouteUsage()) {constB=6})
  n57 = ScalarParameter() {defaultValue=19000000, group=HighAltitudeClouds, parameterName=HA_Cloud_VirtualplaneScale}
  n58 = Panner(
   Coordinate=MaterialFunctionCall(
    Plane Normal Axis=Constant3Vector() {constant=<rgba 0 0 -1 0>}
    Plane X Axis=Constant3Vector() {constant=<rgba 1 0 0 0>}
    Plane Center=VectorParameter() {parameterName=HA_Cloud_Height, defaultValue=<rgba 0 0 1200000 1>}.RGB
    UV Scale=AppendVector(A=@n57, B=@n57)
   ) {materialFunction=&VirtualPlaneCoordinates}.UVs
   Speed=AppendVector(
    A=ScalarParameter() {defaultValue=-0.0001, group=HighAltitudeClouds, parameterName=HA_Cloud_Speed_X}
    B=ScalarParameter() {defaultValue=0.0005, group=HighAltitudeClouds, parameterName=HA_Cloud_Speed_Y}
   )
  )
  n59 = TextureSampleParameter2D!(
   UVs=Multiply(A=@n58, B=ScalarParameter() {defaultValue=2.1989, group=HighAltitudeClouds, parameterName=HighAlititudeClouds_Tiling})
   Apply View MipBias
  ) {
   group=HighAltitudeClouds
   parameterName=HighAlittudeCloudTexture
   samplerSource=SSM_Wrap_WorldGroupSettings
   samplerType=SAMPLERTYPE_LinearColor
   texture=/Game/TimeOfDay/TODM_Sequence/Materials/Sky/Textures/Clouds_m
  }
  n60 = AppendVector(
   A=ScalarParameter() {defaultValue=0, group=HighAltitudeClouds, parameterName=HA_Cloud_Speed2_X}
   B=ScalarParameter() {defaultValue=0.0001, group=HighAltitudeClouds, parameterName=HA_Cloud_Speed2_Y}
  )
  n61 = TextureSampleParameter2D(UVs=Multiply(
    A=Panner(Coordinate=@n58, Speed=@n60)
    B=ScalarParameter() {defaultValue=3, group=HighAltitudeClouds, parameterName=HighAlititudeClouds_Tiling_3}
   )) {
   group=HighAltitudeClouds
   parameterName=HighAlittudeCloudTexture
   samplerSource=SSM_Wrap_WorldGroupSettings
   samplerType=SAMPLERTYPE_LinearColor
   texture=&Clouds_m
  }
  n62 = Power(Base=Multiply(A=@n59.G, B=@n61.B)) {constExponent=1.8}
  n63 = Saturate(None=Multiply(A=Multiply(A=@n56, B=@n62), B=ScalarParameter() {defaultValue=0.5, group=HighAltitudeClouds, parameterName=High Alt Cloud Opacity}))
  NamedRerouteDeclaration(None=Multiply(A=LinearInterpolate(A=@n55) {constAlpha=0.05, constB=0.5}, B=@n63)) {name=High Alt Cloud Color, nodeColor=<rgba 1 0 0.6562 1>}
  NamedRerouteDeclaration(None=@n63) {name=High Alt Cloud Mask, nodeColor=<rgba 1 0.3047 0 1>}
  n64 = Saturate(None=DotProduct(A=NamedRerouteUsage(), B=NamedRerouteUsage()))
  n65 = ScalarParameter() {defaultValue=0.8, group=Horizon Clouds, parameterName=Horizon Cloud Contrast}
  NamedRerouteDeclaration(None=Multiply(
    A=Power(Base=@n64, Exp=@n65)
    B=ScalarParameter() {defaultValue=2.71, group=Horizon Clouds, parameterName="Horizon Cloud Gain (Normalization factor)"}
   )) {name=Clouds, nodeColor=<rgba 0.4062 0 1 1>}
  n66 = Constant()
  n67 = NamedRerouteDeclaration(None=ScalarParameter() {defaultValue=0.016, group=Night Sky, parameterName=NightSkyRotation}) {name=NightSkyRotation, nodeColor=<rgba 0 1 0.4375 1>}
  NamedRerouteDeclaration(None=Add(A=TextureCoordinate(), B=AppendVector(A=@n66, B=@n67))) {name=star pan, nodeColor=<rgba 0.3594 0 1 1>}
  n68 = NamedRerouteUsage()
  n69 = NamedRerouteUsage()
  n70 = TextureSampleParameter2D!(UVs=NamedRerouteUsage(), "DDX(UVs)"=@n68, "DDY(UVs)"=@n69, Apply View MipBias) {
   group=Horizon Clouds
   mipValueMode=TMVM_Derivative
   parameterName=HorizonDepthMap
   samplerType=SAMPLERTYPE_LinearColor
   texture=/Game/TimeOfDay/CloudsAndStorm/Textures/T_DefaultHorizonDepthGradientMap
  }
  n71 = TextureSampleParameter2D(UVs=NamedRerouteUsage(), "DDX(UVs)"=@n68, "DDY(UVs)"=@n69) {
   group=Horizon Clouds
   mipValueMode=TMVM_Derivative
   parameterName=HorizonDepthMap
   samplerType=SAMPLERTYPE_LinearColor
   texture=&T_DefaultHorizonDepthGradientMap
  }
  NamedRerouteDeclaration(None=LinearInterpolate(A=@n70.R, B=@n71.R, Alpha=NamedRerouteUsage())) {name=DepthFlow, nodeColor=<rgba 1 0.3984 0 1>}
  n72 = Divide(A=CollectionParameter() {collection=&FortniteMaterialParameters, parameterName=SunAngle}) {constB=180, desc=normalize}
  MaterialExpressionNamedRerouteDeclaration_5 = NamedRerouteDeclaration(None=Saturate(None=OneMinus(None=@n72))) {name=SunMOon0-1, nodeColor=<rgba 1 0.0234 0 1>}
  n73 = Custom!(x=@MaterialExpressionNamedRerouteDeclaration_5) {
   code="// The below expression will get compiled\r\n// into the output of this node\r\nreturn saturate(1-abs(2*x - float3(0,1,2)));"
   showCode=true
   inputs [{inputName=x, input {expression=&self:MaterialExpressionNamedRerouteDeclaration_5}}]
  }
  NamedRerouteDeclaration(None=@n73) {name=TODCloudBlendFactor, nodeColor=<rgba 1 0.8906 0 1>}
  n74 = NamedRerouteUsage()
  n75 = ScalarParameter() {defaultValue=0.2, group=Horizon Clouds, parameterName=CloudRevealConstrast}
  n76 = Power(
   Base=MaterialFunctionCall(In=Multiply(A=@n74) {constB=2, desc=normalize?}, Contrast=@n75) {materialFunction=&CheapContrast}.Result
   Exp=Add(A=Constant() {r=1}, B=Multiply(A=@n75))
  )
  n77 = ScalarParameter() {defaultValue=1, group=Horizon Clouds, parameterName=CloudRevealThreshold}
  n78 = ConstantBiasScale(None=Power(Base=OneMinus(None=@n77)) {constExponent=1.3}) {bias=-0.5, scale=2}
  n79 = Multiply(A=Saturate(None=Subtract(A=@n76, B=@n78)), B=NamedRerouteUsage())
  n80 = ScalarParameter() {defaultValue=1, group=Horizon Clouds, parameterName=CloudAllpha Gain}
  NamedRerouteDeclaration(None=Saturate(None=Multiply(A=@n79, B=@n80) {constB=2.15})) {name=HorizonReveal, nodeColor=<rgba 0 1 0.4141 1>}
  n81 = MaterialFunctionCall!(CameraVector) {materialFunction=/Engine/Functions/Engine_MaterialFunctions01/Texturing/CameraVectorToLatLongUV}
  n82 = MaterialFunctionCall(R=ComponentMask(None=@n81.Uvs) {r=true}) {materialFunction=&MF_TriangleHatShape}
  NamedRerouteDeclaration(None=Add(A=@n82.TriangleHatShape) {constB=0}) {name=HorizonTriangleHat, nodeColor=<rgba 0 1 0.9531 1>}
  NamedRerouteDeclaration(None=@n26.MoonMask) {name=MoonMask, nodeColor=<rgba 1 0.8672 0 1>}
  NamedRerouteDeclaration(None=@n26.MoonActive) {name=MoonActive, nodeColor=<rgba 0 0.2734 1 1>}
  NamedRerouteDeclaration(None=@n3.A) {name=HorizonEdgeMask, nodeColor=<rgba 1 0 0.5391 1>}
  NamedRerouteDeclaration(None=@n26.MoonHaloMask) {name=MoonHaloMask, nodeColor=<rgba 0.4297 1 0 1>}
  n83 = CollectionParameter() {
   collection=&FortniteMaterialParameters
   parameterName=SunlightDirectionalVector
  }
  n84 = Normalize(VectorInput=@n83)
  n85 = Normalize(VectorInput=WorldPosition().XYZ)
  n86 = Saturate(None=Multiply(A=DotProduct(A=@n84, B=@n85)) {constB=1.5})
  NamedRerouteDeclaration(None=Multiply(A=Power(Base=@n86) {constExponent=4}, B=NamedRerouteUsage())) {name=sun dot, nodeColor=<rgba 1 0.4922 0 1>}
  n87 = WorldPosition()
  n88 = ShadingPathSwitch(Default=Multiply(A=@n87.XYZ, B=Constant3Vector() {constant=<rgba 0.5 0.5 0.25 0>}) {constB=0.1}, Forward=@n87.XYZ, Mobile=@n87.XYZ)
  NamedRerouteDeclaration(None=Multiply(A=SkyAtmosphereAerialPerspective!(World Position=@n88), B=ScalarParameter() {parameterName=Cloud_AerialPerspectiveFactor}) {constB=0.3}) {name=CloudAerialPerspective, nodeColor=<rgba 1 0 1 1>}
  n89 = SphereMask(
   A=TextureCoordinate()
   B=Constant() {r=0.5}
   Radius=ScalarParameter() {defaultValue=0.7, group=KIln, parameterName=KilnCloudMaskRadius}
   Hardness=ScalarParameter() {defaultValue=0.15, group=KIln, parameterName=KilnHardness}
  )
  NamedRerouteDeclaration(None=OneMinus(None=Saturate(None=@n89))) {name=KilnCloudMask, nodeColor=<rgba 0 0.2031 1 1>}
  NamedRerouteDeclaration(None=@n21) {name=DarkenHorizonAtDuskDawn, nodeColor=<rgba 1 0 0.7266 1>}
  n90 = DotProduct(A=Normalize(VectorInput=SkyAtmosphereLightDirection()), B=NamedRerouteUsage())
  NamedRerouteDeclaration(None=Saturate(None=ConstantBiasScale(None=@n90))) {name=Horizon Sun DOt 0-1, nodeColor=<rgba 0 1 0.9531 1>}
  n91 = ScalarParameter() {defaultValue=2, parameterName=BrightenBehind}
  n92 = NamedRerouteUsage()
  NamedRerouteDeclaration(None=LinearInterpolate(A=Multiply(A=@n15, B=@n91), B=ScalarParameter() {defaultValue=1.4, parameterName=PostGain}, Alpha=Power(Base=@n92) {constExponent=0.5}) {constA=2}) {name=brighten halo, nodeColor=<rgba 0 0.0859 1 1>}
  NamedRerouteDeclaration(None=Multiply(A=@n7.RGB, B=ScalarParameter() {defaultValue=0.7, parameterName=CloudLUTGain})) {name=LUT Colors, nodeColor=<rgba 0.4062 1 0 1>}
  n93 = Saturate(None=NamedRerouteUsage())
  n94 = LinearInterpolate(
   A=VectorParameter() {group=Horizon Clouds, parameterName=Dusk Shadow Tint, defaultValue=<rgba 0.4102 0.3681 0.3513 1>}.RGB
   B=VectorParameter() {group=Horizon Clouds, parameterName=Dusk Highlight Tint, defaultValue=<rgba 1 0.8447 0.7279 1>}.RGB
   Alpha=@n93
  )
  n95 = LinearInterpolate(
   A=VectorParameter() {group=Horizon Clouds, parameterName=Dawn Shadow TInt, defaultValue=<rgba 1 0.6828 0.3169 1>}.RGB
   B=VectorParameter() {group=Horizon Clouds, parameterName=Dawn Hightlight TInt, defaultValue=<rgba 1 0.4702 0.2809 1>}.RGB
   Alpha=@n93
  )
  n96 = Divide(A=Subtract(A=CollectionParameter() {collection=&FortniteMaterialParameters, parameterName=SunAngle}) {constB=90}) {constB=90, desc=turn into -1 to 1}
  n97 = NamedRerouteDeclaration(None=ConstantBiasScale(None=@n96)) {name=dawn or dusk, nodeColor=<rgba 0.9219 0 1 1>}
  n98 = Power(Base=Abs(None=@n96), Exp=Constant() {desc=Tighten Dawn Dusk Darkening Timing, r=3.1416}) {constExponent=4}
  NamedRerouteDeclaration(None=LinearInterpolate(
    B=LinearInterpolate(A=@n94, B=@n95, Alpha=@n97)
    Alpha=NamedRerouteDeclaration(None=@n98) {name=Dawn Dusk Peaked Mask, nodeColor=<rgba 0.5234 0 1 1>}
   ) {constA=1, constB=0.6}) {name=DawnDuskTints, nodeColor=<rgba 0.3125 1 0 1>}
  NamedRerouteDeclaration(None=@n90) {name=Horizon Sun DOt Raw, nodeColor=<rgba 0 1 0.9531 1>}
  n99 = WorldPosition()
  NamedRerouteDeclaration(None=Normalize(VectorInput=Multiply(A=@n99.XYZ) {constB=-1})) {name=SmoothNormals, nodeColor=<rgba 0 0.0391 1 1>}
  n100 = ComponentMask(None=NamedRerouteUsage()) {g=true}
  n101 = OneMinus(None=@n100)
  n102 = Saturate(None=MaterialFunctionCall(
    Input=Add(A=@n101) {constB=0.5}
    Input Low=ScalarParameter() {defaultValue=0.7, parameterName=Horizon Gradient Max}
    Input High=ScalarParameter() {defaultValue=0.2, parameterName=Horizon Gradient Min}
   ) {materialFunction=&RemapValueRange}.Result)
  NamedRerouteDeclaration(None=@n102) {name=horizon blend mask, nodeColor=<rgba 0.2187 1 0 1>}
  NamedRerouteDeclaration(None=@n14) {name=hor, nodeColor=<rgba 0.2187 1 0 1>}
  NamedRerouteDeclaration(None=Multiply(
    A=SkyAtmosphereDistantLightScatteredLuminance()
    B=ScalarParameter() {defaultValue=3.16, group=Horizon Clouds, parameterName=Blend Color Gain}
   )) {name=Horizon Blend Color, nodeColor=<rgba 1 0 0.5391 1>}
  NamedRerouteDeclaration(None=@n46) {name=HorizonPureColor, nodeColor=<rgba 0 1 0.0156 1>}
  n103 = NamedRerouteUsage()
  n104 = NamedRerouteUsage()
  n105 = TextureSampleParameter2D(UVs=NamedRerouteUsage(), "DDX(UVs)"=@n103, "DDY(UVs)"=@n104) {
   group=Horizon Clouds
   mipValueMode=TMVM_Derivative
   parameterName=HorizonTexture
   samplerType=SAMPLERTYPE_LinearColor
   texture=&T_CloudCombinedDefault
  }
  n106 = TextureSampleParameter2D(UVs=NamedRerouteUsage(), "DDX(UVs)"=@n103, "DDY(UVs)"=@n104) {
   group=Horizon Clouds
   mipValueMode=TMVM_Derivative
   parameterName=HorizonTexture
   samplerType=SAMPLERTYPE_LinearColor
   texture=&T_CloudCombinedDefault
  }
  n107 = LinearInterpolate(A=@n105.RGBA, B=@n106.RGBA, Alpha=NamedRerouteUsage())
  NamedRerouteDeclaration(None=ComponentMask(None=@n107) {a=true}) {name=HorizonCloudAlpha, nodeColor=<rgba 0.9688 0 1 1>}
  NamedRerouteDeclaration(None=ComponentMask(None=@n107) {b=true, g=true, r=true}) {name=HorizonCloudColor, nodeColor=<rgba 1 0.8906 0 1>}
  n108 = MaterialFunctionCall(
   X=ScalarParameter() {defaultValue=0.012, group=Horizon Clouds, parameterName=Horizontal Offset}
   Y=ScalarParameter() {group=Horizon Clouds, parameterName=Vertical Offset}
  ) {materialFunction=&MakeFloat2}
  NamedRerouteDeclaration(None=Subtract(A=Constant2Vector() {g=0.01}, B=@n108.Result)) {name=uvclampmin, nodeColor=<rgba 1 0 0.5625 1>}
  NamedRerouteDeclaration(None=Subtract(A=Constant2Vector() {g=1.01, r=1}, B=@n108.Result)) {name=uvclampmax, nodeColor=<rgba 1 0 0.5625 1>}
  n109 = NamedRerouteUsage()
  n110 = NamedRerouteUsage() {desc=spatial bias}
  n111 = ComponentMask(None=Multiply(A=Subtract(A=@n110) {constB=0.3333, desc=bias helmholt}) {constB=2}) {g=true, r=true}
  n112 = Constant2Vector() {g=16, r=1}
  n113 = NamedRerouteUsage()
  n114 = Multiply(
   A=Multiply(A=Multiply(A=@n111, B=@n112)) {constB=0.001}
   B=Multiply(A=ScalarParameter() {defaultValue=8, group=Horizon Clouds, parameterName=warp}, B=OneMinus(None=@n113))
  )
  n115 = Time()
  n116 = NamedRerouteUsage()
  n117 = Multiply(A=Multiply(A=ComponentMask(None=@n116) {r=true}, B=ComponentMask(None=@n116) {g=true})) {constB=6.3}
  n118 = NamedRerouteUsage()
  n119 = Multiply(
   A=Add(A=Add(A=@n115, B=@n117), B=ComponentMask(None=@n118) {r=true}) {desc=time bias}
   B=ScalarParameter() {defaultValue=0.125, group=Horizon Clouds, parameterName=HorizonCloudAnimSpeed}
  )
  n120 = Frac(None=@n119)
  n121 = NamedRerouteUsage()
  n122 = Saturate(None=Multiply(A=Power(Base=@n121) {constExponent=0.2}, B=NamedRerouteUsage()))
  n123 = Multiply(A=Multiply(A=@n114, B=@n120), B=@n122)
  n124 = NamedRerouteUsage()
  n125 = NamedRerouteUsage()
  NamedRerouteDeclaration(None=Clamp(None=Add(A=@n109, B=@n123), Min=@n124, Max=@n125)) {name=horizonUVA, nodeColor=<rgba 0.6406 0 1 1>}
  n126 = ScalarParameter() {defaultValue=0.5, parameterName=UV_DDXMult}
  NamedRerouteDeclaration(None=DDX(Value=Multiply(A=@n82.TriangleHatShape, B=@n126))) {name=HorizonDDX, nodeColor=<rgba 0.8047 0 1 1>}
  n127 = Add(A=@n119) {constB=0.5}
  n128 = Multiply(A=Multiply(A=@n114, B=Frac(None=@n127)), B=@n122)
  NamedRerouteDeclaration(None=Clamp(None=Add(A=@n109, B=@n128), Min=@n124, Max=@n125)) {name=horizonUVB, nodeColor=<rgba 0 1 0.3203 1>}
  NamedRerouteDeclaration(None=OneMinus(None=MaterialFunctionCall(R=@n120) {materialFunction=&MF_TriangleHatShape}.TriangleHatShape)) {name=FlowBlend, nodeColor=<rgba 0.7109 1 0 1>}
  n129 = Divide(A=@n1, B=Constant2Vector() {g=28, r=1})
  n130 = ScalarParameter() {defaultValue=3.1416, parameterName=HorizonCloudWarpFrequency}
  n131 = Multiply(A=MaterialFunctionCall(X=Constant() {r=0.5}, Y=Time()) {materialFunction=&MakeFloat2}.Result) {constB=0.002}
  n132 = ScalarParameter() {defaultValue=0.1, group=Horizon Clouds, parameterName=HorizonCloudWarpRiseSpeed}
  n133 = TextureSample!(
   UVs=Add(A=Multiply(A=@n129, B=@n130) {constB=6}, B=Multiply(A=@n131, B=@n132) {constB=0.003})
   Tex
   "DDX(UVs)"=NamedRerouteUsage()
   "DDY(UVs)"=NamedRerouteUsage()
   Apply View MipBias
  ) {mipValueMode=TMVM_Derivative, samplerType=SAMPLERTYPE_LinearColor, texture=/Game/TimeOfDay/CloudsAndStorm/Materials/M_MakeNoise_Baked}
  NamedRerouteDeclaration(None=@n133.RGB) {name=HorizonNoise, nodeColor=<rgba 1 0.375 0 1>}
  n134 = ScalarParameter() {defaultValue=1, parameterName=UV_DDYMult}
  NamedRerouteDeclaration(None=DDY(Value=Multiply(A=@n81.Uvs, B=@n134))) {name=HorizonDDY, nodeColor=<rgba 0.8047 0 1 1>}
  n135 = NamedRerouteUsage()
  n136 = ScalarParameter() {defaultValue=20, parameterName=Start1}
  n137 = TextureObject() {samplerType=SAMPLERTYPE_LinearColor, texture=/Game/TimeOfDay/TODM_Sequence/Materials/Sky/Textures/SkyNightMap_output}
  n138 = ScalarParameter() {defaultValue=8, parameterName=StarMask}
  n139 = Multiply(
   A=TextureSample!(UVs=Multiply(A=@n135, B=@n136) {constB=5}, Tex=@n137, Apply View MipBias) {samplerType=SAMPLERTYPE_LinearColor}.G
   B=TextureSample(UVs=Multiply(A=@n135, B=@n138), Tex=@n137) {samplerSource=SSM_Wrap_WorldGroupSettings, samplerType=SAMPLERTYPE_LinearColor}.R
  )
  n140 = NamedRerouteUsage()
  n141 = ScalarParameter() {defaultValue=12, parameterName=Start2}
  n142 = ScalarParameter() {defaultValue=15, parameterName=StarMask2}
  n143 = Multiply(
   A=TextureSample(UVs=Multiply(A=@n140, B=@n141) {constB=5}, Tex=@n137) {samplerSource=SSM_Wrap_WorldGroupSettings, samplerType=SAMPLERTYPE_LinearColor}.G
   B=TextureSample(UVs=Multiply(A=@n140, B=@n142), Tex=@n137) {samplerSource=SSM_Wrap_WorldGroupSettings, samplerType=SAMPLERTYPE_LinearColor}.B
  )
  n144 = LinearInterpolate(B=Multiply(A=Constant3Vector() {constant=<rgba 43.7499 39.9786 39.1023 1>}, B=Add(A=@n139, B=@n143)), Alpha=NamedRerouteUsage())
  n145 = NamedRerouteUsage()
  n146 = Constant() {r=5}
  n147 = LinearInterpolate(Alpha=Saturate(None=Multiply(A=@n145, B=@n146))) {constA=1, constB=0.6}
  n148 = TextureSample(UVs=Panner(Coordinate=TextureCoordinate() {uTiling=18, vTiling=18}) {speedX=0.36}) {
   samplerSource=SSM_Wrap_WorldGroupSettings
   samplerType=SAMPLERTYPE_Grayscale
   texture=/Game/TimeOfDay/TODM_Sequence/Materials/Sky/Textures/LowResBlurredNoise_Radish_LF_Root
  }
  n149 = LinearInterpolate(Alpha=@n148.R) {constA=0.05}
  n150 = Constant() {r=0.8}
  NamedRerouteDeclaration(None=Multiply(
    A=Multiply(A=@n144, B=@n147)
    B=MaterialFunctionCall!(Default=@n149, Low/Mobile=@n150) {materialFunction=/Game/Effects/Fort_Effects/Materials/_Functions/MF_LowMobile_PathingSwitch}.Result
   )) {name=MilkywayStars, nodeColor=<rgba 0.7812 0 1 1>}
  NamedRerouteDeclaration(None=@n121) {name=WarpMask, nodeColor=<rgba 0.0547 1 0 1>}
  n151 = Constant() {r=1}
  n152 = ScalarParameter() {defaultValue=6, group=Horizon Clouds, parameterName=Unstretch Factor}
  n153 = OneMinus(None=Multiply(
    A=OneMinus(None=@n81.Uvs)
    B=MaterialFunctionCall(X=@n151, Y=@n152) {materialFunction=&MakeFloat2}.Result
   ) {constB=6})
  n154 = MaterialFunctionCall(X=Constant(), Y=Divide(A=@n152)) {materialFunction=&MakeFloat2}
  NamedRerouteDeclaration(None=Subtract(A=Add(A=@n153, B=@n154.Result), B=@n108.Result)) {name=horizon UVs, nodeColor=<rgba 0.1484 1 0 1>}
  n155 = MaterialFunctionCall(UVs=NamedRerouteUsage(), Texture Scale=ScalarParameter() {defaultValue=1, group=Night Sky, parameterName=MIkyWaySize}) {materialFunction=&ScaleUVsByCenter}
  n156 = TextureSample(
   UVs=Multiply(
    A=NamedRerouteDeclaration(None=@n155.UVs) {name=milkyway_uv, nodeColor=<rgba 0 1 0.8359 1>}
    B=ScalarParameter() {defaultValue=5500, group=Night Sky, parameterName=StarmaskFrequency}
   )
   Tex=@n137
  ) {
   samplerSource=SSM_Wrap_WorldGroupSettings
   samplerType=SAMPLERTYPE_LinearColor
   texture=/Game/TimeOfDay/TODM_Sequence/Materials/Sky/Textures/T_milkyway_StarMask
  }
  NamedRerouteDeclaration(None=@n156.B) {name=StarMask, nodeColor=<rgba 1 0.8438 0 1>}
  NamedRerouteUsage()
  NamedRerouteUsage()
  OneMinus()
  OneMinus(None=@n16)
  OneMinus(None=Power(Base=OneMinus(None=@n102), Exp=Divide() {constB=2.6}))
  PixelNormalWS()
  n157 = ScalarParameter() {defaultValue=1, parameterName=Reflect_Color_Boost}
  n158 = MaterialFunctionCall() {materialFunction=/Game/TimeOfDay/TODM_Sequence/Materials/Sky/Clouds/MF_Helios_Skybox_Clouds}
  n159 = NamedRerouteUsage()
  n160 = NamedRerouteDeclaration(None=LinearInterpolate(A=@n158.CloudAlpha, B=Multiply(A=@n158.CloudAlpha, B=@n159), Alpha=ScalarParameter() {group=KIln, parameterName=KlinCloudMaskBlend})) {name=cloudAlpha, nodeColor=<rgba 1 0 0.4687 1>}
  n161 = NamedRerouteUsage()
  n162 = StaticSwitchParameter(True=@n45, False=@n39.Result) {group=Horizon Clouds, parameterName=Enable Horizon Clouds}
  n163 = LinearInterpolate(A=@n162, B=@n158.CloudColor, Alpha=Saturate(None=Subtract(A=@n160, B=@n161))) {constAlpha=0, constB=0}
  n164 = CollectionParameter() {collection=&FortniteMaterialParameters, parameterName=SkyBoxClouds}
  n165 = QualitySwitch(Default=@n163, Epic=LinearInterpolate(A=@n162, B=@n163, Alpha=Saturate(None=@n164)) {constAlpha=1})
  n166 = MaterialFunctionCall(
   Value_Morning=ScalarParameter() {defaultValue=0.2, parameterName=ReflectionDesaturation_Morning}
   Value_Day=ScalarParameter() {defaultValue=0.2, parameterName=ReflectionDesaturation_Day}
   Value_Evening=ScalarParameter() {defaultValue=0.2, parameterName=ReflectionDesaturation_Evening}
   Value_Night=ScalarParameter() {defaultValue=0.45, parameterName=ReflectionDesaturation_Night}
  ) {materialFunction=&Scalar_Time_of_Day}
  n167 = Multiply(A=ScalarParameter() {defaultValue=0.7, parameterName=HadesDesaturation}, B=@n166.Result)
  n168 = CollectionParameter() {collection=&FortniteMaterialParameters, parameterName=HadesActive}
  ReflectionCapturePassSwitch!(
   Default=@n165
   Reflection=Desaturation(None=Multiply(A=@n157, B=@n165), Fraction=LinearInterpolate(A=@n167, B=@n166.Result, Alpha=@n168) {constA=0.7})
  )
  Saturate(None=@n62)
  n169 = ComponentMask(None=CameraPositionWS()) {b=true}
  ShadingPathSwitch(Default=Constant(), Mobile=AppendVector(A=Constant2Vector(), B=Multiply(A=@n169) {constB=0.9}))
  SkyAtmosphereLightIlluminance()
  SkyAtmosphereLightIlluminance()
  SkyAtmosphereViewLuminance()
  SkyAtmosphereViewLuminance()
  SmoothStep(Value=@n76)
  n170 = Saturate(None=@n76)
  n171 = ScalarParameter() {defaultValue=1, group=Horizon Clouds, parameterName=CloudRevealThreshold2}
  Subtract(A=@n47, B=OneMinus(None=Multiply(A=@n170, B=@n171)))
  Subtract(A=@n100) {constB=0.5}
  Subtract(A=ComponentMask(None=@n2) {r=true}, B=ComponentMask() {b=true, g=true})
  VectorParameter() {parameterName=Horizon Blend Color, defaultValue=<rgba 0.1886 0.4981 1 1>}
  VertexInterpolator(VS=VertexNormalWS())
 ]
}
