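# Material function that composites the Fortnite time-of-day sky dome, outputting emissive color and a mobile horizon-line position offset.
# Day and night gradients blend sky-atmosphere luminance with custom colors, adding stars at night.
# Sun/moon, high-altitude clouds and a storm horizon tint layer on by quality tier, then the main cloud blends over by masked alpha.
# Reflection captures receive a boosted copy, desaturated by time of day.
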MaterialFunction /Game/TimeOfDay/TODM_Sequence/Materials/Sky/MaterialFunctions/MF_UEFN_SkyDome
graph {nodes [
  CollectionParameter() {collection=&FortniteMaterialParameters, parameterName=SkyBoxClouds}
  n1 = ScalarParameter() {defaultValue=1, parameterName=Reflect_Color_Boost}
  n2 = Constant()
  n3 = ScalarParameter() {defaultValue=1, parameterName=CloudOpacity}
  n4 = NamedRerouteDeclaration(None=Multiply(
    A=FunctionInput(Preview=@n2) {bUsePreviewValueAsDefault=true, inputName=Cloud Alpha, inputType=FunctionInput_Scalar, sortPriority=5}
    B=NamedRerouteDeclaration(None=@n3) {name=CloudOpacity, nodeColor=<rgba 1 0 0.2109 1>}
   )) {name=cloudAlpha, nodeColor=<rgba 1 0 0.4688 1>}
  n5 = NamedRerouteUsage()
  n6 = FunctionInput(Preview=Constant3Vector()) {bUsePreviewValueAsDefault=true, description=Only Avaliable for High/Epic Qualities, inputName=High Atitude Clouds, sortPriority=4}
  n7 = FunctionInput(Preview=Constant3Vector()) {bUsePreviewValueAsDefault=true, inputName=SunMoon}
  n8 = Multiply(A=NamedRerouteUsage()) {constB=1.5}
  n9 = Multiply(A=@n7, B=OneMinus(None=Saturate(None=@n8)))
  n10 = Saturate(None=CollectionParameter() {collection=&FortniteMaterialParameters, parameterName=SkyBoxClouds})
  n11 = CollectionParameter() {collection=/Game/Athena/SafeZone/MPC_SafeZoneClouds, parameterName=CloudStartingBias}
  n12 = OneMinus(None=Saturate(None=Multiply(A=@n11) {constB=2}))
  n13 = Constant() {r=0.5}
  n14 = SphereMask(
   A=SphereMask(A=TextureCoordinate(), B=@n13, Radius=@n13, Hardness=Constant())
   B=Constant() {r=0.085}
   Radius=Constant() {r=0.07}
   Hardness=Constant() {r=0.1128}
  )
  n15 = LinearInterpolate(
   A=NamedRerouteUsage()
   B=VectorParameter() {parameterName=StormHorizonGradientColor, defaultValue=<rgba 0.0907 0.0645 0.125 1>}.RGB
   Alpha=NamedRerouteDeclaration(None=Multiply(A=@n12, B=@n14)) {name=StormGradientMask, nodeColor=<rgba 0.4297 1 0 1>}
  ) {constAlpha=1}
  n16 = Add(
   A=QualitySwitch(Default=@n9, Epic=LinearInterpolate(A=@n7, B=@n9, Alpha=@n10))
   B=QualitySwitch(Default=NamedRerouteUsage(), High=@n15, Epic=@n15)
  )
  n17 = Add(A=@n6, B=@n16)
  n18 = MaterialFunctionCall!(
   Low _ PC Low and Med=@n16
   "High _ XB1, PS4"=@n17
   "Epic _ PC High, XSX, PS5"=@n17
   --------
   "Mobile Low _ PC Perf, Android"=@n16
   Mobile Med _ Switch=@n16
  ) {materialFunction=/Game/Global/MaterialFunctions/TechArt/MF_Fortnite_QualitySwitch_Vector3}
  n19 = LinearInterpolate(
   A=@n18.Result
   B=FunctionInput(Preview=Constant3Vector()) {bUsePreviewValueAsDefault=true, inputName=Cloud, sortPriority=6}
   Alpha=Saturate(None=Subtract(A=@n4, B=@n5))
  ) {constAlpha=0, constB=0}
  n20 = ScalarParameter() {desc=Needs to be Parameter for the BP to manipulate, parameterName=SkyBoxClouds}
  n21 = QualitySwitch(Default=@n19, Epic=LinearInterpolate(A=@n18.Result, B=@n19, Alpha=Saturate(None=@n20)) {constAlpha=1})
  n22 = ScalarParameter() {defaultValue=0.2, parameterName=ReflectionDesaturation_Morning}
  n23 = ScalarParameter() {defaultValue=0.2, parameterName=ReflectionDesaturation_Day}
  n24 = ScalarParameter() {defaultValue=0.2, parameterName=ReflectionDesaturation_Evening}
  n25 = ScalarParameter() {defaultValue=0.45, parameterName=ReflectionDesaturation_Night}
  n26 = ReflectionCapturePassSwitch!(
   Default=@n21
   Reflection=Desaturation(
    None=Multiply(A=@n1, B=@n21)
    Fraction=MaterialFunctionCall(Value_Morning=@n22, Value_Day=@n23, Value_Evening=@n24, Value_Night=@n25) {materialFunction=&Scalar_Time_of_Day}.Result
   )
  )
  FunctionOutput(None=@n26) {outputName=Emissive Result}
  n27 = Constant()
  n28 = CameraPositionWS()
  n29 = AppendVector(A=Constant2Vector(), B=Multiply(A=ComponentMask(None=@n28) {b=true}) {constB=0.9})
  FunctionOutput(None=NamedRerouteDeclaration(None=ShadingPathSwitch(Default=@n27, Mobile=@n29)) {name=World Position Offet Output, nodeColor=<rgba 0 1 0.2969 1>}) {description=Offset for Moblie's Horizon Line, outputName=World Position Offset, sortPriority=2}
  n30 = TextureCoordinate() {uTiling=0.5, vTiling=0.5}
  n31 = Constant() {r=0.25}
  n32 = AppendVector(A=@n31, B=@n31)
  n33 = LinearInterpolate(Alpha=ScalarParameter() {defaultValue=0.25, group=Night Sky, parameterName=Nightsky Horizon Height}) {constA=0.25, constB=0.2}
  n34 = Constant()
  n35 = CurveAtlasRowParameter(CurveTime=NamedRerouteDeclaration(None=SphereMask(A=@n30, B=@n32, Radius=@n33, Hardness=@n34)) {name=HorizonLine, nodeColor=<rgba 0.875 1 0 1>}) {
   atlas=&/CA_Athena_TOD
   curve=&/NightSkyHorizonCurve
   defaultValue=1
   parameterName=Sky Horizon Curve
  }
  n36 = Saturate(None=@n35.G)
  LinearInterpolate(A=NamedRerouteUsage(), B=NamedRerouteUsage(), Alpha=@n36)
  n37 = ComponentMask(None=@n26) {b=true, g=true, r=true}
  n38 = ScalarParameter() {defaultValue=1, desc=Mix to Original, group=FX Emissive EyeAdaptation, parameterName=EyeAdaptation Alpha}
  n39 = Constant() {r=2}
  n40 = Multiply(A=@n38) {constB=0, desc=keeps default eye adaption for lumen}
  LinearInterpolate(
   A=@n37
   B=Divide(A=@n37, B=EyeAdaptation().EyeAdaptation)
   Alpha=QualitySwitch(Default=Multiply(A=@n38, B=@n39) {constB=20}, High=@n40, Epic=@n40)
  )
  n41 = VectorParameter() {group=Night Sky, parameterName=NightSky Horizon Color, defaultValue=<rgba 0.0581 0.4849 1 1>}
  n42 = VectorParameter() {group=Night Sky, parameterName=Night Sky Color, defaultValue=<rgba 0.0135 0.0198 0.026 1>}
  n43 = LinearInterpolate(
   A=LinearInterpolate(
    A=LinearInterpolate(A=@n41.RGB, B=@n42.RGB, Alpha=@n36)
    B=NamedRerouteUsage()
    Alpha=ScalarParameter() {parameterName=GradientColorStrength}
   )
   B=SkyAtmosphereViewLuminance!(WorldDirection)
   Alpha=ScalarParameter() {defaultValue=1, group=Night Sky, parameterName=NightSkyViewLuminanceFactor}
  ) {constA=1}
  n44 = Add(A=@n43, B=FunctionInput(Preview=Constant3Vector()) {bUsePreviewValueAsDefault=true, inputName=Stars})
  n45 = ScalarParameter() {defaultValue=1, group=Night Sky, parameterName=Nightsky Galaxy Strength}
  NamedRerouteDeclaration(None=LinearInterpolate(
    A=NamedRerouteUsage()
    B=LinearInterpolate(A=@n43, B=@n44, Alpha=@n45)
    Alpha=ScalarParameter() {defaultValue=1, group=Night Sky, parameterName=Night Sky blend}
   )) {name=Sky Gradiant, nodeColor=<rgba 1 0 0.8438 1>}
  n46 = WorldPosition()
  n47 = CameraPositionWS()
  n48 = DotProduct(A=SkyAtmosphereLightDirection(), B=Normalize(VectorInput=Subtract(A=@n46.XYZ, B=@n47)))
  NamedRerouteDeclaration(None=Saturate(None=ConstantBiasScale(None=@n48))) {name=LightDirectionGradient, nodeColor=<rgba 1 0.4688 0 1>}
  n49 = Multiply(A=SkyAtmosphereViewLuminance(), B=ScalarParameter() {defaultValue=1, parameterName=SkyAtmosphereBrightness})
  n50 = Saturate(None=LinearInterpolate(A=NamedRerouteUsage(), B=NamedRerouteUsage(), Alpha=ScalarParameter() {parameterName=ColorGradientUseLightDirection}))
  n51 = Saturate(None=ScalarParameter() {defaultValue=0.5, parameterName=GradientCenter})
  n52 = LinearInterpolate(
   A=NamedRerouteDeclaration(None=VectorParameter() {parameterName=ColorLow, defaultValue=<rgba 1 0.0615 0 1>}.RGB) {name=Custom Low Color, nodeColor=<rgba 0.2656 0 1 1>}
   B=NamedRerouteDeclaration(None=VectorParameter() {parameterName=ColorMid, defaultValue=<rgba 1 0.6067 0.349 1>}.RGB) {name=Custom Mid Color, nodeColor=<rgba 0 1 0.9297 1>}
   Alpha=Saturate(None=Divide(A=@n50, B=@n51))
  )
  n53 = VectorParameter() {parameterName=ColorHigh, defaultValue=<rgba 0.3338 0.7006 1 1>}
  n54 = Saturate(None=Divide(A=Subtract(A=@n50, B=@n51), B=@n51))
  n55 = Add(
   A=NamedRerouteDeclaration(None=LinearInterpolate(A=@n52, B=@n53.RGB, Alpha=@n54)) {name=GradientColor, nodeColor=<rgba 0 0.4844 1 1>}
   B=@n49
  )
  NamedRerouteDeclaration(None=LinearInterpolate(
    A=@n49
    B=NamedRerouteDeclaration(None=@n55) {name=Custom Day Gradiant, nodeColor=<rgba 0 1 0.5312 1>}
    Alpha=ScalarParameter() {parameterName=GradientColorStrength}
   )) {name=Day Gradiant, nodeColor=<rgba 0.5 0 1 1>}
  n56 = ScalarParameter() {defaultValue=-0.5, parameterName=GradientBias}
  n57 = Multiply(
   A=OneMinus(None=@n56)
   B=MaterialFunctionCall() {materialFunction=&BoundingBoxBased_0-1_UVW}.B
  )
  NamedRerouteDeclaration(None=Saturate(None=Add(A=@n57, B=@n56))) {name=VerticalGradient, nodeColor=<rgba 0 0.1562 1 1>}
  ScalarParameter() {desc=Needs to be Parameter for the BP to manipulate, parameterName=SkyBoxClouds}
 ]}
