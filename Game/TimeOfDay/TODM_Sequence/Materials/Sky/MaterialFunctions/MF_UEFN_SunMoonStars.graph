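# Sky material function for the UEFN time-of-day sequence that outputs a sun/moon color and a night star layer.
# MoonActive blends the two SunMoon inputs, built around an atmosphere sun disk and a textured moon with halo, projected using the atmosphere light direction.
# Stars multiply tiled texture channels by offset masks, then apply panning noise twinkle, replaced by a constant on low/mobile.
# Vertex interpolators carry moon UVs, and a halo size parameter supports the day sequence device.
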MaterialFunction /Game/TimeOfDay/TODM_Sequence/Materials/Sky/MaterialFunctions/MF_UEFN_SunMoonStars
graph {nodes [
  n1 = NamedRerouteUsage()
  n2 = NamedRerouteUsage()
  n3 = Saturate(None=ScalarParameter() {group=Sun-Moon, parameterName=MoonActive})
  FunctionOutput(None=NamedRerouteDeclaration(None=LinearInterpolate(A=@n1, B=@n2, Alpha=@n3)) {name=SunMoon, nodeColor=<rgba 0 0.2969 1 1>}) {outputName=SunMoon}
  n4 = NamedRerouteUsage()
  n5 = TextureObjectParameter() {
   group=Night Sky
   parameterName=StarTexture
   samplerType=SAMPLERTYPE_LinearColor
   texture=/Game/TimeOfDay/TODM_Sequence/Materials/Sky/Textures/SkyNightMap_output
  }
  n6 = NamedRerouteDeclaration(None=@n5) {name=Star Texture Object, nodeColor=<rgba 0 1 0.6016 1>}
  n7 = TextureSample(UVs=Multiply(A=@n4, B=ScalarParameter() {defaultValue=20, parameterName=Start1}) {constB=5}, Tex=@n6) {samplerType=SAMPLERTYPE_LinearColor}
  n8 = TextureSample(UVs=Multiply(A=@n4, B=ScalarParameter() {defaultValue=8, parameterName=StarMask}), Tex=@n6) {samplerSource=SSM_Wrap_WorldGroupSettings, samplerType=SAMPLERTYPE_LinearColor}
  n9 = NamedRerouteUsage()
  n10 = TextureSample(UVs=Multiply(A=@n9, B=ScalarParameter() {defaultValue=12, parameterName=Start2}) {constB=5}, Tex=@n5) {samplerSource=SSM_Wrap_WorldGroupSettings, samplerType=SAMPLERTYPE_LinearColor}
  n11 = TextureSample(UVs=Multiply(A=@n9, B=ScalarParameter() {defaultValue=15, parameterName=StarMask2}), Tex=@n5) {samplerSource=SSM_Wrap_WorldGroupSettings, samplerType=SAMPLERTYPE_LinearColor}
  n12 = Multiply(A=Constant3Vector() {constant=<rgba 43.7499 39.9786 39.1023 1>}, B=Add(A=Multiply(A=@n7.G, B=@n8.R), B=Multiply(A=@n10.G, B=@n11.B)))
  n13 = NamedRerouteUsage()
  n14 = Multiply(A=NamedRerouteUsage(), B=Constant() {r=5})
  n15 = Multiply(
   A=Multiply(A=LinearInterpolate(B=@n12, Alpha=@n13), B=ScalarParameter() {defaultValue=1, parameterName=StarsBrightness})
   B=LinearInterpolate(Alpha=Saturate(None=@n14)) {constA=1, constB=0.6}
  )
  n16 = TextureSample(UVs=Panner(Coordinate=TextureCoordinate() {uTiling=18, vTiling=18}) {speedX=0.36}) {
   samplerSource=SSM_Wrap_WorldGroupSettings
   samplerType=SAMPLERTYPE_Grayscale
   texture=/Game/TimeOfDay/TODM_Sequence/Materials/Sky/Textures/LowResBlurredNoise_Radish_LF_Root
  }
  n17 = MaterialFunctionCall!(Default=LinearInterpolate(Alpha=@n16.R) {constA=0.05}, Low/Mobile=Constant() {r=0.8}) {materialFunction=/Game/Effects/Fort_Effects/Materials/_Functions/MF_LowMobile_PathingSwitch}
  n18 = NamedRerouteUsage()
  n19 = NamedRerouteUsage()
  FunctionOutput(None=LinearInterpolate(A=Multiply(A=@n15, B=@n17.Result), Alpha=Multiply(A=@n18, B=@n19)) {constB=0}) {description=Milkyway Stars Output, outputName=Stars, sortPriority=5}
  n20 = Constant() {r=0.5}
  n21 = MaterialFunctionCall(Float2=NamedRerouteUsage()) {materialFunction=&BreakOutFloat2Components}
  NamedRerouteDeclaration(None=Multiply(A=SphereMask(A=@n20, B=@n21.R) {attenuationRadius=0.5}, B=SphereMask(A=@n21.G, B=@n20) {attenuationRadius=0.5})) {name=Mask 01 UVs, nodeColor=<rgba 1 0 0.9141 1>}
  n22 = WorldPosition()
  n23 = CameraPositionWS()
  n24 = NamedRerouteDeclaration(None=Normalize(VectorInput=Subtract(A=@n22.XYZ, B=@n23))) {name=CameraVectorForVertexShader, nodeColor=<rgba 0 1 0.0859 1>}
  n25 = ComponentMask(None=SkyAtmosphereLightDirection()) {b=true, g=true, r=true}
  NamedRerouteDeclaration(None=Step(X=DotProduct(A=@n24, B=@n25))) {name=HemisphereMask, nodeColor=<rgba 1 0 0.3047 1>}
  n26 = Multiply(A=@n25) {constB=-1}
  n27 = CrossProduct(A=@n26, B=Constant3Vector() {constant=<rgba 0 0 -1 0>})
  n28 = Normalize(VectorInput=CrossProduct(A=@n27, B=@n26))
  n29 = NamedRerouteUsage()
  n30 = Normalize(VectorInput=@n27)
  NamedRerouteDeclaration(None=AppendVector(A=DotProduct(A=@n28, B=@n29), B=DotProduct(A=@n30, B=@n29))) {name=SunMoonUVs, nodeColor=<rgba 1 0 0.9141 1>}
  n31 = Divide(A=NamedRerouteUsage(), B=ScalarParameter() {defaultValue=0.07, group=Sun-Moon, parameterName=MoonSize})
  NamedRerouteDeclaration(None=VertexInterpolator(VS=Add(A=@n31) {constB=0.5}).PS) {name=Moon UVs, nodeColor=<rgba 1 0 0.6562 1>}
  n32 = ScalarParameter() {defaultValue=1, group=Sun-Moon, parameterName=SunBrightness, sliderMax=1}
  n33 = SkyAtmosphereLightDiskLuminance!(DiskAngularDiameterOverride=ScalarParameter() {defaultValue=2, group=Sun-Moon, parameterName=SunSize})
  n34 = Multiply(A=@n33, B=VectorParameter() {group=Sun-Moon, parameterName=SunColor, defaultValue=<rgba 1 0.2982 0 1>}.RGB)
  n35 = ScalarParameter() {parameterName=SunColorBlend}
  NamedRerouteDeclaration(None=Multiply(A=Saturate(None=@n32), B=LinearInterpolate(A=@n33, B=@n34, Alpha=@n35))) {name=Sun, nodeColor=<rgba 0 0.9766 1 1>}
  n36 = NamedRerouteUsage()
  n37 = ScalarParameter() {defaultValue=5, desc=Added because of the day sequence device in UEFN, parameterName=MoonHaloSize}
  n38 = SphereMask(
   A=VertexInterpolator(VS=Add(A=@n36) {constB=0.5}).PS
   B=Constant() {r=0.5}
   Radius=Multiply(
    A=ScalarParameter() {defaultValue=0.8, group=Sun-Moon, parameterName=Moon Halo Radius}
    B=Multiply(A=@n37) {constB=0.2, desc=Adjusted so the default value doesn't mess w/ BR}
   )
   Hardness=ScalarParameter() {defaultValue=-0.2, group=Sun-Moon, parameterName=moon halo Hardness}
  ) {hardnessPercent=0.1}
  NamedRerouteDeclaration(None=@n38) {name=MoonHaloMask, nodeColor=<rgba 0 0.3672 1 1>}
  n39 = TextureSample(
   UVs=NamedRerouteUsage()
   Tex=TextureObjectParameter() {parameterName=MoonTexture, samplerType=SAMPLERTYPE_LinearColor, texture=/Game/TimeOfDay/Skydome/T_Moon_Artemis}
  ) {samplerType=SAMPLERTYPE_LinearColor, texture=&T_Moon_Artemis}
  n40 = NamedRerouteUsage()
  n41 = Multiply(
   A=Multiply(A=Multiply(A=@n39.RGB, B=@n40), B=VectorParameter() {group=Sun-Moon, parameterName=MoonColor, defaultValue=<rgba 1.7 1.7 1.7 1>}.RGB)
   B=ScalarParameter() {defaultValue=14, parameterName=MoonIntensity}
  ) {constB=14}
  n42 = Multiply(
   A=Multiply(A=Multiply(A=@n38, B=@n38), B=VectorParameter() {group=Sun-Moon, parameterName=moon halo Color, defaultValue=<rgba 1 1 1 1>}.RGB)
   B=ScalarParameter() {defaultValue=1, parameterName=MoonHaloIntensity}
  )
  NamedRerouteDeclaration(None=Multiply(A=NamedRerouteUsage(), B=Add(A=@n41, B=@n42))) {name=Moon, nodeColor=<rgba 0 1 0.8125 1>}
  NamedRerouteDeclaration(None=@n3) {name=MoonActive, nodeColor=<rgba 0.0312 1 0 1>}
  n43 = Multiply(A=@n39.R) {constB=20}
  NamedRerouteDeclaration(None=Multiply(A=Saturate(None=@n43), B=NamedRerouteUsage())) {name=MoonMask, nodeColor=<rgba 0 1 0.4844 1>}
  n44 = Constant()
  n45 = NamedRerouteDeclaration(None=ScalarParameter() {defaultValue=0.016, group=Night Sky, parameterName=NightSkyRotation}) {name=NightSkyRotation, nodeColor=<rgba 0 1 0.4375 1>}
  n46 = NamedRerouteDeclaration(None=Add(A=TextureCoordinate(), B=AppendVector(A=@n44, B=@n45))) {name=StarPanner, nodeColor=<rgba 0.3594 0 1 1>}
  n47 = ScalarParameter() {defaultValue=1, group=Night Sky, parameterName=MIkyWaySize}
  n48 = TextureSample(
   UVs=Multiply(
    A=MaterialFunctionCall(UVs=@n46, Texture Scale=@n47) {materialFunction=&ScaleUVsByCenter}.UVs
    B=ScalarParameter() {defaultValue=5500, group=Night Sky, parameterName=StarmaskFrequency}
   )
   Tex=NamedRerouteUsage()
  ) {
   samplerSource=SSM_Wrap_WorldGroupSettings
   samplerType=SAMPLERTYPE_LinearColor
   texture=/Game/TimeOfDay/TODM_Sequence/Materials/Sky/Textures/T_milkyway_StarMask
  }
  NamedRerouteDeclaration(None=@n48.B) {name=StarMask, nodeColor=<rgba 1 0.8438 0 1>}
 ]}
