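# M_UEFN_Sky_Ch5 is an unlit sky material, flagged as sky, for the UEFN time-of-day sky, apparently for Chapter 5.
# It is assembled entirely from three material functions.
# MF_UEFN_SunMoonStars supplies the sun, moon and star inputs, and MF_UEFN_Clouds supplies the cloud, cloud alpha and high-altitude cloud inputs.
# Both feed MF_UEFN_SkyDome, the final function that composes the sky.
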Material /Game/TimeOfDay/TODM_Sequence/Materials/Sky/M_UEFN_Sky_Ch5
graph {
 properties {bIsSky=true, pixelDepthOffsetMode=PDOM_Legacy, shadingModel=MSM_Unlit, usedShadingModels=Unlit}
 nodes [
  n1 = MaterialFunctionCall() {materialFunction=/Game/TimeOfDay/TODM_Sequence/Materials/Sky/MaterialFunctions/MF_UEFN_SunMoonStars}
  n2 = MaterialFunctionCall() {materialFunction=/Game/TimeOfDay/TODM_Sequence/Materials/Sky/MaterialFunctions/MF_UEFN_Clouds}
  MaterialFunctionCall!(
   SunMoon=@n1.SunMoon
   Stars=@n1.SunMoon
   High Atitude Clouds=@n2.High Altitude Clouds
   Cloud Alpha=@n2.High Altitude Clouds
   Cloud=@n2.High Altitude Clouds
  ) {materialFunction=/Game/TimeOfDay/TODM_Sequence/Materials/Sky/MaterialFunctions/MF_UEFN_SkyDome}
 ]
}
